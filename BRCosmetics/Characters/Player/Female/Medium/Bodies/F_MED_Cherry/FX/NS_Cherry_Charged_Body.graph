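# Niagara system for Cherry's charged body cosmetic, dressing the female character's skeletal mesh in warm orange fire and energy.
# Arm emitters flow fire and velocity-aligned sprites along bones via FN_BoneFlow, alongside lightning on the forearms.
# Spawn groups pick bones and mirrored offsets for lightning, fist glows and rays at clavicles and head.
# Curl-noise embers rise from the mesh, and a flickering light follows the root; a second light emitter is disabled.
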NiagaraSystem /BRCosmetics/Characters/Player/Female/Medium/Bodies/F_MED_Cherry/FX/NS_Cherry_Charged_Body
dependencies {
 usedDataInterfaces [&NiagaraDataInterfaceCurve, &NiagaraDataInterfaceColorCurve]
 usedDynamicInputs []
 usedModules [&SystemState]
}
emitters {
 Cherry_Arms_BoneFlow_Fire {
  emitterUpdateScript [
   &EmitterState()
   &SpawnRate(SpawnRate=5, Use Spawn Probability=false)
  ]
  particleSpawnScript [
   /Niagara/Modules/Spawn/Initialization/InitializeParticle(
    Lifetime=&RandomRangeFloat(Minimum=1.6, Maximum=2)
    Module.Write Color=true
    Module.Write Lifetime=true
    Module.Write Mass=true
    Module.Write Material Random=true
    Module.Write Position=false
    Module.Write Scale=false
    Module.Write SpriteRotation=true
    Module.Write SpriteSize=true
    Position=&SimulationPosition()
    Sprite Rotation=&RandomRangeFloat(Minimum=-18, Maximum=180)
    Sprite Size=<15 15>
   )
   &FN_SetSpawnGroup(Inclusivity=true, Normalize Spawn Group=true, Particles Per Group=1, Pattern=::NewEnumerator0, Set Ribbon ID and Ribbon Link Order=true) {module=FN_SetSpawnGroup_v4}
   /SetVariables(Particles.RandomFrame=&RandomRangeInt(Minimum=0, Maximum=63)) {isSetParameters=true, module=SetVariables_#1}
  ]
  particleUpdateScript [
   &ParticleState()
   &SolveForcesAndVelocity(Clamp Velocity=false, Limit Acceleration=false)
   &FN_BoneFlow(
    Angle:NiagaraFloat=&Add_Float(
     A=&RandomRangeFloat(Minimum=-1, Maximum=1)
     B=&Multiply_Float(A=$Particles.NormalizedAge, B=0.5)
    )
    BoneID:NiagaraInt32=&RandomRangeInt(Minimum=0, Maximum=1, Randomness Mode=::NewEnumerator1, Random Seed=$Particles.SpawnGroup, Fixed Random Seed=true)
    Height:NiagaraFloat=&FloatFromCurve(FloatCurve=DataInterface() {propertyValues {Curve {keys [
         {arriveTangent=1.1209, interpMode=RCIM_Cubic, leaveTangent=-0.2542, tangentMode=RCTM_User, time=0.0182, value=0.0632}
         {arriveTangent=1.147, interpMode=RCIM_Cubic, leaveTangent=1.3866, tangentMode=RCTM_User, time=1, value=1}
        ]}}})
    Radius:NiagaraFloat=&FloatFromCurve(FloatCurve=DataInterface() {propertyValues {Curve {keys [
         {arriveTangent=1.2509, interpMode=RCIM_Cubic, leaveTangent=1.2509, tangentMode=RCTM_User, time=-0.0001, value=0.5073}
         {arriveTangent=-0.3642, interpMode=RCIM_Cubic, leaveTangent=-0.3642, tangentMode=RCTM_User, time=0.5658, value=0.8726}
         {arriveTangent=-1.426, interpMode=RCIM_Cubic, leaveTangent=-1.426, tangentMode=RCTM_User, time=1.1003, value=0.3729}
        ]}}})
    SkeletalMesh:NiagaraDataInterfaceSkeletalMesh=$User.SkeletalMesh
    Write Position Attribute=true
   )
   &ScaleSpriteSize(Scale Factor=&Vector2DFromFloat(Value=&FloatFromCurve(FloatCurve=DataInterface() {propertyValues {Curve {keys [
          {arriveTangent=3.9126, interpMode=RCIM_Cubic, leaveTangent=3.9126, tangentMode=RCTM_User, time=0.0023, value=0.8242}
          {arriveTangent=-0.8304, interpMode=RCIM_Cubic, leaveTangent=-0.8304, tangentMode=RCTM_User, time=0.4979, value=1.178}
          {arriveTangent=-0.5055, interpMode=RCIM_Cubic, leaveTangent=-0.5055, tangentMode=RCTM_User, time=1.005, value=0.6778}
         ]}}})))
   &CameraOffset(Camera Offset Amount=5)
   &ScaleColor(
    Scale Mode=::NewEnumerator1
    Scale RGBA=&Vector4FromCurve(Vector4Curve=DataInterface() {propertyValues {
       AlphaCurve {keys [<at 0.0008 0>,<at 0.0878 1>,<at 0.9069 1>,<at 1.0267 0>]}
       BlueCurve {keys [<at 0.0122 0.1172>,<at 0.9388 0.3038>]}
       GreenCurve {keys [<at 0.0122 0.6244>,<at 0.9388 3.4008>]}
       RedCurve {keys [<at 0.0122 1.5>,<at 0.9388 5>]}
      }})
    ScaleRGBA=true
   )
   &FN_SubUVAnimation(End Frame=&Add_Integer(A=$Particles.RandomFrame, B=63), Start Frame=$Particles.RandomFrame)
  ]
  rendererClasses [&NiagaraSpriteRendererProperties]
  renderers [{rendererClass=&NiagaraSpriteRendererProperties}]
 }
 Cherry_Arms_BoneFlow_VelocityAligned {
  emitterUpdateScript [
   &EmitterState()
   &SpawnRate(SpawnRate=4, Use Spawn Probability=false)
  ]
  particleSpawnScript [
   &InitializeParticle(
    Lifetime=&RandomRangeFloat(Minimum=0.7, Maximum=0.9)
    Module.Write Color=true
    Module.Write Lifetime=true
    Module.Write Mass=true
    Module.Write Material Random=true
    Module.Write Position=false
    Module.Write Scale=false
    Module.Write SpriteRotation=true
    Module.Write SpriteSize=true
    Position=&SimulationPosition()
    Sprite Size=<15 20>
   )
   &FN_SetSpawnGroup(Inclusivity=true, Normalize Spawn Group=true, Particles Per Group=1, Pattern=::NewEnumerator0, Set Ribbon ID and Ribbon Link Order=true) {module=FN_SetSpawnGroup_v4}
   /SetVariables(Particles.RandomFrame=&RandomRangeInt(Minimum=0, Maximum=63)) {enabled=false, isSetParameters=true, module=SetVariables_#1}
   /SetVariables(Particles.UVScale:Vector2f=&MakeVector2D(X=&UniformAOrBFloat(A=-1), Y=&UniformAOrBFloat(A=-1))) {enabled=false, isSetParameters=true}
   &SubUVAnimation(End Frame=1, SubUV Animation Mode:ENiagara_SubUVLookupMode=::NewEnumerator2)
  ]
  particleUpdateScript [
   &ParticleState()
   &SolveForcesAndVelocity(Clamp Velocity=false, Limit Acceleration=false)
   &FN_BoneFlow(
    Angle:NiagaraFloat=&Add_Float(
     A=&RandomRangeFloat(Minimum=-1, Maximum=1)
     B=&Multiply_Float(A=$Particles.NormalizedAge, B=0.5)
    )
    BoneID:NiagaraInt32=&RandomRangeInt(Minimum=0, Maximum=1, Randomness Mode=::NewEnumerator1, Fixed Random Seed=true)
    Height:NiagaraFloat=&FloatFromCurve(FloatCurve=DataInterface() {propertyValues {Curve {keys [{arriveTangent=1.1209, time=0.0182, value=0.5}, {arriveTangent=1.147, leaveTangent=1.3866, time=1, value=1}]}}})
    SkeletalMesh:NiagaraDataInterfaceSkeletalMesh=$User.SkeletalMesh
    Write Position Attribute=true
   )
   &ScaleSpriteSize(Scale Factor=&Vector2DFromFloat(Value=&FloatFromCurve(FloatCurve=DataInterface() {propertyValues {Curve {keys [
          {arriveTangent=3.9126, interpMode=RCIM_Cubic, time=0, value=0}
          {arriveTangent=0.9951, interpMode=RCIM_Cubic, leaveTangent=0.9951, time=0.457, value=1}
          {interpMode=RCIM_Cubic, leaveTangent=-0.5055, time=1.005, value=1}
         ]}}})))
   &CameraOffset(Camera Offset Amount=5)
   &ScaleColor(
    Scale Mode=::NewEnumerator1
    Scale RGBA=&Vector4FromCurve(Vector4Curve=DataInterface() {propertyValues {
       AlphaCurve {keys [<at 0.4016 0>,<at 0.9069 1>,<at 1.0267 0>]}
       BlueCurve {keys [<at 0.0122 0.2883>,<at 0.9388 0.961>]}
       GreenCurve {keys [<at 0.0122 1.1865>,<at 0.9388 3.9551>]}
       RedCurve {keys [<at 0.0122 1.5>,<at 0.9388 5>]}
      }})
    ScaleRGBA=true
   )
   &FN_SubUVAnimation(End Frame=&Add_Integer(A=$Particles.RandomFrame, B=63), Start Frame=$Particles.RandomFrame) {enabled=false}
   &CalculateAccurateVelocity()
  ]
  rendererClasses [&NiagaraSpriteRendererProperties]
  renderers [{rendererClass=&NiagaraSpriteRendererProperties}]
 }
 Cherry_Arms_Lightning001 {
  emitterUpdateScript [
   &EmitterState()
   &SpawnRate(SpawnRate=3, Use Spawn Probability=false)
  ]
  particleSpawnScript [
   &InitializeParticle(
    Lifetime=&RandomRangeFloat(Minimum=0.5, Maximum=1)
    Mesh Scale=&VectorFromFloat(Value=&RandomRangeFloat(Minimum=0.5, Maximum=0.7))
    Module.Write Color=true
    Module.Write Lifetime=true
    Module.Write Mass=true
    Module.Write Material Random=true
    Module.Write Position=false
    Module.Write Scale=true
    Module.Write SpriteRotation=false
    Module.Write SpriteSize=false
    Position=&SimulationPosition()
   )
   &FN_SetSpawnGroup(Number of Spawn Groups=2)
   &InitialMeshOrientation(
    Rotation=&MakeVector(X=&RandomRangeFloat(Minimum=-1, Maximum=1), Y=0, Z=0)
    Use Orientation Vector=true
    Use Rotation Vector=true
   )
  ]
  particleUpdateScript [
   &ParticleState()
   &SkeletalMeshLocation(
    Apply Sampled Orientation to Mesh Particles=true
    Bone / Socket Index=$Particles.SpawnGroup
    Bone Sampling Mode=::NewEnumerator4
    Inherit Velocity=false
    MaskSpawnGroup=false
    Sampled Position Offset=&MakeVector(
     X=/Game/Effects/Niagara/DynamicInputs/FN_PickFloatFromSpawnGroupCutoff(
      Over Cutoff Float:NiagaraFloat=&FloatFromCurve(FloatCurve=DataInterface() {propertyValues {Curve {keys [<at 0 -0.25>,<at 1 1>]}}}, Scale Curve=20)
      Under Cutoff Float:NiagaraFloat=&FloatFromCurve(FloatCurve=DataInterface() {propertyValues {Curve {keys [<at 0 0.4>,<at 1 1.65>]}}}, Scale Curve=-20)
      SpawnGroup:NiagaraInt32=$Particles.SpawnGroup
      Cutoff:NiagaraInt32=1
     )
     Y=0
     Z=0
    )
    Sampled Position Offset Coordinate Space:ENiagaraMeshTransforms=::NewEnumerator3
    Skeletal Mesh=DataInterface() {propertyValues {
      DefaultMesh=null
      ExcludeBoneName=root
      OptionalComponentClass=null
      PreviewMesh=/BRCosmetics/Characters/Player/Female/Medium/Bodies/F_MED_Cherry/Meshes/F_MED_Cherry
      SkinningMode=SkinOnTheFly
      SoftSourceActor=null
      SourceMode=Default
      UvSetIndex=0
      WholeMeshLOD=-1
      bExcludeBone=false
      bReadDeformedGeometry=true
      bRequireCurrentFrameData=true
      ComponentTags []
      FilteredBones [lowerarm_r,lowerarm_l]
      FilteredSockets []
      MeshUserParameter {parameter {name=None, typeDefHandle {registeredTypeIndex=1}, varData []}}
      SamplingRegions []
     }}
    Write Sampled Mesh Attrributes to Particle Attributes=true
   )
   &SolveForcesAndVelocity(Clamp Velocity=false, Limit Acceleration=false) {enabled=false}
   &CameraOffset(Camera Offset Amount=5)
   &ScaleColor(
    Scale Mode=::NewEnumerator1
    Scale RGBA=&Vector4FromCurve(Vector4Curve=DataInterface() {propertyValues {
       AlphaCurve {keys [<at 0.0008 0>,<at 0.1906 1>,<at 0.9069 1>,<at 1.0267 0>]}
       BlueCurve {keys [<at 0.0527 0.2648>,<at 1.0052 0.2326>]}
       GreenCurve {keys [<at 0.0527 0.519>,<at 1.0052 1.2316>]}
       RedCurve {keys [<at 0.0527 1>,<at 1.0052 2>]}
      }})
    ScaleRGBA=true
   )
   &DynamicMaterialParameters(
    Index 0 Param 1=2
    Index 0 Param 2=0.1
    Index 0 Param 3=7
    Index 0 Param 4=7
    Module.Param0WriteEnabled=true
    Module.Param1WriteEnabled=true
    Module.Param2WriteEnabled=true
    Module.Param3WriteEnabled=true
   )
   &ScaleMeshSize(Scale Factor=&VectorFromFloat(Value=&FloatFromCurve(FloatCurve=DataInterface() {propertyValues {Curve {keys [
          {arriveTangent=4.4642, interpMode=RCIM_Cubic, leaveTangent=4.4642, tangentMode=RCTM_User, time=0, value=0}
          <at 0.441 1.0367 RCIM_Cubic>
          {arriveTangent=-2.5534, interpMode=RCIM_Cubic, leaveTangent=-2.5534, tangentMode=RCTM_User, time=1, value=0}
         ]}}}))) {module=ScaleMeshSize001}
  ]
  rendererClasses [&NiagaraSpriteRendererProperties, &NiagaraMeshRendererProperties]
  renderers [
   {rendererClass=&NiagaraSpriteRendererProperties}
   {rendererClass=&NiagaraMeshRendererProperties, rendererIndex=1}
  ]
 }
 Lights {
  emitterUpdateScript [
   &EmitterState(
    Inactive Response=::NewEnumerator1
    Life Cycle Mode=::NewEnumerator1
    Loop Behavior=::NewEnumerator1
    Loop Duration Mode=::NewEnumerator1
    Reset Age On Awaken=true
    Scalability Mode=::NewEnumerator1
    UseLoopDelay=false
   )
   &SpawnBurst_Instantaneous(Spawn Count=1, Use Spawn Probability=false)
  ]
  particleSpawnScript [&InitializeParticle(
    Color=<rgba 1 0.6101 0.1884 1>
    Module.Write Color=true
    Module.Write Lifetime=true
    Module.Write Mass=true
    Module.Write Material Random=true
    Module.Write Position=false
    Module.Write Scale=false
    Module.Write SpriteRotation=false
    Module.Write SpriteSize=true
    Position=&SimulationPosition()
    Sprite Size=<100 100>
   )]
  particleUpdateScript [
   &ParticleState(Kill Particles When Lifetime Has Elapsed=false, Let Infinitely Lived Particles Die When Emitter Deactivates=true)
   &SkeletalMeshLocation(
    Inherit Velocity=false
    Interpolate Mesh Sampling=false
    MaskSpawnGroup=false
    Mesh Sampling Type=::NewEnumerator5
    Sampled Position Offset=<0 0 75>
    Sampled Position Offset Coordinate Space:ENiagaraMeshTransforms=::NewEnumerator3
    Skeletal Mesh=DataInterface() {propertyValues {
      DefaultMesh=null
      ExcludeBoneName=dyn_ponytail_06
      OptionalComponentClass=null
      PreviewMesh=&F_MED_Cherry
      SkinningMode=SkinOnTheFly
      SoftSourceActor=null
      SourceMode=Default
      UvSetIndex=0
      WholeMeshLOD=-1
      bExcludeBone=false
      bReadDeformedGeometry=true
      bRequireCurrentFrameData=true
      ComponentTags []
      FilteredBones [root]
      FilteredSockets []
      MeshUserParameter {parameter {name=None, typeDefHandle {registeredTypeIndex=1}, varData []}}
      SamplingRegions []
     }}
    Socket Sampling Mode=::NewEnumerator5
    Use Sampled Mesh Attribute as Ribbon ID=false
   ) {module=SkeletalMeshLocation001}
   &FN_Light_Attributes(
    Color=<rgba 1 0.6101 0.1884 10>
    Create Color Attribute=true
    Create Exponent Attribute=true
    Create Radius Attribute=true
    Exponent=5
    Radius=200
    Scale Color=&FloatFromCurve(
     FloatCurve=DataInterface() {propertyValues {Curve {keys [<at 0 0.8>,<at 0.1283 0.7497>,<at 0.3353 0.8961>,<at 0.4583 0.8473>,<at 0.5875 0.8636>,<at 0.7758 0.8473>,<at 0.9704 0.9368>,<at 1.0653 0.8>]}}}
     Scale Curve=5
    )
   )
  ]
  rendererClasses [
   &NiagaraSpriteRendererProperties
   &NiagaraComponentRendererProperties
   &NiagaraLightRendererProperties
  ]
  renderers [
   {rendererClass=&NiagaraSpriteRendererProperties}
   {rendererClass=&NiagaraComponentRendererProperties, rendererIndex=1}
   {rendererClass=&NiagaraLightRendererProperties, rendererIndex=2}
  ]
 }
 NE_Cherry_Embers {
  emitterUpdateScript [
   &EmitterState(Life Cycle Mode=::NewEnumerator1, Loop Duration=1, UseLoopDelay=false)
   &SpawnRate(SpawnRate=20, Use Spawn Probability=false)
  ]
  particleSpawnScript [
   &InitializeParticle(
    Lifetime=&RandomRangeFloat(Minimum=0.25, Maximum=0.7)
    Module.Write Color=true
    Module.Write Lifetime=true
    Module.Write Mass=true
    Module.Write Material Random=true
    Module.Write Position=false
    Module.Write Scale=false
    Module.Write SpriteRotation=false
    Module.Write SpriteSize=true
    Position=&SimulationPosition()
    Sprite Size=&Vector2DFromFloat(Value=&FloatFromCurve(
      FloatCurve=DataInterface() {propertyValues {Curve {keys [<at 0.0015 0.9809>,<at 0.1257 0.9847>,<at 0.2996 0.5>,<at 1 0.3>]}}}
      CurveIndex=&UniformRangedFloat(Minimum=500, Maximum=1000)
      Scale Curve=3.5
     ))
   )
   &DynamicMaterialParameters(
    Index 0 Param 1=&RandomRangeFloat(Minimum=-0.1, Maximum=0.1)
    Module.Param0WriteEnabled=true
    Module.Param1WriteEnabled=false
    Module.Param2WriteEnabled=false
    Module.Param3WriteEnabled=false
   )
   &SkeletalMeshLocation(
    Bone Sampling Mode=::NewEnumerator0
    Inherit Velocity=false
    MaskSpawnGroup=false
    Module.Override Randomness=false
    Sampled Position Offset=&RandomRangeVector(Minimum=<-2 -2 -2>, Maximum=<2 2 2>)
    Skeletal Mesh=$User.SkeletalMesh
   )
   &AddVelocityFromPoint(Module.Use Falloff Distance=false, Velocity Strength=30)
   &InheritVelocity(Module.Limit Speed=true, Velocity Limit=500, Velocity Scale=<0.9 0.9 0.9>)
  ]
  particleUpdateScript [
   &AddVelocity(Velocity=&MakeVector(X=0, Y=0, Z=&RandomRangeFloat(Minimum=9, Maximum=15))) {enabled=false}
   &ParticleState()
   &AccelerationForce(Acceleration=<0 0 100>)
   &Drag(
    Drag=&FloatFromCurve(FloatCurve=DataInterface() {propertyValues {Curve {keys [<at 0 0.5>,<at 0.5 0.5>,<at 1 1>]}}}, Scale Curve=9)
    Use Linear Drag=true
    Use Rotational Drag=false
   )
   /Niagara/Modules/Update/Forces/CurlNoiseForce(
    Noise Frequency=0.04
    Noise Strength=&FloatFromCurve(FloatCurve=DataInterface() {propertyValues {Curve {keys [<at 0 1>,<at 1 0>]}}}, Scale Curve=500)
    Pan Noise=true
    Pan Noise Field=<0 0 1>
   )
   &SolveForcesAndVelocity(Clamp Velocity=false, Limit Acceleration=false)
   &ScaleColor(
    Scale Alpha=&FloatFromCurve(FloatCurve=DataInterface() {propertyValues {Curve {keys [
         {arriveTangent=2.0811, interpMode=RCIM_Cubic, leaveTangent=2.0811, tangentMode=RCTM_User, time=0, value=0}
         <at 0.5 1 RCIM_Cubic>
         {arriveTangent=-2.9595, interpMode=RCIM_Cubic, leaveTangent=-2.9595, tangentMode=RCTM_User, time=1, value=0}
        ]}}})
    Scale RGB=&MakeVectorFromLinearColorRGB(Color=&ColorFromCurve(ColorCurve=DataInterface() {propertyValues {
        AlphaCurve {keys [<at 0 1 RCIM_Cubic>,<at 0.9975 0.5>]}
        BlueCurve {keys [<at 0 1 RCIM_Cubic>,<at 1 0.4809 RCIM_Cubic>]}
        GreenCurve {keys [<at 0 1 RCIM_Cubic>,<at 1 0.4809 RCIM_Cubic>]}
        RedCurve {keys [<at 0 1 RCIM_Cubic>,<at 1 1 RCIM_Cubic>]}
       }}))
    ScaleA=true
    ScaleRGB=true
   )
   &ScaleSpriteSizeBySpeed(
    Max Scale Factor=<2 5>
    Sample Scale Factor By Curve=true
    Scale Factor Curve=DataInterface() {propertyValues {Curve {keys [
        <at -0.0077 0.0765 RCIM_Cubic>
        {arriveTangent=1.2463, interpMode=RCIM_Cubic, leaveTangent=1.2463, tangentMode=RCTM_User, time=0.9985, value=0.9549}
       ]}}}
    Velocity Threshold=500
   )
   &ScaleColor(
    Scale RGB=&MakeVectorFromLinearColorRGB(Color=&ColorFromCurve(ColorCurve=DataInterface() {propertyValues {
        AlphaCurve {keys [<at 0 1 RCIM_Cubic>]}
        BlueCurve {keys [<at 0 12.2396 RCIM_Cubic>,<at 0.2499 0.8507>,<at 0.7114 0.1788 RCIM_Cubic>]}
        GreenCurve {keys [<at 0 35.8364 RCIM_Cubic>,<at 0.2499 4.9272>,<at 0.7114 0.2017 RCIM_Cubic>]}
        RedCurve {keys [<at 0 50 RCIM_Cubic>,<at 0.2499 10>,<at 0.7114 1 RCIM_Cubic>]}
       }}))
    ScaleA=true
    ScaleRGB=true
   ) {module=ScaleColor001}
   &ScaleSpriteSize(Scale Factor=&Vector2DFromFloat(Value=&FloatFromCurve(FloatCurve=DataInterface() {propertyValues {Curve {keys [<at 0 1 RCIM_Cubic>, {arriveTangent=-2.7368, interpMode=RCIM_Cubic, leaveTangent=-2.7368, tangentMode=RCTM_User, time=1, value=0}]}}})))
  ]
  rendererClasses [&NiagaraSpriteRendererProperties]
  renderers [{rendererClass=&NiagaraSpriteRendererProperties}]
 }
 NE_Cherry_Fists_Glow {
  emitterUpdateScript [
   &EmitterState(
    Life Cycle Mode=::NewEnumerator1
    Loop Behavior=::NewEnumerator1
    Loop Duration Mode=::NewEnumerator1
    Reset Age On Awaken=true
    Scalability Mode=::NewEnumerator1
    UseLoopDelay=false
   )
   &SpawnRate(SpawnRate=6, Use Spawn Probability=false) {enabled=false}
   &SpawnBurst_Instantaneous(Spawn Count=2, Use Spawn Probability=false)
  ]
  particleSpawnScript [
   &InitializeParticle(
    Color=<rgba 1 0.6101 0.1884 0.1>
    Module.Write Color=true
    Module.Write Lifetime=true
    Module.Write Mass=true
    Module.Write Material Random=true
    Module.Write Position=false
    Module.Write Scale=false
    Module.Write SpriteRotation=false
    Module.Write SpriteSize=true
    Position=&SimulationPosition()
    Sprite Size=<25 30>
   )
   &DynamicMaterialParameters(
    Index 0 Param 1=&RandomRangeFloat(Minimum=-0.1, Maximum=0.1)
    Module.Param0WriteEnabled=true
    Module.Param1WriteEnabled=false
    Module.Param2WriteEnabled=false
    Module.Param3WriteEnabled=false
   )
   &FN_SetSpawnGroup(Number of Spawn Groups=2)
  ]
  particleUpdateScript [
   &SkeletalMeshLocation(
    Bone / Socket Index=$Particles.SpawnGroup
    Bone Sampling Mode=::NewEnumerator4
    Inherit Velocity=false
    MaskSpawnGroup=false
    Sampled Position Offset=&MakeVector(
     X=&FN_PickFloatFromSpawnGroupCutoff(Over Cutoff Float:NiagaraFloat=7, Under Cutoff Float:NiagaraFloat=-7, SpawnGroup:NiagaraInt32=$Particles.SpawnGroup, Cutoff:NiagaraInt32=1)
     Y=0
     Z=0
    )
    Sampled Position Offset Coordinate Space:ENiagaraMeshTransforms=::NewEnumerator3
    Skeletal Mesh=DataInterface() {propertyValues {
      DefaultMesh=null
      ExcludeBoneName=root
      OptionalComponentClass=null
      PreviewMesh=&F_MED_Cherry
      SkinningMode=SkinOnTheFly
      SoftSourceActor=null
      SourceMode=Default
      UvSetIndex=0
      WholeMeshLOD=-1
      bExcludeBone=false
      bReadDeformedGeometry=true
      bRequireCurrentFrameData=true
      ComponentTags []
      FilteredBones [hand_r,hand_l]
      FilteredSockets []
      MeshUserParameter {parameter {name=None, typeDefHandle {registeredTypeIndex=1}, varData []}}
      SamplingRegions []
     }}
   )
   &AddVelocity(Velocity=&MakeVector(X=0, Y=0, Z=&RandomRangeFloat(Minimum=9, Maximum=15))) {enabled=false}
   &ParticleState(Kill Particles When Lifetime Has Elapsed=false, Let Infinitely Lived Particles Die When Emitter Deactivates=true)
   &SolveForcesAndVelocity(Clamp Velocity=false, Limit Acceleration=false) {enabled=false}
   &ScaleColor(
    Scale Alpha=&FloatFromCurve(FloatCurve=DataInterface() {propertyValues {Curve {keys [
         {arriveTangent=2.0811, interpMode=RCIM_Cubic, leaveTangent=2.0811, tangentMode=RCTM_User, time=0, value=0}
         <at 0.5 1 RCIM_Cubic>
         {arriveTangent=-2.9595, interpMode=RCIM_Cubic, leaveTangent=-2.9595, tangentMode=RCTM_User, time=1, value=0}
        ]}}})
    Scale RGB=&MakeVectorFromLinearColorRGB(Color=&ColorFromCurve(ColorCurve=DataInterface() {propertyValues {
        AlphaCurve {keys [<at 0 1 RCIM_Cubic>]}
        BlueCurve {keys [<at 0 1 RCIM_Cubic>,<at 1 0.8212 RCIM_Cubic>]}
        GreenCurve {keys [<at 0 1 RCIM_Cubic>,<at 1 0.8212 RCIM_Cubic>]}
        RedCurve {keys [<at 0 1 RCIM_Cubic>,<at 1 1 RCIM_Cubic>]}
       }}))
    ScaleA=true
    ScaleRGB=false
   ) {enabled=false}
  ]
  rendererClasses [&NiagaraSpriteRendererProperties]
  renderers [{rendererClass=&NiagaraSpriteRendererProperties}]
 }
 NE_Cherry_Light {
  bEnabled=false
  emitterUpdateScript [
   &EmitterState(
    Life Cycle Mode=::NewEnumerator1
    Loop Behavior=::NewEnumerator1
    Loop Duration Mode=::NewEnumerator1
    Reset Age On Awaken=true
    Scalability Mode=::NewEnumerator1
    UseLoopDelay=false
   )
   &SpawnRate(SpawnRate=20, Use Spawn Probability=false) {enabled=false}
   &SpawnBurst_Instantaneous(Spawn Count=1, Use Spawn Probability=false)
  ]
  particleSpawnScript [&InitializeParticle(
    Module.Write Color=true
    Module.Write Lifetime=true
    Module.Write Mass=true
    Module.Write Material Random=true
    Module.Write Position=false
    Module.Write Scale=false
    Module.Write SpriteRotation=false
    Module.Write SpriteSize=true
    Position=&SimulationPosition()
    Sprite Size=<5 5>
   )]
  particleUpdateScript [
   &ParticleState(
    Kill Particles When Lifetime Has Elapsed=false
    Let Infinitely Lived Particles Die When Emitter Deactivates=true
    Loop Particles Lifetime=true
   )
   &SkeletalMeshLocation(
    Apply Sampled Orientation to Mesh Particles=true
    Inherit Velocity=false
    Interpolate Mesh Sampling=false
    MaskSpawnGroup=false
    Mesh Sampling Type=::NewEnumerator5
    Sampled Position Offset=<0 0 85>
    Sampled Position Offset Coordinate Space:ENiagaraMeshTransforms=::NewEnumerator3
    Skeletal Mesh=DataInterface() {propertyValues {
      DefaultMesh=null
      ExcludeBoneName=dyn_ponytail_06
      OptionalComponentClass=null
      PreviewMesh=&F_MED_Cherry
      SkinningMode=SkinOnTheFly
      SoftSourceActor=null
      SourceMode=Default
      UvSetIndex=0
      WholeMeshLOD=-1
      bExcludeBone=false
      bReadDeformedGeometry=true
      bRequireCurrentFrameData=true
      ComponentTags []
      FilteredBones [root]
      FilteredSockets []
      MeshUserParameter {parameter {name=None, typeDefHandle {registeredTypeIndex=1}, varData []}}
      SamplingRegions []
     }}
    Socket Sampling Mode=::NewEnumerator5
   ) {module=SkeletalMeshLocation001}
   &SolveForcesAndVelocity(Clamp Velocity=false, Limit Acceleration=false)
   &FN_Light_Attributes(
    Color=<rgba 1 0.6101 0.1884 10>
    Create Color Attribute=true
    Create Exponent Attribute=true
    Create Radius Attribute=true
    Exponent=5
    Radius=200
    Scale Color=&FloatFromCurve(
     FloatCurve=DataInterface() {propertyValues {Curve {keys [<at 0 0.8>,<at 0.1283 0.7497>,<at 0.3353 0.8961>,<at 0.4583 0.8473>,<at 0.5875 0.8636>,<at 0.7758 0.8473>,<at 0.9704 0.9368>,<at 1.0653 0.8>]}}}
     Scale Curve=5
    )
   )
  ]
  rendererClasses [&NiagaraSpriteRendererProperties, &NiagaraLightRendererProperties]
  renderers [
   {rendererClass=&NiagaraSpriteRendererProperties}
   {rendererClass=&NiagaraLightRendererProperties, rendererIndex=1}
  ]
 }
 NE_Cherry_VerticalRays {
  emitterUpdateScript [
   &EmitterState(Life Cycle Mode=::NewEnumerator1, Loop Duration=1, UseLoopDelay=false)
   &SpawnRate(SpawnRate=16, Use Spawn Probability=false)
  ]
  particleSpawnScript [
   &FN_SetSpawnGroup(Number of Spawn Groups=3)
   &InitializeParticle(
    Color=<rgba 1 0.5855 0.1372 0.5>
    Lifetime=3
    Module.Write Color=true
    Module.Write Lifetime=true
    Module.Write Mass=true
    Module.Write Material Random=true
    Module.Write Position=false
    Module.Write Scale=false
    Module.Write SpriteRotation=false
    Module.Write SpriteSize=true
    Position=&SimulationPosition()
    Sprite Size=&MaskVector2dBySpawnGroup(
     Masked Vector2D=&RandomRangeVector2D(Minimum=<1 30>, Maximum=<15 40>)
     Passthrough Vector2D=&RandomRangeVector2D(Minimum=<1 25>, Maximum=<5 35>)
     Spawn Group Mask=2
    )
   )
   &DynamicMaterialParameters(
    Index 0 Param 1=&RandomRangeFloat(Minimum=-0.1, Maximum=0.1)
    Module.Param0WriteEnabled=true
    Module.Param1WriteEnabled=false
    Module.Param2WriteEnabled=false
    Module.Param3WriteEnabled=false
   )
  ]
  particleUpdateScript [
   &SkeletalMeshLocation(
    Bone / Socket Index=$Particles.SpawnGroup
    Bone Sampling Mode=::NewEnumerator4
    Inherit Velocity=false
    MaskSpawnGroup=false
    Sampled Position Offset=&MakeVector(
     X=&FN_PickFloatFromSpawnGroupCutoff(
      Over Cutoff Float:NiagaraFloat=&RandomRangeFloat(Minimum=0, Maximum=0)
      Under Cutoff Float:NiagaraFloat=&FN_PickFloatFromSpawnGroupCutoff(
       Over Cutoff Float:NiagaraFloat=&RandomRangeFloat(Minimum=7, Maximum=14)
       Under Cutoff Float:NiagaraFloat=&RandomRangeFloat(Minimum=-7, Maximum=-14)
       SpawnGroup:NiagaraInt32=$Particles.SpawnGroup
       Cutoff:NiagaraInt32=1
      )
      SpawnGroup:NiagaraInt32=$Particles.SpawnGroup
      Cutoff:NiagaraInt32=2
     )
     Y=&FN_PickFloatFromSpawnGroupCutoff(
      Over Cutoff Float:NiagaraFloat=&RandomRangeFloat(Minimum=-3, Maximum=6)
      Under Cutoff Float:NiagaraFloat=&FN_PickFloatFromSpawnGroupCutoff(
       Over Cutoff Float:NiagaraFloat=&RandomRangeFloat(Minimum=-3, Maximum=3)
       Under Cutoff Float:NiagaraFloat=&RandomRangeFloat(Minimum=-3, Maximum=3)
       SpawnGroup:NiagaraInt32=$Particles.SpawnGroup
       Cutoff:NiagaraInt32=1
      )
      SpawnGroup:NiagaraInt32=$Particles.SpawnGroup
      Cutoff:NiagaraInt32=2
     )
     Z=&FN_PickFloatFromSpawnGroupCutoff(
      Over Cutoff Float:NiagaraFloat=&RandomRangeFloat(Minimum=-6, Maximum=6)
      Under Cutoff Float:NiagaraFloat=&FN_PickFloatFromSpawnGroupCutoff(
       Over Cutoff Float:NiagaraFloat=&RandomRangeFloat(Minimum=-8, Maximum=-5)
       Under Cutoff Float:NiagaraFloat=&RandomRangeFloat(Minimum=5, Maximum=8)
       SpawnGroup:NiagaraInt32=$Particles.SpawnGroup
       Cutoff:NiagaraInt32=1
      )
      SpawnGroup:NiagaraInt32=$Particles.SpawnGroup
      Cutoff:NiagaraInt32=2
     )
    )
    Sampled Position Offset Coordinate Space:ENiagaraMeshTransforms=::NewEnumerator3
    Skeletal Mesh=DataInterface() {propertyValues {
      DefaultMesh=null
      ExcludeBoneName=root
      OptionalComponentClass=null
      PreviewMesh=&F_MED_Cherry
      SkinningMode=SkinOnTheFly
      SoftSourceActor=null
      SourceMode=Default
      UvSetIndex=0
      WholeMeshLOD=-1
      bExcludeBone=false
      bReadDeformedGeometry=true
      bRequireCurrentFrameData=true
      ComponentTags []
      FilteredBones [clavicle_r,clavicle_l,head]
      FilteredSockets []
      MeshUserParameter {parameter {name=None, typeDefHandle {registeredTypeIndex=1}, varData []}}
      SamplingRegions []
     }}
   )
   &AddVelocity(Velocity=&MakeVector(X=0, Y=0, Z=&RandomRangeFloat(Minimum=9, Maximum=15))) {enabled=false}
   &ParticleState()
   &SolveForcesAndVelocity(Clamp Velocity=false, Limit Acceleration=false)
   &ScaleColor(
    Scale Alpha=&FloatFromCurve(FloatCurve=DataInterface() {propertyValues {Curve {keys [
         {arriveTangent=5.1832, interpMode=RCIM_Cubic, leaveTangent=5.1832, tangentMode=RCTM_User, time=0, value=0}
         {arriveTangent=2.2426, interpMode=RCIM_Cubic, leaveTangent=2.2426, tangentMode=RCTM_User, time=0.133, value=0.764}
         {arriveTangent=-0.0043, interpMode=RCIM_Cubic, leaveTangent=-0.0043, time=0.5, value=1}
         {arriveTangent=-1.5406, interpMode=RCIM_Cubic, leaveTangent=-1.5406, tangentMode=RCTM_User, time=0.8415, value=0.7609}
         {arriveTangent=-3.9306, interpMode=RCIM_Cubic, leaveTangent=-3.9306, tangentMode=RCTM_User, time=1, value=0}
        ]}}})
    Scale RGB=&MakeVectorFromLinearColorRGB(Color=&ColorFromCurve(ColorCurve=DataInterface() {propertyValues {
        AlphaCurve {keys [<at 0 1 RCIM_Cubic>]}
        BlueCurve {keys [<at 0 1 RCIM_Cubic>,<at 1 0.8559 RCIM_Cubic>]}
        GreenCurve {keys [<at 0 1 RCIM_Cubic>,<at 1 0.8559 RCIM_Cubic>]}
        RedCurve {keys [<at 0 1 RCIM_Cubic>,<at 1 1 RCIM_Cubic>]}
       }}))
    ScaleA=true
    ScaleRGB=false
   )
   &ScaleSpriteSize(Scale Factor=&MakeVector2D(X=1, Y=&FloatFromCurve(FloatCurve=DataInterface() {propertyValues {Curve {keys [<at 0 0>,<at 1 1>]}}})))
   &Color(Color=&MaskLinearColorBySpawnGroup(
     Masked Linear Color=<rgba 1 0 0.2339 1>
     Passthrough Linear Color=&MaskLinearColorBySpawnGroup(Masked Linear Color=<rgba 0.1072 1 0 1>, Passthrough Linear Color=<rgba 0 0.449 1 1>, Spawn Group Mask=1)
     Spawn Group Mask=2
    )) {enabled=false}
  ]
  rendererClasses [&NiagaraSpriteRendererProperties]
  renderers [{rendererClass=&NiagaraSpriteRendererProperties}]
 }
}
summary {
 systemName=NS_Cherry_Charged_Body
 emitters [
  {emitterName=Cherry_Arms_BoneFlow_Fire, rendererClasses [&NiagaraSpriteRendererProperties]}
  {emitterName=NE_Cherry_Embers, rendererClasses [&NiagaraSpriteRendererProperties]}
  {emitterName=NE_Cherry_VerticalRays, rendererClasses [&NiagaraSpriteRendererProperties]}
  {emitterName=Cherry_Arms_BoneFlow_VelocityAligned, rendererClasses [&NiagaraSpriteRendererProperties]}
  {
   emitterName=Lights
   rendererClasses [
    &NiagaraSpriteRendererProperties
    &NiagaraComponentRendererProperties
    &NiagaraLightRendererProperties
   ]
  }
  {
   bEnabled=false
   emitterName=NE_Cherry_Light
   rendererClasses [&NiagaraSpriteRendererProperties, &NiagaraLightRendererProperties]
  }
  {
   emitterName=Cherry_Arms_Lightning001
   rendererClasses [&NiagaraSpriteRendererProperties, &NiagaraMeshRendererProperties]
  }
  {emitterName=NE_Cherry_Fists_Glow, rendererClasses [&NiagaraSpriteRendererProperties]}
 ]
 userVariables [
  User.Color_Tint {type=LinearColor, description="", defaultValue {struct=LinearColor, value=<rgba 1 1 1 1>}}
  User.EmissiveIntensity {type=NiagaraFloat, defaultValue=0, description=""}
  User.IsSkydiving {type=NiagaraBool}
  User.Light_falloff {type=NiagaraFloat, defaultValue=0, description=""}
  User.RibbonIDOffset {type=NiagaraInt32}
  User.SkeletalMesh {
   type=NiagaraDataInterfaceSkeletalMesh
   defaultValue {
    struct=NiagaraExt_VariableValue_DataInterface
    value {dataInterface=&self:NiagaraDataInterfaceSkeletalMesh_0}
   }
  }
 ]
}
