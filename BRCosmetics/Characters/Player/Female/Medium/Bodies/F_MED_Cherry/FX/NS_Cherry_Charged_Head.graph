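# Niagara system for the Cherry cosmetic's charged head effect, adding warm gold glowing eyes.
# EyeGlow pins two sprite-and-light particles to User.SkeletalMesh, indexed by spawn group.
# EyeTrail keeps one oriented mesh particle offset from the head bone.
# Disabled hair emitters hold head-bone arc meshes and FX_Hair socket ribbons, shaped by Hermite splines whose ends lag against owner velocity.
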NiagaraSystem /BRCosmetics/Characters/Player/Female/Medium/Bodies/F_MED_Cherry/FX/NS_Cherry_Charged_Head
dependencies {
 usedDataInterfaces [&NiagaraDataInterfaceCurve, &NiagaraDataInterfaceColorCurve]
 usedDynamicInputs []
 usedModules [&SystemState]
}
emitters {
 EyeGlow {
  emitterUpdateScript [
   &EmitterState(
    Life Cycle Mode=::NewEnumerator1
    Loop Behavior=::NewEnumerator1
    Loop Duration Mode=::NewEnumerator1
    Reset Age On Awaken=true
    Scalability Mode=::NewEnumerator1
    UseLoopDelay=false
   )
   &SpawnBurst_Instantaneous(Spawn Count=2, Use Spawn Probability=false)
  ]
  particleSpawnScript [
   &FN_SetSpawnGroup(Number of Spawn Groups=2)
   /Niagara/Modules/Spawn/Initialization/InitializeParticle(
    Color=<rgba 3 1.3429 0.0677 1>
    Mesh Scale=<1.3 1.3 1.3>
    Module.Write Color=true
    Module.Write Lifetime=true
    Module.Write Mass=true
    Module.Write Material Random=true
    Module.Write Position=false
    Module.Write Scale=false
    Module.Write SpriteRotation=false
    Module.Write SpriteSize=true
    Position=&SimulationPosition()
    Sprite Size=&Vector2DFromFloat(Value=15)
   )
  ]
  particleUpdateScript [
   &SkeletalMeshLocation(
    Bone / Socket Index=$Particles.SpawnGroup
    Bone Sampling Mode=::NewEnumerator4
    Inherit Velocity=false
    MaskSpawnGroup=false
    Skeletal Mesh=$User.SkeletalMesh
   )
   &ParticleState(Kill Particles When Lifetime Has Elapsed=false, Let Infinitely Lived Particles Die When Emitter Deactivates=true)
   &DynamicMaterialParameters(
    Index 1 Param 1=4
    Module.Param0WriteEnabled=true
    Module.Param0WriteEnabled1=true
    Module.Param1WriteEnabled=true
    Module.Param1WriteEnabled1=true
    Module.Param2WriteEnabled=true
    Module.Param2WriteEnabled1=true
    Module.Param3WriteEnabled=true
    Module.Param3WriteEnabled1=true
    Write Parameter Index 1=true
   )
   &CameraOffset(Camera Offset Amount=2.5)
  ]
  rendererClasses [&NiagaraSpriteRendererProperties, &NiagaraLightRendererProperties]
  renderers [
   {rendererClass=&NiagaraSpriteRendererProperties}
   {rendererClass=&NiagaraLightRendererProperties, rendererIndex=1}
  ]
 }
 EyeTrail {
  emitterUpdateScript [
   &EmitterState(
    Life Cycle Mode=::NewEnumerator1
    Loop Behavior=::NewEnumerator1
    Loop Duration Mode=::NewEnumerator1
    Reset Age On Awaken=true
    Scalability Mode=::NewEnumerator1
    UseLoopDelay=false
   )
   &SpawnBurst_Instantaneous(Spawn Count=1, Use Spawn Probability=false)
  ]
  particleSpawnScript [
   &InitializeParticle(
    Color=<rgba 7 4.6372 1.2517 1>
    Mesh Scale=<0.7 0.7 0.7>
    Module.Write Color=true
    Module.Write Lifetime=true
    Module.Write Mass=true
    Module.Write Material Random=true
    Module.Write Position=false
    Module.Write Scale=true
    Module.Write SpriteRotation=false
    Module.Write SpriteSize=false
    Position=&SimulationPosition()
   )
   /Game/Effects/Niagara/Modules/Orientation/FN_ReorientMeshV2(Forward Vector:Vector3f=<0 1 0>, Mesh Reference Axis:Vector3f=<0 1 0>, Spin=-0.13, Spin Mesh=true) {enabled=false}
  ]
  particleUpdateScript [
   &SkeletalMeshLocation(
    Apply Sampled Orientation to Mesh Particles=true
    Bone Sampling Mode=::NewEnumerator4
    Inherit Velocity=false
    MaskSpawnGroup=false
    Sampled Position Offset=<3.5 7 0>
    Sampled Position Offset Coordinate Space:ENiagaraMeshTransforms=::NewEnumerator3
    Skeletal Mesh=DataInterface() {propertyValues {
      DefaultMesh=null
      ExcludeBoneName=root
      OptionalComponentClass=null
      PreviewMesh=/BRCosmetics/Characters/Player/Female/Medium/Heads/F_MED_Cherry_Head/Meshes/Female_Medium_Cherry_Head
      SkinningMode=SkinOnTheFly
      SoftSourceActor=null
      SourceMode=Default
      UvSetIndex=0
      WholeMeshLOD=-1
      bExcludeBone=false
      bReadDeformedGeometry=true
      bRequireCurrentFrameData=true
      ComponentTags []
      FilteredBones [head]
      FilteredSockets []
      MeshUserParameter {parameter {name=None, typeDefHandle {registeredTypeIndex=1}, varData []}}
      SamplingRegions []
     }}
    Write Sampled Mesh Attrributes to Particle Attributes=true
   )
   &ParticleState(Kill Particles When Lifetime Has Elapsed=false, Let Infinitely Lived Particles Die When Emitter Deactivates=true)
   &CameraOffset(Camera Offset Amount=2)
  ]
  rendererClasses [&NiagaraMeshRendererProperties]
  renderers [{rendererClass=&NiagaraMeshRendererProperties, rendererIndex=0}]
 }
 Hair_Arcs_L {
  bEnabled=false
  emitterUpdateScript [
   &EmitterState()
   &SpawnRate(SpawnRate=2, Use Spawn Probability=false)
  ]
  particleSpawnScript [
   &InitializeParticle(
    Color=<rgba 7 4.6372 1.2517 1>
    Lifetime=&RandomRangeFloat(Minimum=1.3, Maximum=1.6)
    Mesh Scale=&Multiply_VectorByFloat(Vector=<1 -1 1>, Float=&RandomRangeFloat(Minimum=0.5, Maximum=0.7))
    Module.Write Color=false
    Module.Write Lifetime=true
    Module.Write Mass=true
    Module.Write Material Random=true
    Module.Write Position=false
    Module.Write Scale=true
    Module.Write SpriteRotation=false
    Module.Write SpriteSize=false
    Position=&SimulationPosition()
   )
   &FN_ReorientMeshV2(Forward Vector:Vector3f=<0 1 0>, Mesh Reference Axis:Vector3f=<0 1 0>, Spin=-0.13, Spin Mesh=true)
   &InitialMeshOrientation(
    Rotation=&MakeVector(X=&RandomRangeFloat(Minimum=0.2, Maximum=0.6), Y=0, Z=0)
    Use Orientation Vector=true
    Use Rotation Vector=true
   )
  ]
  particleUpdateScript [
   &ParticleState()
   &SkeletalMeshLocation(
    Apply Sampled Orientation to Mesh Particles=true
    Bone Sampling Mode=::NewEnumerator4
    Inherit Velocity=false
    MaskSpawnGroup=false
    Sampled Position Offset=&MakeVector(
     X=&FloatFromCurve(FloatCurve=DataInterface() {propertyValues {Curve {keys [<at 0 -0.85>,<at 1 1>]}}}, Scale Curve=-15)
     Y=0
     Z=-3
    )
    Sampled Position Offset Coordinate Space:ENiagaraMeshTransforms=::NewEnumerator3
    Skeletal Mesh=DataInterface() {propertyValues {
      DefaultMesh=null
      ExcludeBoneName=root
      OptionalComponentClass=null
      PreviewMesh=&Female_Medium_Cherry_Head
      SkinningMode=SkinOnTheFly
      SoftSourceActor=null
      SourceMode=Default
      UvSetIndex=0
      WholeMeshLOD=-1
      bExcludeBone=false
      bReadDeformedGeometry=true
      bRequireCurrentFrameData=true
      ComponentTags []
      FilteredBones [head]
      FilteredSockets []
      MeshUserParameter {parameter {name=None, typeDefHandle {registeredTypeIndex=1}, varData []}}
      SamplingRegions []
     }}
    Write Sampled Mesh Attrributes to Particle Attributes=true
   )
   &SolveForcesAndVelocity(Clamp Velocity=false, Limit Acceleration=false)
   &ScaleColor(
    Scale Alpha=&FloatFromCurve(FloatCurve=DataInterface() {propertyValues {Curve {keys [<at 0 0 RCIM_Cubic>,<at 0.5 1 RCIM_Cubic>,<at 1 0 RCIM_Cubic>]}}})
    ScaleA=true
    ScaleRGB=true
   ) {enabled=false}
   &ScaleMeshSize(Scale Factor=&VectorFromFloat(Value=&FloatFromCurve(FloatCurve=DataInterface() {propertyValues {Curve {keys [
          {arriveTangent=4.4642, interpMode=RCIM_Cubic, leaveTangent=4.4642, tangentMode=RCTM_User, time=0, value=0}
          <at 0.441 1.0367 RCIM_Cubic>
          {arriveTangent=-2.5534, interpMode=RCIM_Cubic, leaveTangent=-2.5534, tangentMode=RCTM_User, time=1, value=0}
         ]}}}))) {module=ScaleMeshSize001}
   &DynamicMaterialParameters(
    Index 0 Param 1=2
    Index 0 Param 2=0.1
    Index 0 Param 3=7
    Index 0 Param 4=7
    Module.Param0WriteEnabled=true
    Module.Param1WriteEnabled=true
    Module.Param2WriteEnabled=true
    Module.Param3WriteEnabled=true
   )
   &ScaleMeshSize(Scale Factor=<2 2 2>) {enabled=false}
  ]
  rendererClasses [&NiagaraMeshRendererProperties]
  renderers [{rendererClass=&NiagaraMeshRendererProperties, rendererIndex=0}]
 }
 Hair_Arcs_R {
  bEnabled=false
  emitterUpdateScript [
   &EmitterState()
   &SpawnRate(SpawnRate=2, Use Spawn Probability=false)
  ]
  particleSpawnScript [
   &InitializeParticle(
    Color=<rgba 7 4.6372 1.2517 1>
    Lifetime=&RandomRangeFloat(Minimum=1.3, Maximum=1.6)
    Mesh Scale=&VectorFromFloat(Value=&RandomRangeFloat(Minimum=0.5, Maximum=0.7))
    Module.Write Color=false
    Module.Write Lifetime=true
    Module.Write Mass=true
    Module.Write Material Random=true
    Module.Write Position=false
    Module.Write Scale=true
    Module.Write SpriteRotation=false
    Module.Write SpriteSize=false
    Position=&SimulationPosition()
   )
   &FN_ReorientMeshV2(Forward Vector:Vector3f=<0 1 0>, Mesh Reference Axis:Vector3f=<0 1 0>, Spin=-0.13, Spin Mesh=true)
   &InitialMeshOrientation(
    Rotation=&MakeVector(X=&RandomRangeFloat(Minimum=-0.1, Maximum=0.4), Y=0, Z=0)
    Use Orientation Vector=true
    Use Rotation Vector=true
   )
  ]
  particleUpdateScript [
   &ParticleState()
   &SkeletalMeshLocation(
    Apply Sampled Orientation to Mesh Particles=true
    Bone Sampling Mode=::NewEnumerator4
    Inherit Velocity=false
    MaskSpawnGroup=false
    Sampled Position Offset=&MakeVector(
     X=&FloatFromCurve(FloatCurve=DataInterface() {propertyValues {Curve {keys [<at 0 -0.85>,<at 1 1>]}}}, Scale Curve=-15)
     Y=0
     Z=3
    )
    Sampled Position Offset Coordinate Space:ENiagaraMeshTransforms=::NewEnumerator3
    Skeletal Mesh=DataInterface() {propertyValues {
      DefaultMesh=null
      ExcludeBoneName=root
      OptionalComponentClass=null
      PreviewMesh=&Female_Medium_Cherry_Head
      SkinningMode=SkinOnTheFly
      SoftSourceActor=null
      SourceMode=Default
      UvSetIndex=0
      WholeMeshLOD=-1
      bExcludeBone=false
      bReadDeformedGeometry=true
      bRequireCurrentFrameData=true
      ComponentTags []
      FilteredBones [head]
      FilteredSockets []
      MeshUserParameter {parameter {name=None, typeDefHandle {registeredTypeIndex=1}, varData []}}
      SamplingRegions []
     }}
    Write Sampled Mesh Attrributes to Particle Attributes=true
   )
   &SolveForcesAndVelocity(Clamp Velocity=false, Limit Acceleration=false)
   &ScaleColor(
    Scale Alpha=&FloatFromCurve(FloatCurve=DataInterface() {propertyValues {Curve {keys [<at 0 0 RCIM_Cubic>,<at 0.5 1 RCIM_Cubic>,<at 1 0 RCIM_Cubic>]}}})
    ScaleA=true
    ScaleRGB=true
   ) {enabled=false}
   &ScaleMeshSize(Scale Factor=&VectorFromFloat(Value=&FloatFromCurve(FloatCurve=DataInterface() {propertyValues {Curve {keys [
          {arriveTangent=4.4642, interpMode=RCIM_Cubic, leaveTangent=4.4642, tangentMode=RCTM_User, time=0, value=0}
          <at 0.441 1.0367 RCIM_Cubic>
          {arriveTangent=-2.5534, interpMode=RCIM_Cubic, leaveTangent=-2.5534, tangentMode=RCTM_User, time=1, value=0}
         ]}}}))) {module=ScaleMeshSize001}
   &DynamicMaterialParameters(
    Index 0 Param 1=2
    Index 0 Param 2=0.1
    Index 0 Param 3=7
    Index 0 Param 4=7
    Module.Param0WriteEnabled=true
    Module.Param1WriteEnabled=true
    Module.Param2WriteEnabled=true
    Module.Param3WriteEnabled=true
   )
   &ScaleMeshSize(Scale Factor=<2 2 2>) {enabled=false}
  ]
  rendererClasses [&NiagaraMeshRendererProperties]
  renderers [{rendererClass=&NiagaraMeshRendererProperties, rendererIndex=0}]
 }
 Hair_L {
  bEnabled=false
  emitterSpawnScript [/Game/Effects/Niagara/Modules/Partition/FN_PartitionAttributes(Number of Bone Sets=1, Particles Per Partition=10)]
  emitterUpdateScript [
   &EmitterState(Life Cycle Mode=::NewEnumerator1, Loop Behavior=::NewEnumerator1, Loop Duration Mode=::NewEnumerator1, UseLoopDelay=false)
   &SpawnBurst_Instantaneous(Spawn Count=$Emitter.FN_PartitionAttributes.Spawn Total, Use Spawn Probability=false)
  ]
  particleSpawnScript [
   &InitializeParticle(
    Lifetime=5
    Module.Write Color=true
    Module.Write Lifetime=true
    Module.Write Mass=true
    Module.Write Material Random=true
    Module.Write Position=true
    Module.Write Scale=false
    Module.Write SpriteRotation=false
    Module.Write SpriteSize=true
    Sprite Size=<5 5>
   )
   /Game/Effects/Niagara/Modules/Partition/FN_Particle_partition_v5/FN_Particle_Partition_v5(Enable Ribbon ID=true)
   &InitializeRibbon(
    Module.Write Color=true
    Module.Write Lifetime=true
    Module.Write Mass=true
    Module.Write Material Random=true
    Module.Write Position=false
    Module.Write RibbonTwist=false
    Module.Write RibbonWidth=false
    Ribbon Width=&FloatFromCurve(
     FloatCurve=DataInterface() {propertyValues {Curve {keys [
         {interpMode=RCIM_Cubic, leaveTangent=-1.6682, tangentMode=RCTM_Break, time=0, value=1}
         {arriveTangent=-0.3052, interpMode=RCIM_Cubic, leaveTangent=-0.3052, tangentMode=RCTM_Break, time=0.2774, value=0.441}
         {arriveTangent=-0.3335, interpMode=RCIM_Cubic, leaveTangent=-0.3335, tangentMode=RCTM_Break, time=0.6089, value=0.7531}
         {interpMode=RCIM_Cubic, tangentMode=RCTM_Break, time=1.001, value=0.053}
        ]}}}
     CurveIndex=$Particles.Partition.NormalizedID
     Scale Curve=7
    )
   )
   /SetVariables(Particles.RibbonFacing=<1 0 1>) {enabled=false, isSetParameters=true, module=SetVariables_#1}
  ]
  particleUpdateScript [
   &ParticleState(Kill Particles When Lifetime Has Elapsed=false, Let Infinitely Lived Particles Die When Emitter Deactivates=true)
   &SkeletalMeshLocation(
    Inherit Velocity=false
    MaskSpawnGroup=false
    Mesh Sampling Type=::NewEnumerator5
    Skeletal Mesh=DataInterface() {propertyValues {
      DefaultMesh=null
      ExcludeBoneName=root
      OptionalComponentClass=null
      PreviewMesh=&Female_Medium_Cherry_Head
      SkinningMode=SkinOnTheFly
      SoftSourceActor=null
      SourceMode=Default
      UvSetIndex=0
      WholeMeshLOD=-1
      bExcludeBone=false
      bReadDeformedGeometry=true
      bRequireCurrentFrameData=true
      ComponentTags []
      FilteredBones []
      FilteredSockets [FX_Hair]
      MeshUserParameter {parameter {name=None, typeDefHandle {registeredTypeIndex=1}, varData []}}
      SamplingRegions []
     }}
    Socket Sampling Mode=::NewEnumerator5
   )
   /Game/Effects/Niagara/Modules/Curves/FN_HermiteCubicSpline_v1/FN_HermiteCubicSpline(
    Direction End:Vector3f=<-40 0 0>
    Direction Start:Vector3f=<40 25 0>
    Position Blend:NiagaraFloat=$Particles.Initial.Partition.NormalizedID
    Position End:Vector3f=&Add_Vector(
     A=$Output.SkeletalMeshLocation.SampledPosition
     B=&Add_Vector(
      A=<20 -10 -15>
      B=&Multiply_Vector(
       A=&TransformVector(Vector=$Engine.Owner.Velocity, Source Space=::World, Destination Space=::Local)
       B=&Multiply_Vector(A=<-1 -1 -1>, B=/Niagara/DynamicInputs/Math/Clamp_Vector(Vector:Vector3f=<0.02 0.05 0.015>, Min:Vector3f=<0 0 0>, Max:Vector3f=<1 1 1>))
      )
     )
    )
    Position Start:Vector3f=$Output.SkeletalMeshLocation.SampledPosition
    Write Particles Position=true
   )
   /Game/Effects/Niagara/Modules/Materials/FN_RibbonMaterial/FN_RibbonMaterial_v2(
    Erosion Pan Speed U=1
    Erossion Scale U=2
    Macro Pan Speed=0.5
    Micro Pan Speed=0.2
    Opacity Erosion=0.2
    Opacity Pan U Speed=0.1
    Opacity Power=0.6
    Opacity Scale U=0.2
   ) {enabled=false}
   &FN_RibbonMaterial(
    Color=&ColorFromCurve(
     ColorCurve=DataInterface() {propertyValues {
       AlphaCurve {keys [<at 0.6757 1 RCIM_Cubic>,<at 1.0002 0>]}
       BlueCurve {keys [<at 0.0118 1.3802 RCIM_Cubic>,<at 0.9761 0.026 RCIM_Cubic>]}
       GreenCurve {keys [<at 0.0118 11.4764 RCIM_Cubic>,<at 0.9761 0.5163 RCIM_Cubic>]}
       RedCurve {keys [<at 0.0118 15 RCIM_Cubic>,<at 0.9761 1 RCIM_Cubic>]}
      }}
     CurveIndex=$Particles.Partition.NormalizedID
    )
    Macro Scale U=0
    Micro Scale U=0
    Opacity Pan U Speed=0.5
    Scale Alpha=0.25
   )
   &ScaleRibbonWidth(Ribbon Width Scale=4)
  ]
  rendererClasses [&NiagaraRibbonRendererProperties]
  renderers [{rendererClass=&NiagaraRibbonRendererProperties}]
 }
 Hair_R {
  bEnabled=false
  emitterSpawnScript [&FN_PartitionAttributes(Number of Bone Sets=1, Particles Per Partition=10)]
  emitterUpdateScript [
   &EmitterState(Life Cycle Mode=::NewEnumerator1, Loop Behavior=::NewEnumerator1, Loop Duration Mode=::NewEnumerator1, UseLoopDelay=false)
   &SpawnBurst_Instantaneous(Spawn Count=$Emitter.FN_PartitionAttributes.Spawn Total, Use Spawn Probability=false)
  ]
  particleSpawnScript [
   &InitializeParticle(
    Lifetime=5
    Module.Write Color=true
    Module.Write Lifetime=true
    Module.Write Mass=true
    Module.Write Material Random=true
    Module.Write Position=true
    Module.Write Scale=false
    Module.Write SpriteRotation=false
    Module.Write SpriteSize=true
    Sprite Size=<5 5>
   )
   &FN_Particle_Partition_v5(Enable Ribbon ID=true)
   &InitializeRibbon(
    Module.Write Color=true
    Module.Write Lifetime=true
    Module.Write Mass=true
    Module.Write Material Random=true
    Module.Write Position=false
    Module.Write RibbonTwist=false
    Module.Write RibbonWidth=false
    Ribbon Width=&FloatFromCurve(
     FloatCurve=DataInterface() {propertyValues {Curve {keys [
         {interpMode=RCIM_Cubic, leaveTangent=-1.6682, tangentMode=RCTM_Break, time=0, value=1}
         {arriveTangent=-0.3052, interpMode=RCIM_Cubic, leaveTangent=-0.3052, tangentMode=RCTM_Break, time=0.2774, value=0.441}
         {arriveTangent=-0.3335, interpMode=RCIM_Cubic, leaveTangent=-0.3335, tangentMode=RCTM_Break, time=0.6089, value=0.7531}
         {interpMode=RCIM_Cubic, tangentMode=RCTM_Break, time=1.001, value=0.053}
        ]}}}
     CurveIndex=$Particles.Partition.NormalizedID
     Scale Curve=7
    )
   )
   /SetVariables(Particles.RibbonFacing=<1 0 1>) {enabled=false, isSetParameters=true, module=SetVariables_#1}
  ]
  particleUpdateScript [
   &ParticleState(Kill Particles When Lifetime Has Elapsed=false, Let Infinitely Lived Particles Die When Emitter Deactivates=true)
   &SkeletalMeshLocation(
    Inherit Velocity=false
    MaskSpawnGroup=false
    Mesh Sampling Type=::NewEnumerator5
    Skeletal Mesh=DataInterface() {propertyValues {
      DefaultMesh=null
      ExcludeBoneName=root
      OptionalComponentClass=null
      PreviewMesh=&Female_Medium_Cherry_Head
      SkinningMode=SkinOnTheFly
      SoftSourceActor=null
      SourceMode=Default
      UvSetIndex=0
      WholeMeshLOD=-1
      bExcludeBone=false
      bReadDeformedGeometry=true
      bRequireCurrentFrameData=true
      ComponentTags []
      FilteredBones []
      FilteredSockets [FX_Hair]
      MeshUserParameter {parameter {name=None, typeDefHandle {registeredTypeIndex=1}, varData []}}
      SamplingRegions []
     }}
    Socket Sampling Mode=::NewEnumerator5
   )
   &FN_HermiteCubicSpline(
    Direction End:Vector3f=<40 0 0>
    Direction Start:Vector3f=<-30 25 0>
    Position Blend:NiagaraFloat=$Particles.Initial.Partition.NormalizedID
    Position End:Vector3f=&Add_Vector(
     A=$Output.SkeletalMeshLocation.SampledPosition
     B=&Add_Vector(
      A=<-17 -10 -15>
      B=&Multiply_Vector(
       A=&TransformVector(Vector=$Engine.Owner.Velocity, Source Space=::World, Destination Space=::Local)
       B=&Multiply_Vector(A=<-1 -1 -1>, B=&Clamp_Vector(Vector:Vector3f=<0.02 0.05 0.015>, Min:Vector3f=<0 0 0>, Max:Vector3f=<1 1 1>))
      )
     )
    )
    Position Start:Vector3f=$Output.SkeletalMeshLocation.SampledPosition
    Write Particles Position=true
   )
   &FN_RibbonMaterial_v2(
    Erosion Pan Speed U=1
    Erossion Scale U=2
    Macro Pan Speed=0.5
    Micro Pan Speed=0.2
    Opacity Erosion=0.2
    Opacity Pan U Speed=0.1
    Opacity Power=0.6
    Opacity Scale U=0.2
   ) {enabled=false}
   &FN_RibbonMaterial(
    Color=&ColorFromCurve(
     ColorCurve=DataInterface() {propertyValues {
       AlphaCurve {keys [<at 0.6757 1 RCIM_Cubic>,<at 1.0002 0>]}
       BlueCurve {keys [<at 0.0118 1.3802 RCIM_Cubic>,<at 0.9761 0.026 RCIM_Cubic>]}
       GreenCurve {keys [<at 0.0118 11.4764 RCIM_Cubic>,<at 0.9761 0.5163 RCIM_Cubic>]}
       RedCurve {keys [<at 0.0118 15 RCIM_Cubic>,<at 0.9761 1 RCIM_Cubic>]}
      }}
     CurveIndex=$Particles.Partition.NormalizedID
    )
    Macro Scale U=0
    Micro Scale U=0
    Opacity Pan U Speed=0.5
    Scale Alpha=0.25
   )
   &ScaleRibbonWidth(Ribbon Width Scale=4)
  ]
  rendererClasses [&NiagaraRibbonRendererProperties]
  renderers [{rendererClass=&NiagaraRibbonRendererProperties}]
 }
}
summary {
 systemName=NS_Cherry_Charged_Head
 emitters [
  {emitterName=EyeGlow, rendererClasses [&NiagaraSpriteRendererProperties, &NiagaraLightRendererProperties]}
  {emitterName=EyeTrail, rendererClasses [&NiagaraMeshRendererProperties]}
  {bEnabled=false, emitterName=Hair_L, rendererClasses [&NiagaraRibbonRendererProperties]}
  {bEnabled=false, emitterName=Hair_R, rendererClasses [&NiagaraRibbonRendererProperties]}
  {bEnabled=false, emitterName=Hair_Arcs_R, rendererClasses [&NiagaraMeshRendererProperties]}
  {bEnabled=false, emitterName=Hair_Arcs_L, rendererClasses [&NiagaraMeshRendererProperties]}
 ]
 userVariables [
  User.bIsFrontEndPreview {type=NiagaraBool}
  User.SkeletalMesh {
   type=NiagaraDataInterfaceSkeletalMesh
   defaultValue {
    struct=NiagaraExt_VariableValue_DataInterface
    value {dataInterface=&self:NiagaraDataInterfaceSkeletalMesh_0}
   }
  }
 ]
}
