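# Material function in the F_MED_Cherry cosmetic body folder that builds a soft cylinder-style mask around a center point.
# It rotates the position (world position by default) about an axis through that center, then derives Z and XY distance falloffs with separate divisors.
# The inverted, saturated falloffs multiply into a mask, which is then multiplied by itself plus Bias.
# The XY term feeds rotated X into both X and Y, ignoring rotated Y.
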MaterialFunction /BRCosmetics/Characters/Player/Female/Medium/Bodies/F_MED_Cherry/Materials/MF_CylindricalMask
graph {nodes [
  n1 = FunctionInput(Preview=Constant3Vector() {constant=<rgba 1 0 0 0>}) {bUsePreviewValueAsDefault=true, inputName=NormalizedRotationAxis, sortPriority=1}
  n2 = Constant()
  n3 = FunctionInput(Preview=Constant3Vector()) {bUsePreviewValueAsDefault=true, inputName=CenterPoint}
  n4 = FunctionInput(Preview=WorldPosition().XYZ) {bUsePreviewValueAsDefault=true, inputName=Position, sortPriority=5}
  n5 = Add(
   A=RotateAboutAxis(
    NormalizedRotationAxis=Normalize(VectorInput=@n1)
    RotationAngle=FunctionInput(Preview=@n2) {bUsePreviewValueAsDefault=true, inputName=RotationAngle, inputType=FunctionInput_Scalar, sortPriority=2}
    PivotPoint=@n3
    Position=@n4
   )
   B=@n4
  )
  n6 = Constant3Vector() {constant=<rgba 0 0 1 0>}
  n7 = Divide(
   A=Distance(A=Multiply(A=@n5, B=@n6), B=@n3)
   B=FunctionInput(Preview=Constant() {r=140}) {bUsePreviewValueAsDefault=true, inputName=Z_Divisor, inputType=FunctionInput_Scalar, sortPriority=4}
  )
  n8 = MaterialFunctionCall(Float3=@n5) {materialFunction=&BreakOutFloat3Components}
  n9 = MaterialFunctionCall(Float3=@n3) {materialFunction=&BreakOutFloat3Components}
  n10 = Divide(
   A=Distance(
    A=@n3
    B=MaterialFunctionCall(X=@n8.R, Y=@n8.R, Z=@n9.B) {materialFunction=&MakeFloat3}.Result
   )
   B=FunctionInput(Preview=Constant() {r=20}) {bUsePreviewValueAsDefault=true, inputName=XY_Divisor, inputType=FunctionInput_Scalar, sortPriority=3}
  )
  n11 = Multiply(A=OneMinus(None=Saturate(None=@n7)), B=OneMinus(None=Saturate(None=@n10)))
  n12 = FunctionInput() {bUsePreviewValueAsDefault=true, inputName=Bias, inputType=FunctionInput_Scalar, sortPriority=6, previewValue=<1 0 0 1>}
  FunctionOutput(None=Multiply(A=@n11, B=Add(A=@n11, B=@n12)))
 ]}
