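# Material function for the Cherry character body that adds an animated energy glow to the incoming material's emissive.
# Panning electric noise, projected in pre-skinned local space, fills cylindrical leg and arm masks, joined by hand and foot sphere glows; absolute position mirrors both sides.
# Timestamp-driven in/out transitions animate the masks.
# A Charged static switch holds hand glows on and adds tinted, noisy Fresnel, optionally height-weighted and with UV-masked eye glow.
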MaterialFunction /BRCosmetics/Characters/Player/Female/Medium/Bodies/F_MED_Cherry/Materials/MF_Cherry_Glow
graph {nodes [
  n1 = GetMaterialAttributes(None=FunctionInput() {inputType=FunctionInput_MaterialAttributes}) {attributeGetTypes [#1, #2, #3]}
  n2 = FunctionInput() {inputName=PreskinLocalPosition, sortPriority=1}
  n3 = PreSkinnedLocalBounds()
  n4 = ComponentMask(None=Divide(A=Subtract(A=@n2, B=@n3.Min), B=@n3.Extents)) {b=true}
  n5 = ScalarParameter() {defaultValue=3, group=Cherry, parameterName=PreskinBoundsExponent}
  n6 = MaterialFunctionCall(
   Brightness=ScalarParameter() {defaultValue=1, group=Cherry, parameterName=FresnelBrightness}
   Exponent=ScalarParameter() {defaultValue=1.8, group=Cherry, parameterName=FresnelExponent}
   Vertex / Pixel Normal=Transform(None=@n1.Normal)
  ) {materialFunction=&MF_Fresnel_FX}
  n7 = Multiply(
   A=Power(Base=@n6.Fresnel Inverted Unclamped, Exp=ScalarParameter() {defaultValue=10, group=Cherry, parameterName=FresnelHotEdgeExponent})
   B=ScalarParameter() {defaultValue=200, group=Cherry, parameterName=FresnelHotEdgeScale}
  )
  n8 = Abs(None=@n2)
  n9 = Multiply(
   A=MaterialFunctionCall!(
    Time
    Divisor=ScalarParameter() {defaultValue=0.5, group=Cherry, parameterName=TimeDivisor}
    Mod=ScalarParameter() {defaultValue=1000, group=Cherry, parameterName=TimeMod}
   ) {materialFunction=/Game/Effects/Fort_Effects/Materials/_Functions/MF_ModuloTime}.Unclamped
   B=VectorParameter() {group=Cherry, parameterName=NoisePanSpeed, defaultValue=<rgba -10 2 10 0>}.RGB
  )
  n10 = PreSkinnedNormal()
  n11 = ComponentMask(None=MaterialFunctionCall(
    TextureObject=TextureObject() {samplerType=SAMPLERTYPE_LinearColor, texture=/Game/Characters/Player/Male/Medium/Bodies/M_MED_SRDrift/Textures/T_ElectricNoise}
    TextureSize=ScalarParameter() {defaultValue=80, group=Cherry, parameterName=NoiseWorldTileScale}
    WorldPosition=Add(A=@n8, B=@n9)
    World Space Normal=VertexInterpolator(VS=@n10).PS
    ProjectionTransitionContrast=ScalarParameter() {defaultValue=1, group=Cherry, parameterName=ProjectionTransitionContrast}
   ) {materialFunction=/Game/Weapons/WeaponSkins/Wraps/MaterialFunctions/WorldAlignedTexture}.XYZ Texture) {r=true}
  n12 = Multiply(A=ScalarParameter() {defaultValue=6, group=Cherry, parameterName=ChargedNoiseBrighness}, B=@n11)
  n13 = Multiply(
   A=StaticSwitchParameter(True=Power(Base=@n4, Exp=@n5), False=Constant() {r=1}) {defaultValue=true, group=Cherry, parameterName=ModulateChargeByPreskinBoundsZ}
   B=Add(A=Add(A=@n7, B=@n6.Fresnel Inverted Unclamped), B=Multiply(A=@n6.Fresnel Inverted Unclamped, B=@n12))
  )
  n14 = SphereMask(
   A=TextureCoordinate() {uTiling=2}
   B=ComponentMask(None=VectorParameter() {group=Cherry, parameterName=ChargedEyesCenter}.RGB) {g=true, r=true}
   Radius=ScalarParameter() {defaultValue=0.1, group=Cherry, parameterName=EyesRadius}
   Hardness=Constant() {r=1}
  )
  n15 = ScalarParameter() {defaultValue=30, group=Cherry, parameterName=EyesScale}
  n16 = StaticSwitchParameter(True=Add(A=@n13, B=Multiply(A=@n14, B=@n15)), False=@n13) {group=Cherry, parameterName=ChargedEyes}
  n17 = VectorParameter() {group=Cherry, parameterName=ChargedColor, defaultValue=<rgba 0.8 0.4261 0.0764 0>}
  n18 = VectorParameter() {group=Cherry, parameterName=FeetCenterPoint, defaultValue=<rgba 16 6 4 0>}
  n19 = StaticSwitchParameter(True=@n8, False=@n2) {defaultValue=true, group=Cherry, parameterName=UseAbsPreskinLocalPosition}
  n20 = ScalarParameter() {defaultValue=0.6, group=Cherry, parameterName=TransitionLength}
  n21 = MaterialFunctionCall(
   TimestampA=ScalarParameter() {group=Cherry, parameterName=TimestampA}
   TransitionLength_In=@n20
   TimestampB=ScalarParameter() {group=Cherry, parameterName=TimestampB}
   TransitionLength_Out=@n20
   Transition In Only?=StaticBool()
  ) {materialFunction=&MF_Timestamp_InOut_01}
  n22 = ConstantBiasScale(None=@n21.Linear) {bias=-0.5, scale=2}
  n23 = MaterialFunctionCall!(
   CenterPoint=@n18.RGB
   NormalizedRotationAxis=VectorParameter() {group=Cherry, parameterName=FeetNormalizedRotationAxis, defaultValue=<rgba 0 1 0 0>}.RGB
   RotationAngle=ScalarParameter() {defaultValue=0.015, group=Cherry, parameterName=FeetMaskRotationAngle}
   XY_Divisor=ScalarParameter() {defaultValue=9, group=Cherry, parameterName=Legs_XY_Divisor}
   Z_Divisor=ScalarParameter() {defaultValue=65, group=Cherry, parameterName=Legs_Z_Divisor}
   Position=@n19
   Bias=@n22
  ) {materialFunction=&/MF_CylindricalMask}
  n24 = VectorParameter() {group=Cherry, parameterName=ArmsCenterPoint, defaultValue=<rgba 51.24 6 98 0>}
  n25 = MaterialFunctionCall(
   CenterPoint=@n24.RGB
   NormalizedRotationAxis=VectorParameter() {group=Cherry, parameterName=ArmsNormalizedRotationAxis, defaultValue=<rgba 0 1 0 0>}.RGB
   RotationAngle=ScalarParameter() {defaultValue=0.15, desc=0.105, group=Cherry, parameterName=ArmsMaskRotationAngle}
   XY_Divisor=ScalarParameter() {defaultValue=20, group=Cherry, parameterName=Arms_XY_Divisor}
   Z_Divisor=ScalarParameter() {defaultValue=85, group=Cherry, parameterName=Arms_Z_Divisor}
   Position=@n19
   Bias=StaticSwitchParameter(True=ConstantBiasScale(None=@n21.Linear), False=@n22) {group=Cherry, parameterName=Charged}
  ) {materialFunction=&MF_CylindricalMask}
  n26 = Power(
   Base=Multiply(A=@n11, B=Add(A=@n23.Result, B=@n25.Result) {desc=Masks})
   Exp=ScalarParameter() {defaultValue=1.3, group=Cherry, parameterName=NoisePreExponentA}
  )
  n27 = Power(
   Base=SphereMask(
    A=@n19
    B=@n18.RGB
    Radius=ScalarParameter() {defaultValue=26, group=Cherry, parameterName=FeetRadius}
    Hardness=ScalarParameter() {group=Cherry, parameterName=FeetHardness}
   )
   Exp=ScalarParameter() {defaultValue=2.8, group=Cherry, parameterName=FeetExponent}
  )
  n28 = Power(
   Base=SphereMask(A=@n19, B=@n24.RGB, Radius=ScalarParameter() {defaultValue=45, group=Cherry, parameterName=HandRadius}, Hardness=Constant())
   Exp=ScalarParameter() {defaultValue=3, group=Cherry, parameterName=HandExponent}
  )
  n29 = StaticSwitchParameter(True=Constant() {r=1}, False=@n21.Linear) {group=Cherry, parameterName=Charged}
  n30 = Multiply(
   A=Add(A=Multiply(A=@n27, B=@n21.Linear), B=Multiply(A=@n28, B=@n29)) {desc=Sphere Glows}
   B=ScalarParameter() {defaultValue=0.35, group=Cherry, parameterName=HandFeetMaskScale}
  )
  n31 = Power(
   Base=Multiply(A=Add(A=@n26, B=@n30), B=VectorParameter() {group=Cherry, parameterName=NoiseColor, defaultValue=<rgba 10 6.4337 1.441 0>}.RGB)
   Exp=ScalarParameter() {defaultValue=2, desc=2.8, group=Cherry, parameterName=NoiseExponentA}
  )
  n32 = StaticSwitchParameter(True=Add(A=Multiply(A=@n16, B=@n17.RGB), B=@n31), False=@n31) {group=Cherry, parameterName=Charged}
  n33 = StaticSwitchParameter(True=Multiply(A=@n1.BaseColor, B=@n32), False=@n32) {group=Cherry, parameterName=ModulateEmissiveByBaseColor}
  FunctionOutput(None=SetMaterialAttributes!(MaterialAttributes=@n1.MaterialAttributes, EmissiveColor=Add(A=@n1.EmissiveColor, B=@n33)) {attributeSetTypes [#1]})
 ]}
