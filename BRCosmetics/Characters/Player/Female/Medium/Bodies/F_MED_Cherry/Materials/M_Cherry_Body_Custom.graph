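# Body material for the Cherry female medium player character, a Default Lit skeletal mesh material built on material attributes and usable with morph targets and mesh deformers.
# It chains shared character functions: PreFX feeds CharacterSkinFX, whose result passes through the Cherry-specific MF_Cherry_Glow and into PostFX.
# The glow function also receives pre-skinned local position from PreFX, and all surface logic lives inside these material functions.
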Material /BRCosmetics/Characters/Player/Female/Medium/Bodies/F_MED_Cherry/Materials/M_Cherry_Body_Custom
graph {
 properties {
  bUseMaterialAttributes=true
  bUsedWithMeshDeformer=true
  bUsedWithMorphTargets=true
  bUsedWithSkeletalMesh=true
  pixelDepthOffsetMode=PDOM_Legacy
  refractionMethod=RM_IndexOfRefraction
  usedShadingModels=Default Lit
 }
 nodes [
  n1 = MaterialFunctionCall() {materialFunction=&BaseFN_CharacterFunction_PreFX}
  MaterialFunctionCall(
   AO=@n1.AO
   Distance Blend=@n1.AO
   Fresnel=@n1.AO
   Gmap=@n1.AO
   Pre Skinned Local Position=@n1.AO
   Material=MaterialFunctionCall!(
    In=MaterialFunctionCall(Gmap=@n1.Gmap, MaterialAttributes=@n1.Gmap, AO=@n1.Gmap) {materialFunction=&CharacterSkinFX}.Result
    PreskinLocalPosition=@n1.Pre skinned Local Position
   ) {materialFunction=&/MF_Cherry_Glow}.Result
  ) {materialFunction=&BaseFN_CharacterFunction_PostFX}
 ]
}
