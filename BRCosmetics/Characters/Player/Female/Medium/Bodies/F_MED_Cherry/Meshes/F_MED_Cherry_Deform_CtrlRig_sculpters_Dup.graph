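# Control Rig for the F_MED_Cherry body that drives corrective deform bones and procedurally poses its scarf bones.
# Elbow, knee, shoulder and hip angles are remapped into weights that blend the deform bones toward authored offsets.
# When isRBANEnabled is false, swing/twist readings of the right leg weight scarf and medallion offsets, alongside interpolated poses for crouch, knee slide and baseball slide states.
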ControlRigBlueprint /BRCosmetics/Characters/Player/Female/Medium/Bodies/F_MED_Cherry/Meshes/F_MED_Cherry_Deform_CtrlRig_sculpters_Dup
graphs [
 Multiply_1_ContainedGraph {nodes [
   n1 = RigVMFunctionEntryNode()
   Multiply = &RigVMFunction_MathDoubleMul(A=@n1.A, B=@n1.B)
   RigVMFunctionReturnNode(Result=&RigVMFunction_MathDoubleMul(B=@n1.C, A=@Multiply.Result).Result)
  ]}
 New Function_ContainedGraph {nodes [
   n1 = RigVMFunctionEntryNode()
   RigVMFunction_MathBoolToInteger = &RigVMFunction_MathBoolToInteger(Value=@n1.onlyTwist)
   RigVMFunction_MathQuaternionSwingTwist = &RigVMFunction_MathQuaternionSwingTwist(
    TwistAxis=@n1.twistAxis
    Input=&RigUnit_GetRelativeTransformForItem(bChildInitial=False, bParentInitial=False, Child=@n1.child, Parent=@n1.parent).RelativeTransform.Rotation
   )
   n2 = RigVMDispatchNode(
    Values=((),())
    Index=@RigVMFunction_MathBoolToInteger.Result
    Values.0=@RigVMFunction_MathQuaternionSwingTwist.Swing
    Values.1=@RigVMFunction_MathQuaternionSwingTwist.Twist
   ) {template="DISPATCH_RigVMDispatch_SelectInt32(in Index,in Values,out Result)"}
   RigVMFunction_MathQuaternionToEuler = &RigVMFunction_MathQuaternionToEuler(Value=@n2.Result, RotationOrder=@n1.Rotation Order)
   > DISPATCH_RigVMDispatch_Print = RigVMDispatchNode(ScreenColor=(R=1,G=1,B=1,A=1), ScreenDuration=1.5, Enabled=@n1.debugRotation, Value=@RigVMFunction_MathQuaternionToEuler.Result) {template="DISPATCH_RigVMDispatch_Print(in Prefix,in Value,in Enabled,in ScreenDuration,in ScreenColor)"}
   Cast = RigVMDispatchNode(Value=@n1.triggerAxis) {template="DISPATCH_RigVMDispatch_CastEnumToInt(in Value,out Result)"}
   DISPATCH_RigVMDispatch_SelectInt32 = RigVMDispatchNode(
    Values=(0,0,0)
    Index=@Cast.Result
    Values.0=@RigVMFunction_MathQuaternionToEuler.Result.X
    Values.1=@RigVMFunction_MathQuaternionToEuler.Result.Y
    Values.2=@RigVMFunction_MathQuaternionToEuler.Result.Z
   ) {template="DISPATCH_RigVMDispatch_SelectInt32(in Index,in Values,out Result)"}
   Remap = &RigVMFunction_MathDoubleRemap(
    bClamp=False
    Value=@DISPATCH_RigVMDispatch_SelectInt32.Result
    TargetMinimum=@n1.outputMinMax.X
    TargetMaximum=@n1.outputMinMax.Y
    SourceMinimum=@n1.rotationMinMax.X
    SourceMaximum=@n1.rotationMinMax.Y
   )
   Absolute = &RigVMFunction_MathDoubleAbs(Value=@Remap.Result)
   Multiply = &RigVMFunction_MathDoubleMul(
    A=&RigVMFunction_MathDoubleClamp(Value=RigVMDispatchNode(True=@Absolute.Result, Condition=@n1.absoluteValue, False=@Remap.Result) {template="DISPATCH_RigVMDispatch_If(in Condition,in True,in False,out Result)"}.Result).Result
    B=@n1.premult
   )
   > RigVMFunctionReturnNode(outputQuat=@n2.Result, outputWeight=@Multiply.Result)
  ]}
 New Function_ContainedGraph {nodes [
   n1 = RigVMFunctionEntryNode()
   > For_Each = RigVMDispatchNode(Array=@n1.transforms) {template="DISPATCH_RigVMDispatch_ArrayIterator(in Array,out Element,out Index,out Count,out Ratio)"}
   > &RigUnit_OffsetTransformForItem(Item=@For_Each.Element.Item, OffsetTransform=@For_Each.Element.Transform, Weight=@n1.weights, bPropagateToChildren=@n1.inheritsTransform)
   RigVMFunctionReturnNode(ExecuteContext=@For_Each.Completed)
  ]}
 New Function_ContainedGraph {nodes [
   n1 = RigVMFunctionEntryNode()
   RigVMFunction_MathBoolToInteger = &RigVMFunction_MathBoolToInteger(Value=@n1.onlyTwist)
   RigVMFunction_MathQuaternionSwingTwist = &RigVMFunction_MathQuaternionSwingTwist(
    TwistAxis=@n1.twistAxis
    Input=&RigUnit_GetRelativeTransformForItem(bChildInitial=False, bParentInitial=False, Child=@n1.child, Parent=@n1.parent).RelativeTransform.Rotation
   )
   n2 = RigVMDispatchNode(
    Values=((),())
    Index=@RigVMFunction_MathBoolToInteger.Result
    Values.0=@RigVMFunction_MathQuaternionSwingTwist.Swing
    Values.1=@RigVMFunction_MathQuaternionSwingTwist.Twist
   ) {template="DISPATCH_RigVMDispatch_SelectInt32(in Index,in Values,out Result)"}
   RigVMFunction_MathQuaternionToEuler = &RigVMFunction_MathQuaternionToEuler(Value=@n2.Result, RotationOrder=@n1.Rotation Order)
   > DISPATCH_RigVMDispatch_Print = RigVMDispatchNode(ScreenColor=(R=1,G=1,B=1,A=1), ScreenDuration=1.5, Enabled=@n1.debugRotation, Value=@RigVMFunction_MathQuaternionToEuler.Result) {template="DISPATCH_RigVMDispatch_Print(in Prefix,in Value,in Enabled,in ScreenDuration,in ScreenColor)"}
   Cast = RigVMDispatchNode(Value=@n1.triggerAxis) {template="DISPATCH_RigVMDispatch_CastEnumToInt(in Value,out Result)"}
   DISPATCH_RigVMDispatch_SelectInt32 = RigVMDispatchNode(
    Values=(0,0,0)
    Index=@Cast.Result
    Values.0=@RigVMFunction_MathQuaternionToEuler.Result.X
    Values.1=@RigVMFunction_MathQuaternionToEuler.Result.Y
    Values.2=@RigVMFunction_MathQuaternionToEuler.Result.Z
   ) {template="DISPATCH_RigVMDispatch_SelectInt32(in Index,in Values,out Result)"}
   Remap = &RigVMFunction_MathDoubleRemap(
    bClamp=False
    Value=@DISPATCH_RigVMDispatch_SelectInt32.Result
    TargetMinimum=@n1.outputMinMax.X
    TargetMaximum=@n1.outputMinMax.Y
    SourceMinimum=@n1.rotationMinMax.X
    SourceMaximum=@n1.rotationMinMax.Y
   )
   Absolute = &RigVMFunction_MathDoubleAbs(Value=@Remap.Result)
   Multiply = &RigVMFunction_MathDoubleMul(
    A=&RigVMFunction_MathDoubleClamp(Value=RigVMDispatchNode(True=@Absolute.Result, Condition=@n1.absoluteValue, False=@Remap.Result) {template="DISPATCH_RigVMDispatch_If(in Condition,in True,in False,out Result)"}.Result).Result
    B=@n1.premult
   )
   > RigVMFunctionReturnNode(outputQuat=@n2.Result, outputWeight=@Multiply.Result)
  ]}
 New Function_ContainedGraph {nodes [
   n1 = RigVMFunctionEntryNode()
   RigVMFunction_MathBoolToInteger = &RigVMFunction_MathBoolToInteger(Value=@n1.onlyTwist)
   RigVMFunction_MathQuaternionSwingTwist = &RigVMFunction_MathQuaternionSwingTwist(
    TwistAxis=@n1.twistAxis
    Input=&RigUnit_GetRelativeTransformForItem(bChildInitial=False, bParentInitial=False, Child=@n1.child, Parent=@n1.parent).RelativeTransform.Rotation
   )
   n2 = RigVMDispatchNode(
    Values=((),())
    Index=@RigVMFunction_MathBoolToInteger.Result
    Values.0=@RigVMFunction_MathQuaternionSwingTwist.Swing
    Values.1=@RigVMFunction_MathQuaternionSwingTwist.Twist
   ) {template="DISPATCH_RigVMDispatch_SelectInt32(in Index,in Values,out Result)"}
   RigVMFunction_MathQuaternionToEuler = &RigVMFunction_MathQuaternionToEuler(Value=@n2.Result, RotationOrder=@n1.Rotation Order)
   > DISPATCH_RigVMDispatch_Print = RigVMDispatchNode(ScreenColor=(R=1,G=1,B=1,A=1), ScreenDuration=1.5, Enabled=@n1.debugRotation, Value=@RigVMFunction_MathQuaternionToEuler.Result) {template="DISPATCH_RigVMDispatch_Print(in Prefix,in Value,in Enabled,in ScreenDuration,in ScreenColor)"}
   Cast = RigVMDispatchNode(Value=@n1.triggerAxis) {template="DISPATCH_RigVMDispatch_CastEnumToInt(in Value,out Result)"}
   DISPATCH_RigVMDispatch_SelectInt32 = RigVMDispatchNode(
    Values=(0,0,0)
    Index=@Cast.Result
    Values.0=@RigVMFunction_MathQuaternionToEuler.Result.X
    Values.1=@RigVMFunction_MathQuaternionToEuler.Result.Y
    Values.2=@RigVMFunction_MathQuaternionToEuler.Result.Z
   ) {template="DISPATCH_RigVMDispatch_SelectInt32(in Index,in Values,out Result)"}
   Remap = &RigVMFunction_MathDoubleRemap(
    bClamp=False
    Value=@DISPATCH_RigVMDispatch_SelectInt32.Result
    TargetMinimum=@n1.outputMinMax.X
    TargetMaximum=@n1.outputMinMax.Y
    SourceMinimum=@n1.rotationMinMax.X
    SourceMaximum=@n1.rotationMinMax.Y
   )
   Absolute = &RigVMFunction_MathDoubleAbs(Value=@Remap.Result)
   Multiply = &RigVMFunction_MathDoubleMul(
    A=&RigVMFunction_MathDoubleClamp(Value=RigVMDispatchNode(True=@Absolute.Result, Condition=@n1.absoluteValue, False=@Remap.Result) {template="DISPATCH_RigVMDispatch_If(in Condition,in True,in False,out Result)"}.Result).Result
    B=@n1.premult
   )
   > RigVMFunctionReturnNode(outputQuat=@n2.Result, outputWeight=@Multiply.Result)
  ]}
 New Function_ContainedGraph {nodes [
   n1 = RigVMFunctionEntryNode()
   RigVMFunction_MathBoolToInteger = &RigVMFunction_MathBoolToInteger(Value=@n1.onlyTwist)
   RigVMFunction_MathQuaternionSwingTwist = &RigVMFunction_MathQuaternionSwingTwist(
    TwistAxis=@n1.twistAxis
    Input=&RigUnit_GetRelativeTransformForItem(bChildInitial=False, bParentInitial=False, Child=@n1.child, Parent=@n1.parent).RelativeTransform.Rotation
   )
   n2 = RigVMDispatchNode(
    Values=((),())
    Index=@RigVMFunction_MathBoolToInteger.Result
    Values.0=@RigVMFunction_MathQuaternionSwingTwist.Swing
    Values.1=@RigVMFunction_MathQuaternionSwingTwist.Twist
   ) {template="DISPATCH_RigVMDispatch_SelectInt32(in Index,in Values,out Result)"}
   RigVMFunction_MathQuaternionToEuler = &RigVMFunction_MathQuaternionToEuler(Value=@n2.Result, RotationOrder=@n1.Rotation Order)
   > DISPATCH_RigVMDispatch_Print = RigVMDispatchNode(ScreenColor=(R=1,G=1,B=1,A=1), ScreenDuration=1.5, Enabled=@n1.debugRotation, Value=@RigVMFunction_MathQuaternionToEuler.Result) {template="DISPATCH_RigVMDispatch_Print(in Prefix,in Value,in Enabled,in ScreenDuration,in ScreenColor)"}
   Cast = RigVMDispatchNode(Value=@n1.triggerAxis) {template="DISPATCH_RigVMDispatch_CastEnumToInt(in Value,out Result)"}
   DISPATCH_RigVMDispatch_SelectInt32 = RigVMDispatchNode(
    Values=(0,0,0)
    Index=@Cast.Result
    Values.0=@RigVMFunction_MathQuaternionToEuler.Result.X
    Values.1=@RigVMFunction_MathQuaternionToEuler.Result.Y
    Values.2=@RigVMFunction_MathQuaternionToEuler.Result.Z
   ) {template="DISPATCH_RigVMDispatch_SelectInt32(in Index,in Values,out Result)"}
   Remap = &RigVMFunction_MathDoubleRemap(
    bClamp=False
    Value=@DISPATCH_RigVMDispatch_SelectInt32.Result
    TargetMinimum=@n1.outputMinMax.X
    TargetMaximum=@n1.outputMinMax.Y
    SourceMinimum=@n1.rotationMinMax.X
    SourceMaximum=@n1.rotationMinMax.Y
   )
   Absolute = &RigVMFunction_MathDoubleAbs(Value=@Remap.Result)
   Multiply = &RigVMFunction_MathDoubleMul(
    A=&RigVMFunction_MathDoubleClamp(Value=RigVMDispatchNode(True=@Absolute.Result, Condition=@n1.absoluteValue, False=@Remap.Result) {template="DISPATCH_RigVMDispatch_If(in Condition,in True,in False,out Result)"}.Result).Result
    B=@n1.premult
   )
   > RigVMFunctionReturnNode(outputQuat=@n2.Result, outputWeight=@Multiply.Result)
  ]}
 New Function_ContainedGraph {nodes [
   n1 = RigVMFunctionEntryNode()
   RigVMFunction_MathBoolToInteger = &RigVMFunction_MathBoolToInteger(Value=@n1.onlyTwist)
   RigVMFunction_MathQuaternionSwingTwist = &RigVMFunction_MathQuaternionSwingTwist(
    TwistAxis=@n1.twistAxis
    Input=&RigUnit_GetRelativeTransformForItem(bChildInitial=False, bParentInitial=False, Child=@n1.child, Parent=@n1.parent).RelativeTransform.Rotation
   )
   n2 = RigVMDispatchNode(
    Values=((),())
    Index=@RigVMFunction_MathBoolToInteger.Result
    Values.0=@RigVMFunction_MathQuaternionSwingTwist.Swing
    Values.1=@RigVMFunction_MathQuaternionSwingTwist.Twist
   ) {template="DISPATCH_RigVMDispatch_SelectInt32(in Index,in Values,out Result)"}
   RigVMFunction_MathQuaternionToEuler = &RigVMFunction_MathQuaternionToEuler(Value=@n2.Result, RotationOrder=@n1.Rotation Order)
   > DISPATCH_RigVMDispatch_Print = RigVMDispatchNode(ScreenColor=(R=1,G=1,B=1,A=1), ScreenDuration=1.5, Enabled=@n1.debugRotation, Value=@RigVMFunction_MathQuaternionToEuler.Result) {template="DISPATCH_RigVMDispatch_Print(in Prefix,in Value,in Enabled,in ScreenDuration,in ScreenColor)"}
   Cast = RigVMDispatchNode(Value=@n1.triggerAxis) {template="DISPATCH_RigVMDispatch_CastEnumToInt(in Value,out Result)"}
   DISPATCH_RigVMDispatch_SelectInt32 = RigVMDispatchNode(
    Values=(0,0,0)
    Index=@Cast.Result
    Values.0=@RigVMFunction_MathQuaternionToEuler.Result.X
    Values.1=@RigVMFunction_MathQuaternionToEuler.Result.Y
    Values.2=@RigVMFunction_MathQuaternionToEuler.Result.Z
   ) {template="DISPATCH_RigVMDispatch_SelectInt32(in Index,in Values,out Result)"}
   Remap = &RigVMFunction_MathDoubleRemap(
    bClamp=False
    Value=@DISPATCH_RigVMDispatch_SelectInt32.Result
    TargetMinimum=@n1.outputMinMax.X
    TargetMaximum=@n1.outputMinMax.Y
    SourceMinimum=@n1.rotationMinMax.X
    SourceMaximum=@n1.rotationMinMax.Y
   )
   Absolute = &RigVMFunction_MathDoubleAbs(Value=@Remap.Result)
   Multiply = &RigVMFunction_MathDoubleMul(
    A=&RigVMFunction_MathDoubleClamp(Value=RigVMDispatchNode(True=@Absolute.Result, Condition=@n1.absoluteValue, False=@Remap.Result) {template="DISPATCH_RigVMDispatch_If(in Condition,in True,in False,out Result)"}.Result).Result
    B=@n1.premult
   )
   > RigVMFunctionReturnNode(outputQuat=@n2.Result, outputWeight=@Multiply.Result)
  ]}
 RigVMFunctionLibrary {links [], nodes []}
 RigVMModel {nodes [
   RigUnit_BeginExecution = &RigUnit_BeginExecution()
   n1 = &RigVMFunction_ControlFlowBranch(ExecuteContext=@RigUnit_BeginExecution.ExecutePin, Condition=RigVMVariableNode() {variable=isRBANEnabled}.Value)
   RigUnit_GetRelativeTransformForItem = &RigUnit_GetRelativeTransformForItem(Child=(Type=Bone,Name="lowerarm_l"), Parent=(Type=Bone,Name="upperarm_l"), bParentInitial=false)
   n2 = &RigVMFunction_MathQuaternionToEuler(RotationOrder=ZYX, Value=@RigUnit_GetRelativeTransformForItem.RelativeTransform.Rotation)
   n3 = &RigVMFunction_MathFloatRemap(SourceMaximum=92, SourceMinimum=28, TargetMinimum=0, bClamp=true, Value=@n2.Result.Z)
   RigUnit_GetTransform = &RigUnit_GetTransform(Item=(Type=Bone,Name="deform_elbow_bk_l"), Space=LocalSpace, bInitial=true)
   &RigUnit_ModifyTransforms(
    ItemToModify=((Item=(Type=Bone,Name="deform_elbow_bk_l"),Transform=(Rotation=(X=0,Y=0,Z=0,W=1),Translation=(X=-3.505,Y=-1.083,Z=0),Scale3D=(X=1,Y=1,Z=1))))
    Mode=OverrideLocal
    Weight=@n3.Result
    ExecutePin=@n1.Completed
    ItemToModify.0.Transform.Rotation=@RigUnit_GetTransform.Transform.Rotation
   )
   n4 = &RigVMFunction_MathQuaternionToEuler(
    RotationOrder=ZYX
    Value=&RigUnit_GetTransform(Item=(Type=Bone,Name="deform_elbow_fr_l"), Space=LocalSpace, bInitial=true).Transform.Rotation
   )
   > &RigUnit_ModifyTransforms(
    ItemToModify=((Item=(Type=Bone,Name="deform_elbow_fr_l"),Transform=(Rotation=(X=0,Y=0,Z=0,W=1),Translation=(X=20.65,Y=2.176,Z=0),Scale3D=(X=1,Y=1,Z=1.125))))
    Mode=OverrideLocal
    Weight=@n3.Result
    ItemToModify.0.Transform.Rotation=&RigVMFunction_MathQuaternionFromEuler(Euler=/Script/RigVM.RigVMFunction_MathVectorAdd(B=(X=0,Y=0,Z=-42.186), A=@n4.Result) {is=struct RigVMUnitNode, template="Add::Execute(in A,in B,out Result)"}.Result).Result
   )
   n5 = &RigVMFunction_MathFloatRemap(SourceMaximum=148, SourceMinimum=91, TargetMinimum=0, bClamp=true, Value=@n2.Result.Z)
   > &RigUnit_ModifyTransforms(
    ItemToModify=((Item=(Type=Bone),Transform=(Rotation=(X=0,Y=0,Z=0,W=1),Translation=(X=0,Y=0,Z=0),Scale3D=(X=1,Y=1,Z=1))),(Item=(Type=Bone,Name="deform_elbow_bk_l"),Transform=(Rotation=(X=0,Y=0,Z=0,W=1),Translation=(X=-5,Y=-1,Z=0),Scale3D=(X=1,Y=1,Z=1))))
    Mode=OverrideLocal
    Weight=@n5.Result
    ItemToModify.1.Transform.Rotation=@RigUnit_GetTransform.Transform.Rotation
   )
   > &RigUnit_ModifyTransforms(
    ItemToModify=((Item=(Type=Bone),Transform=(Rotation=(X=0,Y=0,Z=0,W=1),Translation=(X=0,Y=0,Z=0),Scale3D=(X=1,Y=1,Z=1))),(Item=(Type=Bone,Name="deform_elbow_fr_l"),Transform=(Rotation=(X=0,Y=0,Z=0,W=1),Translation=(X=15,Y=2.176,Z=0),Scale3D=(X=1,Y=1,Z=1.337))))
    Mode=OverrideLocal
    Weight=@n5.Result
    ItemToModify.1.Transform.Rotation=&RigVMFunction_MathQuaternionFromEuler(Euler=&RigVMFunction_MathVectorAdd(B=(X=0,Y=0,Z=-50), A=@n4.Result) {template="Add::Execute(in A,in B,out Result)"}.Result).Result
   )
   n6 = &RigVMFunction_MathQuaternionToEuler(
    RotationOrder=ZYX
    Value=&RigUnit_GetRelativeTransformForItem(Child=(Type=Bone,Name="lowerarm_r"), Parent=(Type=Bone,Name="upperarm_r"), bParentInitial=false).RelativeTransform.Rotation
   )
   n7 = &RigVMFunction_MathFloatRemap(SourceMaximum=92, SourceMinimum=28, TargetMinimum=0, bClamp=true, Value=@n6.Result.Z)
   n8 = &RigUnit_GetTransform(Item=(Type=Bone,Name="deform_elbow_bk_r"), Space=LocalSpace, bInitial=true)
   > &RigUnit_ModifyTransforms(
    ItemToModify=((Item=(Type=Bone,Name="deform_elbow_bk_r"),Transform=(Rotation=(X=0,Y=0,Z=0,W=1),Translation=(X=3.505,Y=1.083,Z=0),Scale3D=(X=1,Y=1,Z=1))))
    Mode=OverrideLocal
    Weight=@n7.Result
    ItemToModify.0.Transform.Rotation=@n8.Transform.Rotation
   )
   n9 = &RigVMFunction_MathQuaternionToEuler(
    RotationOrder=ZYX
    Value=&RigUnit_GetTransform(Item=(Type=Bone,Name="deform_elbow_fr_r"), Space=LocalSpace, bInitial=true).Transform.Rotation
   )
   > &RigUnit_ModifyTransforms(
    ItemToModify=((Item=(Type=Bone,Name="deform_elbow_fr_r"),Transform=(Rotation=(X=0,Y=0,Z=0,W=1),Translation=(X=-20.65,Y=-2.176,Z=0),Scale3D=(X=1,Y=1,Z=1.125))))
    Mode=OverrideLocal
    Weight=@n7.Result
    ItemToModify.0.Transform.Rotation=&RigVMFunction_MathQuaternionFromEuler(Euler=&RigVMFunction_MathVectorAdd(B=(X=0,Y=0,Z=-42.186), A=@n9.Result) {template="Add::Execute(in A,in B,out Result)"}.Result).Result
   )
   n10 = &RigVMFunction_MathFloatRemap(SourceMaximum=148, SourceMinimum=91, TargetMinimum=0, bClamp=true, Value=@n6.Result.Z)
   > &RigUnit_ModifyTransforms(
    ItemToModify=((Item=(Type=Bone),Transform=(Rotation=(X=0,Y=0,Z=0,W=1),Translation=(X=0,Y=0,Z=0),Scale3D=(X=1,Y=1,Z=1))),(Item=(Type=Bone,Name="deform_elbow_bk_r"),Transform=(Rotation=(X=0,Y=0,Z=0,W=1),Translation=(X=5,Y=1,Z=0),Scale3D=(X=1,Y=1,Z=1))))
    Mode=OverrideLocal
    Weight=@n10.Result
    ItemToModify.1.Transform.Rotation=@n8.Transform.Rotation
   )
   > &RigUnit_ModifyTransforms(
    ItemToModify=((Item=(Type=Bone),Transform=(Rotation=(X=0,Y=0,Z=0,W=1),Translation=(X=0,Y=0,Z=0),Scale3D=(X=1,Y=1,Z=1))),(Item=(Type=Bone,Name="deform_elbow_fr_r"),Transform=(Rotation=(X=0,Y=0,Z=0,W=1),Translation=(X=-15,Y=-2.176,Z=0),Scale3D=(X=1,Y=1,Z=1.337))))
    Mode=OverrideLocal
    Weight=@n10.Result
    ItemToModify.1.Transform.Rotation=&RigVMFunction_MathQuaternionFromEuler(Euler=&RigVMFunction_MathVectorAdd(B=(X=0,Y=0,Z=-50), A=@n9.Result) {template="Add::Execute(in A,in B,out Result)"}.Result).Result
   )
   n11 = &RigVMFunction_MathQuaternionToEuler(
    RotationOrder=ZYX
    Value=&RigUnit_GetRelativeTransformForItem(Child=(Type=Bone,Name="calf_l"), Parent=(Type=Bone,Name="thigh_l")).RelativeTransform.Rotation
   )
   n12 = &RigVMFunction_MathFloatRemap(SourceMaximum=90, SourceMinimum=4, TargetMinimum=0, bClamp=true, Value=@n11.Result.Z)
   n13 = &RigVMFunction_MathQuaternionToEuler(
    RotationOrder=ZYX
    Value=&RigUnit_GetTransform(Item=(Type=Bone,Name="deform_knee_bk_01_l"), Space=LocalSpace, bInitial=true).Transform.Rotation
   )
   > &RigUnit_ModifyTransforms(
    ItemToModify=((Item=(Type=Bone),Transform=(Rotation=(X=0,Y=0,Z=0,W=1),Translation=(X=0,Y=0,Z=0),Scale3D=(X=1,Y=1,Z=1))),(Item=(Type=Bone,Name="deform_knee_bk_01_l"),Transform=(Rotation=(X=0,Y=0,Z=0,W=1),Translation=(X=-9.801,Y=-12.293,Z=0),Scale3D=(X=1,Y=1,Z=1.5))))
    Mode=OverrideLocal
    Weight=@n12.Result
    ItemToModify.1.Transform.Rotation=&RigVMFunction_MathQuaternionFromEuler(Euler=&RigVMFunction_MathVectorAdd(B=(X=0,Y=0,Z=3.141), A=@n13.Result) {template="Add::Execute(in A,in B,out Result)"}.Result).Result
   )
   n14 = &RigUnit_GetTransform(Item=(Type=Bone,Name="deform_knee_fr_01_l"), Space=LocalSpace, bInitial=true)
   > &RigUnit_ModifyTransforms(
    ItemToModify=((Item=(Type=Bone),Transform=(Rotation=(X=0,Y=0,Z=0,W=1),Translation=(X=0,Y=0,Z=0),Scale3D=(X=1,Y=1,Z=1))),(Item=(Type=Bone,Name="deform_knee_fr_01_l"),Transform=(Rotation=(X=0,Y=0,Z=0,W=1),Translation=(X=1.196,Y=1.065,Z=0),Scale3D=(X=1,Y=1,Z=1))))
    Mode=OverrideLocal
    Weight=@n12.Result
    ItemToModify.1.Transform.Rotation=@n14.Transform.Rotation
   )
   n15 = &RigVMFunction_MathFloatRemap(SourceMaximum=130, SourceMinimum=89, TargetMinimum=0, bClamp=true, Value=@n11.Result.Z)
   > &RigUnit_ModifyTransforms(
    ItemToModify=((Item=(Type=Bone),Transform=(Rotation=(X=0,Y=0,Z=0,W=1),Translation=(X=0,Y=0,Z=0),Scale3D=(X=1,Y=1,Z=1))),(Item=(Type=Bone,Name="deform_knee_bk_01_l"),Transform=(Rotation=(X=0,Y=0,Z=0,W=1),Translation=(X=-11,Y=-14,Z=0),Scale3D=(X=1,Y=1,Z=1.6))))
    Mode=OverrideLocal
    Weight=@n15.Result
    ItemToModify.1.Transform.Rotation=&RigVMFunction_MathQuaternionFromEuler(Euler=&RigVMFunction_MathVectorAdd(B=(X=0,Y=0,Z=5), A=@n13.Result) {template="Add::Execute(in A,in B,out Result)"}.Result).Result
   )
   > &RigUnit_ModifyTransforms(
    ItemToModify=((Item=(Type=Bone),Transform=(Rotation=(X=0,Y=0,Z=0,W=1),Translation=(X=0,Y=0,Z=0),Scale3D=(X=1,Y=1,Z=1))),(Item=(Type=Bone,Name="deform_knee_fr_01_l"),Transform=(Rotation=(X=0,Y=0,Z=0,W=1),Translation=(X=1.3,Y=1.5,Z=0),Scale3D=(X=1,Y=1,Z=1))))
    Mode=OverrideLocal
    Weight=@n15.Result
    ItemToModify.1.Transform.Rotation=@n14.Transform.Rotation
   )
   n16 = &RigVMFunction_MathQuaternionToEuler(
    RotationOrder=ZYX
    Value=&RigUnit_GetRelativeTransformForItem(Child=(Type=Bone,Name="calf_r"), Parent=(Type=Bone,Name="thigh_r")).RelativeTransform.Rotation
   )
   n17 = &RigVMFunction_MathFloatRemap(SourceMaximum=90, SourceMinimum=4, TargetMinimum=0, bClamp=true, Value=@n16.Result.Z)
   n18 = &RigVMFunction_MathQuaternionToEuler(
    RotationOrder=ZYX
    Value=&RigUnit_GetTransform(Item=(Type=Bone,Name="deform_knee_bk_01_r"), Space=LocalSpace, bInitial=true).Transform.Rotation
   )
   > &RigUnit_ModifyTransforms(
    ItemToModify=((Item=(Type=Bone),Transform=(Rotation=(X=0,Y=0,Z=0,W=1),Translation=(X=0,Y=0,Z=0),Scale3D=(X=1,Y=1,Z=1))),(Item=(Type=Bone,Name="deform_knee_bk_01_r"),Transform=(Rotation=(X=0,Y=0,Z=0,W=1),Translation=(X=9.801,Y=12.293,Z=0),Scale3D=(X=1,Y=1,Z=1.5))))
    Mode=OverrideLocal
    Weight=@n17.Result
    ItemToModify.1.Transform.Rotation=&RigVMFunction_MathQuaternionFromEuler(Euler=&RigVMFunction_MathVectorAdd(B=(X=0,Y=0,Z=3.141), A=@n18.Result) {template="Add::Execute(in A,in B,out Result)"}.Result).Result
   )
   n19 = &RigUnit_GetTransform(Item=(Type=Bone,Name="deform_knee_fr_01_r"), Space=LocalSpace, bInitial=true)
   > &RigUnit_ModifyTransforms(
    ItemToModify=((Item=(Type=Bone),Transform=(Rotation=(X=0,Y=0,Z=0,W=1),Translation=(X=0,Y=0,Z=0),Scale3D=(X=1,Y=1,Z=1))),(Item=(Type=Bone,Name="deform_knee_fr_01_r"),Transform=(Rotation=(X=0,Y=0,Z=0,W=1),Translation=(X=-1.196,Y=-1.065,Z=0),Scale3D=(X=1,Y=1,Z=1))))
    Mode=OverrideLocal
    Weight=@n17.Result
    ItemToModify.1.Transform.Rotation=@n19.Transform.Rotation
   )
   n20 = &RigVMFunction_MathFloatRemap(SourceMaximum=130, SourceMinimum=89, TargetMinimum=0, bClamp=true, Value=@n16.Result.Z)
   > &RigUnit_ModifyTransforms(
    ItemToModify=((Item=(Type=Bone),Transform=(Rotation=(X=0,Y=0,Z=0,W=1),Translation=(X=0,Y=0,Z=0),Scale3D=(X=1,Y=1,Z=1))),(Item=(Type=Bone,Name="deform_knee_bk_01_r"),Transform=(Rotation=(X=0,Y=0,Z=0,W=1),Translation=(X=11,Y=14,Z=0),Scale3D=(X=1,Y=1,Z=1.6))))
    Mode=OverrideLocal
    Weight=@n20.Result
    ItemToModify.1.Transform.Rotation=&RigVMFunction_MathQuaternionFromEuler(Euler=&RigVMFunction_MathVectorAdd(B=(X=0,Y=0,Z=5), A=@n18.Result) {template="Add::Execute(in A,in B,out Result)"}.Result).Result
   )
   > &RigUnit_ModifyTransforms(
    ItemToModify=((Item=(Type=Bone),Transform=(Rotation=(X=0,Y=0,Z=0,W=1),Translation=(X=0,Y=0,Z=0),Scale3D=(X=1,Y=1,Z=1))),(Item=(Type=Bone,Name="deform_knee_fr_01_r"),Transform=(Rotation=(X=0,Y=0,Z=0,W=1),Translation=(X=-1.3,Y=-1.5,Z=0),Scale3D=(X=1,Y=1,Z=1))))
    Mode=OverrideLocal
    Weight=@n20.Result
    ItemToModify.1.Transform.Rotation=@n19.Transform.Rotation
   )
   n21 = &RigUnit_GetRelativeTransformForItem(Child=(Type=Bone,Name="upperarm_l"), Parent=(Type=Bone,Name="spine_05"))
   n22 = &RigVMFunction_MathFloatRemap(
    SourceMaximum=50
    SourceMinimum=170
    TargetMinimum=0
    bClamp=true
    Value=/Script/RigVM.RigVMFunction_MathFloatAbs(Value=&RigVMFunction_MathQuaternionToEuler(RotationOrder=ZYX, Value=@n21.RelativeTransform.Rotation).Result.Z) {is=struct RigVMUnitNode, template="Absolute::Execute(in Value,out Result)"}.Result
   )
   > &RigUnit_ModifyTransforms(
    ItemToModify=((Item=(Type=Bone,Name="deform_pectoral_fr_l"),Transform=(Rotation=(X=0,Y=0,Z=0,W=1),Translation=(X=4.877,Y=2.398,Z=1.048),Scale3D=(X=1,Y=1,Z=1))))
    Mode=OverrideLocal
    Weight=@n22.Result
    ItemToModify.0.Transform.Rotation=&RigUnit_GetTransform(Item=(Type=Bone,Name="deform_pectoral_fr_l"), Space=LocalSpace, bInitial=true).Transform.Rotation
   )
   > &RigUnit_ModifyTransforms(
    ItemToModify=((Item=(Type=Bone,Name="deform_deltoid_fr_l"),Transform=(Rotation=(X=0,Y=0,Z=0,W=1),Translation=(X=-4.337,Y=2.052,Z=0.026),Scale3D=(X=1,Y=1,Z=1))))
    Mode=OverrideLocal
    Weight=@n22.Result
    ItemToModify.0.Transform.Rotation=&RigUnit_GetTransform(Item=(Type=Bone,Name="deform_deltoid_fr_l"), Space=LocalSpace, bInitial=true).Transform.Rotation
   )
   n23 = &RigUnit_GetRelativeTransformForItem(Child=(Type=Bone,Name="upperarm_r"), Parent=(Type=Bone,Name="spine_05"))
   n24 = &RigVMFunction_MathFloatRemap(
    SourceMaximum=123
    SourceMinimum=30
    TargetMinimum=0
    bClamp=true
    Value=&RigVMFunction_MathFloatAbs(Value=&RigVMFunction_MathQuaternionToEuler(RotationOrder=ZYX, Value=@n23.RelativeTransform.Rotation).Result.Z) {template="Absolute::Execute(in Value,out Result)"}.Result
   )
   n25 = &RigVMFunction_MathQuaternionToEuler(
    RotationOrder=ZYX
    Value=&RigUnit_GetTransform(Item=(Type=Bone,Name="deform_pectoral_fr_r"), Space=LocalSpace, bInitial=true).Transform.Rotation
   )
   > &RigUnit_ModifyTransforms(
    ItemToModify=((Item=(Type=Bone,Name="deform_pectoral_fr_r"),Transform=(Rotation=(X=0,Y=0,Z=0,W=1),Translation=(X=-5.561,Y=-1.107,Z=-5.161),Scale3D=(X=1,Y=1,Z=1))))
    Mode=OverrideLocal
    Weight=@n24.Result
    ItemToModify.0.Transform.Rotation=&RigVMFunction_MathQuaternionFromEuler(Euler=&RigVMFunction_MathVectorAdd(B=(X=0,Y=-17.297,Z=0), A=@n25.Result) {template="Add::Execute(in A,in B,out Result)"}.Result).Result
   )
   n26 = &RigVMFunction_MathQuaternionToEuler(
    RotationOrder=ZYX
    Value=&RigUnit_GetTransform(Item=(Type=Bone,Name="deform_deltoid_fr_r"), Space=LocalSpace, bInitial=true).Transform.Rotation
   )
   > &RigUnit_ModifyTransforms(
    ItemToModify=((Item=(Type=Bone,Name="deform_deltoid_fr_r"),Transform=(Rotation=(X=0,Y=0,Z=0,W=1),Translation=(X=-0.063,Y=-3.462,Z=0.39),Scale3D=(X=1,Y=1,Z=1))))
    Mode=OverrideLocal
    Weight=@n24.Result
    ItemToModify.0.Transform.Rotation=&RigVMFunction_MathQuaternionFromEuler(Euler=&RigVMFunction_MathVectorAdd(B=(X=0,Y=28.666,Z=0), A=@n26.Result) {template="Add::Execute(in A,in B,out Result)"}.Result).Result
   )
   n27 = &RigUnit_GetRelativeTransformForItem(Child=(Type=Bone,Name="thigh_l"), Parent=(Type=Bone,Name="pelvis"))
   n28 = &RigVMFunction_MathFloatRemap(
    SourceMaximum=100
    SourceMinimum=2
    TargetMinimum=0
    bClamp=true
    Value=&RigVMFunction_MathFloatAbs(Value=&RigVMFunction_MathQuaternionToEuler(RotationOrder=ZYX, Value=@n27.RelativeTransform.Rotation).Result.Z) {template="Absolute::Execute(in Value,out Result)"}.Result
   )
   n29 = &RigVMFunction_MathQuaternionToEuler(
    RotationOrder=ZYX
    Value=&RigUnit_GetTransform(Item=(Type=Bone,Name="deform_thigh_fr_l"), Space=LocalSpace, bInitial=true).Transform.Rotation
   )
   > &RigUnit_ModifyTransforms(
    ItemToModify=((Item=(Type=Bone,Name="deform_thigh_fr_l"),Transform=(Rotation=(X=0,Y=0,Z=0,W=1),Translation=(X=18,Y=12,Z=0),Scale3D=(X=1,Y=1,Z=1))))
    Mode=OverrideLocal
    Weight=@n28.Result
    ItemToModify.0.Transform.Rotation=&RigVMFunction_MathQuaternionFromEuler(Euler=&RigVMFunction_MathVectorAdd(B=(X=0,Y=0,Z=45), A=@n29.Result) {template="Add::Execute(in A,in B,out Result)"}.Result).Result
   )
   n30 = &RigVMFunction_MathQuaternionToEuler(
    RotationOrder=ZYX
    Value=&RigUnit_GetTransform(Item=(Type=Bone,Name="deform_thigh_bk_l"), Space=LocalSpace, bInitial=true).Transform.Rotation
   )
   > &RigUnit_ModifyTransforms(
    ItemToModify=((Item=(Type=Bone,Name="deform_thigh_bk_l"),Transform=(Rotation=(X=0,Y=0,Z=0,W=1),Translation=(X=25,Y=-5,Z=8),Scale3D=(X=1,Y=1,Z=1))))
    Mode=OverrideLocal
    Weight=@n28.Result
    ItemToModify.0.Transform.Rotation=&RigVMFunction_MathQuaternionFromEuler(Euler=&RigVMFunction_MathVectorAdd(B=(X=0,Y=0,Z=23.3), A=@n30.Result) {template="Add::Execute(in A,in B,out Result)"}.Result).Result
   )
   n31 = &RigUnit_GetRelativeTransformForItem(Child=(Type=Bone,Name="thigh_r"), Parent=(Type=Bone,Name="pelvis"))
   n32 = &RigVMFunction_MathFloatRemap(
    SourceMaximum=77
    SourceMinimum=177
    TargetMinimum=0
    bClamp=true
    Value=&RigVMFunction_MathFloatAbs(Value=&RigVMFunction_MathQuaternionToEuler(RotationOrder=ZYX, Value=@n31.RelativeTransform.Rotation).Result.Z) {template="Absolute::Execute(in Value,out Result)"}.Result
   )
   n33 = &RigVMFunction_MathQuaternionToEuler(
    RotationOrder=ZYX
    Value=&RigUnit_GetTransform(Item=(Type=Bone,Name="deform_thigh_fr_r"), Space=LocalSpace, bInitial=true).Transform.Rotation
   )
   > &RigUnit_ModifyTransforms(
    ItemToModify=((Item=(Type=Bone,Name="deform_thigh_fr_r"),Transform=(Rotation=(X=0,Y=0,Z=0,W=1),Translation=(X=-18,Y=-12,Z=0),Scale3D=(X=1,Y=1,Z=1))))
    Mode=OverrideLocal
    Weight=@n32.Result
    ItemToModify.0.Transform.Rotation=&RigVMFunction_MathQuaternionFromEuler(Euler=&RigVMFunction_MathVectorAdd(B=(X=0,Y=0,Z=45), A=@n33.Result) {template="Add::Execute(in A,in B,out Result)"}.Result).Result
   )
   n34 = &RigVMFunction_MathQuaternionToEuler(
    RotationOrder=ZYX
    Value=&RigUnit_GetTransform(Item=(Type=Bone,Name="deform_thigh_bk_r"), Space=LocalSpace, bInitial=true).Transform.Rotation
   )
   > &RigUnit_ModifyTransforms(
    ItemToModify=((Item=(Type=Bone,Name="deform_thigh_bk_r"),Transform=(Rotation=(X=0,Y=0,Z=0,W=1),Translation=(X=-30,Y=4.5,Z=0.327),Scale3D=(X=1,Y=1,Z=1))))
    Mode=OverrideLocal
    Weight=@n32.Result
    ItemToModify.0.Transform.Rotation=&RigVMFunction_MathQuaternionFromEuler(Euler=&RigVMFunction_MathVectorAdd(B=(X=0,Y=0,Z=24), A=@n34.Result) {template="Add::Execute(in A,in B,out Result)"}.Result).Result
   )
   > /Script/ControlRig.RigUnit_DebugHierarchy(
    Color=(R=1,G=1,B=1,A=1)
    DebugDrawSettings=(DepthPriority=SDPG_Foreground,Lifetime=-1)
    Items=()
    Scale=3
    Thickness=0.1
    WorldOffset=(Rotation=(X=0,Y=0,Z=0,W=1),Translation=(X=0,Y=0,Z=0),Scale3D=(X=1,Y=1,Z=1))
    bEnabled=false
   ) {is=struct RigVMUnitNode}
   DeltaRotation v4 = RigVMCollapseNode(
    Rotation Order=ZXY
    child=(Type=Bone,Name="thigh_r")
    debugRotation=false
    outputMinMax=(X=0,Y=1)
    parent=(Type=Null,Name="thigh_r_null")
    premult=1
    rotationMinMax=(X=0,Y=-80)
    triggerAxis=Y
    twistAxis=(X=1,Y=0,Z=0)
    ExecuteContext=@n1.False
   )
   > RigVMVariableNode(Value=@"DeltaRotation v4".outputWeight) {variable=R Leg Out}
   > &RigUnit_ModifyTransforms(
    ItemToModify=((Item=(Type=Bone,Name="dyn_scarf_fr_01_MSTR"),Transform=(Rotation=(X=0.0531,Y=-0.6064,Z=0.0691,W=0.7903),Translation=(X=0,Y=0,Z=0),Scale3D=(X=1,Y=1,Z=1))))
    WeightMinimum=0
    Weight=RigVMVariableNode() {variable=R Leg Out}.Value
   )
   > DeltaRotation v4_2 = RigVMCollapseNode(
    Rotation Order=ZXY
    child=(Type=Bone,Name="thigh_r")
    debugRotation=false
    outputMinMax=(X=0,Y=1)
    parent=(Type=Null,Name="thigh_r_null")
    premult=1
    rotationMinMax=(X=0,Y=-100)
    triggerAxis=Z
    twistAxis=(X=1,Y=0,Z=0)
   )
   > RigVMVariableNode(Value=@"DeltaRotation v4_2".outputWeight) {variable=R Leg Front}
   > &RigUnit_ModifyTransforms(
    ItemToModify=((Item=(Type=Bone,Name="dyn_scarf_medallion_MSTR"),Transform=(Rotation=(X=-0.1674,Y=-0.075,Z=0.0751,W=0.9802),Translation=(X=0,Y=0,Z=1.5),Scale3D=(X=1,Y=1,Z=1))))
    WeightMinimum=0
    Weight=RigVMVariableNode() {variable=R Leg Front}.Value
   )
   > DeltaRotation v4_3 = RigVMCollapseNode(
    Rotation Order=ZXY
    absoluteValue=false
    child=(Type=Bone,Name="thigh_r")
    debugRotation=false
    outputMinMax=(X=0,Y=1)
    parent=(Type=Null,Name="thigh_r_null")
    premult=1
    rotationMinMax=(X=0,Y=100)
    triggerAxis=Z
    twistAxis=(X=1,Y=0,Z=0)
   )
   > RigVMVariableNode(Value=@"DeltaRotation v4_3".outputWeight) {variable=R Leg Back}
   > &RigUnit_ModifyTransforms(
    ItemToModify=((Item=(Type=Bone,Name="dyn_scarf_medallion_MSTR"),Transform=(Rotation=(X=0.0262,Y=-0.0262,Z=0.0007,W=0.9993),Translation=(X=0,Y=0,Z=0.5),Scale3D=(X=1,Y=1,Z=1))))
    WeightMinimum=0
    Weight=RigVMVariableNode() {variable=R Leg Back}.Value
   )
   > DeltaRotation v4_5 = RigVMCollapseNode(
    Rotation Order=ZXY
    absoluteValue=false
    child=(Type=Bone,Name="calf_r")
    debugRotation=false
    outputMinMax=(X=0,Y=1)
    parent=(Type=Null,Name="calf_r_null")
    premult=1
    rotationMinMax=(X=0,Y=50)
    triggerAxis=Z
    twistAxis=(X=1,Y=0,Z=0)
   )
   > RigVMVariableNode(Value=@"DeltaRotation v4_5".outputWeight) {variable=R Calf Bend}
   > DeltaRotation v4_4 = RigVMCollapseNode(
    Rotation Order=ZXY
    absoluteValue=false
    child=(Type=Bone,Name="thigh_r")
    debugRotation=false
    onlyTwist=true
    outputMinMax=(X=0,Y=1)
    parent=(Type=Null,Name="thigh_r_null")
    premult=1
    rotationMinMax=(X=0,Y=30)
    triggerAxis=X
    twistAxis=(X=1,Y=0,Z=0)
   )
   > RigVMVariableNode(Value=@"DeltaRotation v4_4".outputWeight) {variable=R Leg Twist}
   RigVMFunction_MathBoolToFloat = &RigVMFunction_MathBoolToFloat(Value=RigVMVariableNode() {variable=baseBallSliding}.Value)
   AlphaInterp = &RigVMFunction_AlphaInterp(InterpSpeedDecreasing=5, InterpSpeedIncreasing=5, Scale=1, bClampResult=False, Value=@RigVMFunction_MathBoolToFloat.Result)
   > &RigUnit_ModifyTransforms(
    ItemToModify="((Item=(Type=Bone,Name=\"dyn_scarf_fr_01\"),Transform=(Rotation=(X=-0.0176,Y=-0.7268,Z=-0.0441,W=0.6852),Translation=(X=0,Y=0,Z=0),Scale3D=(X=1,Y=1,Z=1))),(Item=(Type=Bone,Name=\"dyn_scarf_fr_03\"),Transform=(Rotation=(X=0.0091,Y=-0.1041,Z=0.0867,W=0.9907),Tra ...(truncated)"
    WeightMinimum=0
    Weight=@AlphaInterp.Result
   )
   n35 = &RigVMFunction_AlphaInterp(
    InterpSpeedDecreasing=5
    InterpSpeedIncreasing=5
    Scale=1
    bClampResult=False
    Value=&RigVMFunction_MathBoolToFloat(Value=RigVMVariableNode() {variable=isCrouching}.Value).Result
   )
   Subtract = &RigVMFunction_MathDoubleSub(A=1, B=@n35.Result)
   Multiply = RigVMAggregateNode(C=@Subtract.Result, A=RigVMVariableNode() {variable=R Calf Bend}.Value, B=RigVMVariableNode() {variable=R Leg Twist}.Value)
   > &RigUnit_ModifyTransforms(
    ItemToModify="((Item=(Type=Bone,Name=\"dyn_scarf_fr_01_MSTR\"),Transform=(Rotation=(X=-0.0057,Y=-0.087,Z=-0.0652,W=0.9941),Translation=(X=0,Y=0,Z=0),Scale3D=(X=1,Y=1,Z=1))),(Item=(Type=Bone,Name=\"dyn_scarf_bk_03\"),Transform=(Rotation=(X=-0.0064,Y=-0.1477,Z=-0.0431,W=0.988 ...(truncated)"
    WeightMinimum=0
    Weight=@Multiply.Result
   )
   n36 = &RigVMFunction_AlphaInterp(
    InterpSpeedDecreasing=5
    InterpSpeedIncreasing=10
    Scale=1
    bClampResult=False
    Value=&RigVMFunction_MathBoolToFloat(Value=RigVMVariableNode() {variable=isKneeSliding}.Value).Result
   )
   > &RigUnit_ModifyTransforms(
    ItemToModify="((Item=(Type=Bone,Name=\"dyn_scarf_fr_01\"),Transform=(Rotation=(X=0.067,Y=-0.25,Z=0.25,W=0.933),Translation=(X=0,Y=0,Z=0),Scale3D=(X=1,Y=1,Z=1))),(Item=(Type=Bone,Name=\"dyn_scarf_fr_03\"),Transform=(Rotation=(X=0.0262,Y=-0.2996,Z=0.0831,W=0.9501),Translation ...(truncated)"
    WeightMinimum=0
    Weight=@n36.Result
   )
   AlterTransform = RigVMCollapseNode(
    ExecuteContext=()
    inheritsTransform=false
    transforms=((Item=(Type=Bone,Name="dyn_scarf_fr_01"),Transform=(Rotation=(X=0,Y=-0.1736,Z=0,W=0.9848),Translation=(X=0,Y=0,Z=4),Scale3D=(X=1,Y=1,Z=1))))
    weights=RigVMVariableNode() {variable=R Leg Out}.Value
   )
   &RigUnit_PrepareForExecution()
   RigVMVariableNode() {variable=KneeRot_fr_l_114deg}
   RigVMVariableNode() {variable=KneeTrans_fr_l_114deg}
   RigVMVariableNode() {variable=KneeScale_fr_l_114deg}
   RigVMVariableNode() {variable=KneeScale_fr_l_130deg}
   RigVMVariableNode() {variable=KneeRot_fr_l_130deg}
   RigVMVariableNode() {variable=KneeTrans_fr_l_130deg}
   addFKCtrl = RigVMFunctionReferenceNode(
    Custom_FKCtrl=()
    ExecuteContext=()
    Options_FKCtrl="(pelvis_1_#1=true,spine_3_#2=false,clavicle_5_#3=true,arm_7_#4=true,leg_9_#5=true,head_11_46DC15EF465CD297 ...(truncated)"
    Shape_Settings="(pelvis_2_#6=(bVisible=True,Name=\"Hexagon_Thin\",Color=(R=0.0117,G=0.8281,B=1,A=1),Transform=(Rotation=(X=0,Y=-0.7071,Z=0,W=0.7071),Translation=(X=2,Y=0,Z=0),Scale3D=(X=0.5,Y=0.5,Z=1))),spine_4_#7= ...(truncated)"
   )
   setFKCtrl = RigVMFunctionReferenceNode(ExecuteContext=(), setFK_Ctrl=false)
  ]}
]
hierarchy [
 Root {loc=<0 0 0>, rot=<0 0 0 1>}
 pelvis {parent=Root, loc=<0 -0.5301 98.0571>, rot=<0.0127 0.707 -0.0127 -0.707>}
 Spine_01 {parent=pelvis, loc=<6.1208 -0.9328 0>, rot=<0 0 0.034 -0.9994>}
 Spine_02 {parent=Spine_01, loc=<6.5996 -0.2095 0>, rot=<0 0 0.0238 -0.9997>}
 Spine_03 {parent=Spine_02, loc=<6.4914 -0.1237 0>, rot=<0 0 0.0149 -0.9999>}
 spine_04 {parent=Spine_03, loc=<6.9598 0.0794 0>, rot=<0 0 -0.0132 -0.9999>}
 spine_05 {parent=spine_04, loc=<6.9987 0 0>, rot=<0 0 -0.0015 -1>}
 clavicle_l {parent=spine_05, loc=<5.4831 -0.1053 -3.2946>, rot=<0.0465 -0.7289 -0.009 -0.6829>}
 upperarm_l {parent=clavicle_l, loc=<9.8658 -1.2136 -0.4105>, rot=<0.0164 -0.3245 -0.0269 -0.9454>}
 lowerarm_l {parent=upperarm_l, loc=<26.465 -0.8345 0.259>, rot=<0.0036 0.0161 -0.2298 -0.9731>}
 hand_l {parent=lowerarm_l, loc=<22.7355 -0.3203 -0.6105>, rot=<-0.6784 -0.0223 0.0735 -0.7307>}
 index_metacarpal_l {parent=hand_l, loc=<1.6535 0.0183 -2.96>, rot=<0.0158 -0.0124 -0.0173 -0.9996>}
 index_01_l {parent=index_metacarpal_l, loc=<8.3008 0.0357 -0.0524>, rot=<-0.0044 0.0485 0.0924 -0.9945>}
 index_02_l {parent=index_01_l, loc=<3.3587 -0.3003 0.0452>, rot=<-0.002 0.0027 0.3225 -0.9466>}
 index_03_l {parent=index_02_l, loc=<2.5578 -0.123 -0.0252>, rot=<-0.002 0.0067 0.2558 -0.9667>}
 middle_metacarpal_l {parent=hand_l, loc=<2.1856 -0.1725 -0.5984>, rot=<0.006 0.0168 -0.0329 -0.9993>}
 middle_01_l {parent=middle_metacarpal_l, loc=<7.3275 -0.2447 0.0129>, rot=<-0.0208 0.0042 0.0858 -0.9961>}
 middle_02_l {parent=middle_01_l, loc=<4.0123 0.1401 0.1175>, rot=<-0.0218 0.0258 0.3512 -0.9357>}
 middle_03_l {parent=middle_02_l, loc=<2.6077 0.0534 0.1373>, rot=<0.0082 0.0188 0.1708 -0.9851>}
 pinky_metacarpal_l {parent=hand_l, loc=<1.7616 -0.9753 3.3814>, rot=<-0.0397 0.0814 -0.069 -0.9935>}
 pinky_01_l {parent=pinky_metacarpal_l, loc=<7.1709 -0.4831 -0.1094>, rot=<-0.0654 -0.0035 0.2242 -0.9723>}
 pinky_02_l {parent=pinky_01_l, loc=<2.8274 0.0893 -0.1043>, rot=<0.0037 0.0011 0.2854 -0.9584>}
 pinky_03_l {parent=pinky_02_l, loc=<2.2046 0.0308 -0.0078>, rot=<-0.0002 0.0098 0.1445 -0.9895>}
 ring_metacarpal_l {parent=hand_l, loc=<2.2074 -0.3625 1.3638>, rot=<-0.0284 0.0722 -0.0458 -0.9959>}
 ring_01_l {parent=ring_metacarpal_l, loc=<7.0295 -0.6197 -0.1446>, rot=<-0.0489 -0.0252 0.1892 -0.9804>}
 ring_02_l {parent=ring_01_l, loc=<3.6174 0.4647 -0.1027>, rot=<0 0 0.2806 -0.9598>}
 ring_03_l {parent=ring_02_l, loc=<2.4152 0.1141 0.0199>, rot=<0 0 0.1189 -0.9929>}
 thumb_01_l {parent=hand_l, loc=<1.3118 -0.3191 -2.459>, rot=<0.6546 -0.4042 0.0664 -0.6354>}
 thumb_02_l {parent=thumb_01_l, loc=<5.9336 -0.0448 0.0401>, rot=<-0.0037 0.0213 0.298 -0.9543>}
 thumb_03_l {parent=thumb_02_l, loc=<2.2223 0.3682 -0.0907>, rot=<0.0118 -0.013 0.0066 -0.9998>}
 lowerarm_twist_01_l {parent=lowerarm_l, loc=<18.316 0 0>, rot=<0 0 0 1>}
 lowerarm_twist_02_l {parent=lowerarm_l, loc=<10.6707 0 0>, rot=<0 0 0 1>}
 deform_elbow_bk_l {parent=lowerarm_l, loc=<-0.583 -1.471 -0.561>, rot=<0 0 0 1>}
 upperarm_twist_01_l {parent=upperarm_l, loc=<7.2135 -0.1134 0>, rot=<0 0 0 -1>}
 deform_deltoid_fr_l {parent=upperarm_twist_01_l, loc=<-4.7624 0.1134 0>, rot=<0 0.0007 0.0191 -0.9998>}
 upperarm_twist_02_l {parent=upperarm_l, loc=<19.7998 -0.1134 0>, rot=<0 0 0 -1>}
 deform_elbow_fr_l {parent=upperarm_l, loc=<26.5048 0.1875 -0.5816>, rot=<0.0026 0.0163 -0.1662 -0.986>}
 deform_pectoral_fr_l {parent=clavicle_l, loc=<6.2796 0 0>, rot=<0 0 0 1>}
 clavicle_r {parent=spine_05, loc=<5.4831 -0.1053 3.2946>, rot=<0.7289 0.0465 -0.6829 0.009>}
 upperarm_r {parent=clavicle_r, loc=<-9.8658 1.2136 0.4104>, rot=<0.0164 -0.3245 -0.0269 -0.9454>}
 lowerarm_r {parent=upperarm_r, loc=<-26.4651 0.8345 -0.259>, rot=<0.0036 0.0161 -0.2298 -0.9731>}
 hand_r {parent=lowerarm_r, loc=<-22.7356 0.3203 0.6106>, rot=<-0.6784 -0.0223 0.0735 -0.7307>}
 index_metacarpal_r {parent=hand_r, loc=<-1.6534 -0.0181 2.9601>, rot=<0.0158 -0.0124 -0.0173 -0.9996>}
 index_01_r {parent=index_metacarpal_r, loc=<-8.3008 -0.0357 0.0524>, rot=<-0.0044 0.0485 0.0924 -0.9945>}
 index_02_r {parent=index_01_r, loc=<-3.3589 0.3003 -0.0453>, rot=<-0.002 0.0027 0.3225 -0.9466>}
 index_03_r {parent=index_02_r, loc=<-2.5578 0.123 0.0252>, rot=<-0.002 0.0067 0.2558 -0.9667>}
 middle_metacarpal_r {parent=hand_r, loc=<-2.1856 0.1726 0.5984>, rot=<0.006 0.0168 -0.0329 -0.9993>}
 middle_01_r {parent=middle_metacarpal_r, loc=<-7.3273 0.2447 -0.0129>, rot=<-0.0208 0.0042 0.0858 -0.9961>}
 middle_02_r {parent=middle_01_r, loc=<-4.0124 -0.14 -0.1175>, rot=<-0.0218 0.0258 0.3512 -0.9357>}
 middle_03_r {parent=middle_02_r, loc=<-2.6077 -0.0533 -0.1373>, rot=<0.0082 0.0188 0.1708 -0.9851>}
 pinky_metacarpal_r {parent=hand_r, loc=<-1.7617 0.9753 -3.3814>, rot=<-0.0397 0.0814 -0.069 -0.9935>}
 pinky_01_r {parent=pinky_metacarpal_r, loc=<-7.1707 0.4832 0.1094>, rot=<-0.0654 -0.0035 0.2242 -0.9723>}
 pinky_02_r {parent=pinky_01_r, loc=<-2.8274 -0.0892 0.1043>, rot=<0.0037 0.0011 0.2854 -0.9584>}
 pinky_03_r {parent=pinky_02_r, loc=<-2.2045 -0.031 0.0078>, rot=<-0.0002 0.0098 0.1445 -0.9895>}
 ring_metacarpal_r {parent=hand_r, loc=<-2.2073 0.3627 -1.3638>, rot=<-0.0284 0.0722 -0.0458 -0.9959>}
 ring_01_r {parent=ring_metacarpal_r, loc=<-7.0297 0.6197 0.1446>, rot=<-0.0489 -0.0252 0.1892 -0.9804>}
 ring_02_r {parent=ring_01_r, loc=<-3.6172 -0.4647 0.1027>, rot=<0 0 0.2806 -0.9598>}
 ring_03_r {parent=ring_02_r, loc=<-2.4152 -0.1141 -0.0199>, rot=<0 0 0.1189 -0.9929>}
 thumb_01_r {parent=hand_r, loc=<-1.3119 0.3192 2.459>, rot=<0.6546 -0.4042 0.0664 -0.6354>}
 thumb_02_r {parent=thumb_01_r, loc=<-5.9336 0.0448 -0.0401>, rot=<-0.0037 0.0213 0.298 -0.9543>}
 thumb_03_r {parent=thumb_02_r, loc=<-2.2223 -0.3682 0.0907>, rot=<0.0118 -0.013 0.0066 -0.9998>}
 lowerarm_twist_01_r {parent=lowerarm_r, loc=<-18.3161 0 0>, rot=<0 0 0 1>}
 lowerarm_twist_02_r {parent=lowerarm_r, loc=<-10.6707 0 0>, rot=<0 0 0 1>}
 deform_elbow_bk_r {parent=lowerarm_r, loc=<0.5829 1.4707 0.561>, rot=<0 0 0 1>}
 upperarm_twist_01_r {parent=upperarm_r, loc=<-7.2135 0.1134 0>, rot=<0 0 0 1>}
 deform_deltoid_fr_r {parent=upperarm_twist_01_r, loc=<4.7623 -0.1134 0>, rot=<0 0.0007 0.0191 -0.9998>}
 upperarm_twist_02_r {parent=upperarm_r, loc=<-19.7998 0.1134 0>, rot=<0 0 0 1>}
 deform_elbow_fr_r {parent=upperarm_r, loc=<-26.5046 -0.1875 0.5813>, rot=<0.0026 0.0163 -0.1662 -0.986>}
 deform_pectoral_fr_r {parent=clavicle_r, loc=<-6.28 0 0>, rot=<0 0 0 -1>}
 neck_01 {parent=spine_05, loc=<11.3138 0.1529 0>, rot=<0 0 -0.2003 -0.9797>}
 neck_02 {parent=neck_01, loc=<4.5627 0 0>, rot=<0 0 0.0504 -0.9987>}
 head {parent=neck_02, loc=<5.3914 0 0>, rot=<0 0 0.115 -0.9934>}
 thigh_l {parent=pelvis, loc=<-1.0855 -0.2109 -8.7053>, rot=<0.0009 0.0355 0.0155 -0.9993>}
 calf_l {parent=thigh_l, loc=<-42.5284 -0.0019 -0.7297>, rot=<0.0006 -0.0161 -0.0391 -0.9991>}
 foot_l {parent=calf_l, loc=<-46.7401 -2.5014 -0.592>, rot=<0.0427 -0.0238 0.0353 -0.9982>}
 ball_l {parent=foot_l, loc=<-3.5658 10.8066 -0.1544>, rot=<0.0001 -0.0001 -0.7063 -0.7079>}
 calf_twist_01_l {parent=calf_l, loc=<-35.8483 -1.9197 -0.5794>, rot=<0.0237 0.0095 -0.0035 -0.9997>}
 calf_twist_02_l {parent=calf_l, loc=<-19.1546 -1.7943 -0.2638>, rot=<0.0237 0.0095 -0.0035 -0.9997>}
 deform_knee_bk_01_l {parent=calf_l, loc=<0.6022 -1.0607 0>, rot=<0.0026 0.0216 -0.0407 -0.9989>}
 deform_knee_fr_01_l {parent=calf_l, loc=<-0.2342 0 0>, rot=<0.0017 0.0216 0.0001 -0.9998>}
 thigh_twist_01_l {parent=thigh_l, loc=<-27.952 0 0>, rot=<0.0013 0 0 -1>}
 deform_thigh_bk_l {parent=thigh_twist_01_l, loc=<23.2101 0.2058 0.1808>, rot=<-0.0013 0 0 -1>}
 deform_thigh_fr_l {parent=thigh_twist_01_l, loc=<27.7671 0.2281 -0.0598>, rot=<-0.0013 0 0 -1>}
 thigh_r {parent=pelvis, loc=<-1.0853 -0.2109 8.7053>, rot=<0.0355 -0.0009 0.9993 0.0155>}
 calf_r {parent=thigh_r, loc=<42.5286 0.0019 0.7297>, rot=<0.0006 -0.0161 -0.0391 -0.9991>}
 foot_r {parent=calf_r, loc=<46.7403 2.5014 0.592>, rot=<0.0427 -0.0238 0.0353 -0.9982>}
 ball_r {parent=foot_r, loc=<3.5658 -10.8066 0.1544>, rot=<0.0001 -0.0001 -0.7063 -0.7079>}
 calf_twist_01_r {parent=calf_r, loc=<35.8485 1.9197 0.5794>, rot=<0.0237 0.0095 -0.0035 -0.9997>}
 calf_twist_02_r {parent=calf_r, loc=<19.1546 1.7943 0.2638>, rot=<0.0237 0.0095 -0.0035 -0.9997>}
 deform_knee_fr_01_r {parent=calf_r, loc=<0.2341 0 0>, rot=<0.0017 0.0216 0.0001 -0.9998>}
 deform_knee_bk_01_r {parent=calf_r, loc=<-0.6022 1.0607 0>, rot=<0.0026 0.0216 -0.0407 -0.9989>}
 thigh_twist_01_r {parent=thigh_r, loc=<27.952 0 0>, rot=<0.0013 0 0 -1>}
 deform_thigh_bk_r {parent=thigh_twist_01_r, loc=<-23.21 -0.206 -0.181>, rot=<-0.0013 0 0 -1>}
 deform_thigh_fr_r {parent=thigh_twist_01_r, loc=<-27.767 -0.228 0.06>, rot=<-0.0013 0 0 -1>}
 calf_r_null {parent=thigh_r, type=Null, loc=<42.5286 0.0019 0.7297>, rot=<0.0006 -0.0161 -0.0391 -0.9991>}
 dyn_scarf_medallion_MSTR {parent=pelvis, loc=<7.5295 1.4567 11.0792>, rot=<0.1773 0.15 0.9726 0.0103>}
 dyn_scarf_medallion {parent=dyn_scarf_medallion_MSTR, loc=<0.3 0 0>, rot=<0 0 0 -1>}
 dyn_scarf_fr_01_MSTR {parent=dyn_scarf_medallion, loc=<4.3731 0.6911 -0.2539>, rot=<0.1227 0.0027 -0.0217 -0.9922>}
 dyn_scarf_fr_01 {parent=dyn_scarf_fr_01_MSTR, loc=<0.3 0 0>, rot=<0 0 0 1>}
 dyn_scarf_fr_02 {parent=dyn_scarf_fr_01, loc=<11.8119 0 0>, rot=<0 -0.0933 0 -0.9956>}
 dyn_scarf_fr_03 {parent=dyn_scarf_fr_02, loc=<12.9265 0 0>, rot=<0 -0.0619 0 -0.9981>}
 dyn_scarf_fr_04 {parent=dyn_scarf_fr_03, loc=<12.3457 0 0>, rot=<0 0 0 1>}
 dyn_scarf_fr_05 {parent=dyn_scarf_fr_04, loc=<11.8691 0 0>, rot=<0 0 0 1>}
 dyn_scarf_fr_06 {parent=dyn_scarf_fr_05, loc=<11.824 0 0>, rot=<0 0 0 1>}
 dyn_scarf_fr_07 {parent=dyn_scarf_fr_06, loc=<10.1117 0 0>, rot=<0 0 0 1>}
 dyn_scarf_bk_02 {parent=dyn_scarf_fr_01, loc=<12.0505 4.7701 0.1851>, rot=<0.1346 -0.0998 -0.1134 -0.9793>}
 dyn_scarf_bk_03 {parent=dyn_scarf_bk_02, loc=<12.9265 0 0>, rot=<0 -0.0619 0 -0.9981>}
 dyn_scarf_bk_04 {parent=dyn_scarf_bk_03, loc=<12.3457 0 0>, rot=<0 0 0 1>}
 dyn_scarf_bk_05 {parent=dyn_scarf_bk_04, loc=<11.8691 0 0>, rot=<0 0 0 1>}
 dyn_scarf_bk_06 {parent=dyn_scarf_bk_05, loc=<11.824 0 0>, rot=<0 0 0 1>}
 dyn_scarf_bk_07 {parent=dyn_scarf_bk_06, loc=<10.1117 0 0>, rot=<0 0 0 1>}
 thigh_r_null {parent=pelvis, type=Null, loc=<-1.0853 -0.2109 8.7053>, rot=<0.0355 -0.0009 0.9993 0.0155>}
 dyn_root_simspace {parent=Root, loc=<0 0 0>, rot=<0 0 0 1>}
 dyn_root_slope {parent=Root, loc=<0 0 0>, rot=<0 0 0 1>}
 r_knee_corrective {type=Curve, loc=<0 0 0>, rot=<0 0 0 1>}
 l_knee_corrective {type=Curve, loc=<0 0 0>, rot=<0 0 0 1>}
 Male_Med_Head_01_Pout {type=Curve, loc=<0 0 0>, rot=<0 0 0 1>}
 Male_Med_Head_02_Pout {type=Curve, loc=<0 0 0>, rot=<0 0 0 1>}
 Male_Med_Head_03_Pout {type=Curve, loc=<0 0 0>, rot=<0 0 0 1>}
 Male_Med_Head_05_Effort {type=Curve, loc=<0 0 0>, rot=<0 0 0 1>}
 Stretch {type=Curve, loc=<0 0 0>, rot=<0 0 0 1>}
 Squash {type=Curve, loc=<0 0 0>, rot=<0 0 0 1>}
 polySurface151 {type=Curve, loc=<0 0 0>, rot=<0 0 0 1>}
 l_cheek_puff {type=Curve, loc=<0 0 0>, rot=<0 0 0 1>}
 r_cheek_puff {type=Curve, loc=<0 0 0>, rot=<0 0 0 1>}
 l_smile_fix {type=Curve, loc=<0 0 0>, rot=<0 0 0 1>}
 r_smile_fix {type=Curve, loc=<0 0 0>, rot=<0 0 0 1>}
 jaw_open_fix {type=Curve, loc=<0 0 0>, rot=<0 0 0 1>}
 r_lip_sneer_fix {type=Curve, loc=<0 0 0>, rot=<0 0 0 1>}
 l_lip_sneer_fix {type=Curve, loc=<0 0 0>, rot=<0 0 0 1>}
 L_Blink {type=Curve, loc=<0 0 0>, rot=<0 0 0 1>}
 R_Blink {type=Curve, loc=<0 0 0>, rot=<0 0 0 1>}
 L_Arm_Stretch {type=Curve, loc=<0 0 0>, rot=<0 0 0 1>}
 Hair_Hat {type=Curve, loc=<0 0 0>, rot=<0 0 0 1>}
 Hair_Helmet {type=Curve, loc=<0 0 0>, rot=<0 0 0 1>}
 Male_Med_Head_02_What {type=Curve, loc=<0 0 0>, rot=<0 0 0 1>}
 Male_Med_Head_02_Smile {type=Curve, loc=<0 0 0>, rot=<0 0 0 1>}
 Male_Med_Head_02_Scowl {type=Curve, loc=<0 0 0>, rot=<0 0 0 1>}
 Male_Med_Head_02_Hmmph {type=Curve, loc=<0 0 0>, rot=<0 0 0 1>}
 Male_Med_Head_02_Grrr {type=Curve, loc=<0 0 0>, rot=<0 0 0 1>}
 Male_Med_Head_02_Effort {type=Curve, loc=<0 0 0>, rot=<0 0 0 1>}
 Male_Med_Head_03_What {type=Curve, loc=<0 0 0>, rot=<0 0 0 1>}
 Male_Med_Head_03_Grrr {type=Curve, loc=<0 0 0>, rot=<0 0 0 1>}
 Male_Med_Head_03_Effort {type=Curve, loc=<0 0 0>, rot=<0 0 0 1>}
 Male_Med_Head_03_Hmmph {type=Curve, loc=<0 0 0>, rot=<0 0 0 1>}
 Male_Med_Head_03_Scowl {type=Curve, loc=<0 0 0>, rot=<0 0 0 1>}
 Male_Med_Head_03_Smile {type=Curve, loc=<0 0 0>, rot=<0 0 0 1>}
 Male_Med_Head_05_Teeth_Low_What {type=Curve, loc=<0 0 0>, rot=<0 0 0 1>}
 Male_Med_Head_05_Teeth_Low_Pout {type=Curve, loc=<0 0 0>, rot=<0 0 0 1>}
 Male_Med_Head_05_Teeth_Low_Hmmph {type=Curve, loc=<0 0 0>, rot=<0 0 0 1>}
 Male_Med_Head_05_Teeth_Low_Grrr {type=Curve, loc=<0 0 0>, rot=<0 0 0 1>}
 Male_Med_Head_05_Teeth_Low_Effort {type=Curve, loc=<0 0 0>, rot=<0 0 0 1>}
 Male_Med_Head_05_What {type=Curve, loc=<0 0 0>, rot=<0 0 0 1>}
 Male_Med_Head_05_Grrr {type=Curve, loc=<0 0 0>, rot=<0 0 0 1>}
 Male_Med_Head_05_Hmmph {type=Curve, loc=<0 0 0>, rot=<0 0 0 1>}
 Male_Med_Head_05_Pout {type=Curve, loc=<0 0 0>, rot=<0 0 0 1>}
 Male_Med_Head_05_Tongue_What {type=Curve, loc=<0 0 0>, rot=<0 0 0 1>}
 Male_Med_Head_05_Tongue_Pout {type=Curve, loc=<0 0 0>, rot=<0 0 0 1>}
 Male_Med_Head_05_Tongue_Hmmph {type=Curve, loc=<0 0 0>, rot=<0 0 0 1>}
 Male_Med_Head_05_Tongue_Grrr {type=Curve, loc=<0 0 0>, rot=<0 0 0 1>}
 Male_Med_Head_05_Tongue_Effort {type=Curve, loc=<0 0 0>, rot=<0 0 0 1>}
 AdditiveLayer {type=Curve, loc=<0 0 0>, rot=<0 0 0 1>}
 RightHandIK {type=Curve, loc=<0 0 0>, rot=<0 0 0 1>}
 LeftHandIK {type=Curve, loc=<0 0 0>, rot=<0 0 0 1>}
 LeftHandFK {type=Curve, loc=<0 0 0>, rot=<0 0 0 1>}
 Male_Med_Head_01_Grrr {type=Curve, loc=<0 0 0>, rot=<0 0 0 1>}
 WeaponAiming {type=Curve, loc=<0 0 0>, rot=<0 0 0 1>}
 Male_Med_Head_01_Hmmph {type=Curve, loc=<0 0 0>, rot=<0 0 0 1>}
 DragonSlash {type=Curve, loc=<0 0 0>, rot=<0 0 0 1>}
 Male_Med_Head_01_Effort {type=Curve, loc=<0 0 0>, rot=<0 0 0 1>}
 LeftHankFK {type=Curve, loc=<0 0 0>, rot=<0 0 0 1>}
 blendOrient1 {type=Curve, loc=<0 0 0>, rot=<0 0 0 1>}
 blendParent1 {type=Curve, loc=<0 0 0>, rot=<0 0 0 1>}
 bottom_R_Corner_Down {type=Curve, loc=<0 0 0>, rot=<0 0 0 1>}
 bottom_R_Corner_Up {type=Curve, loc=<0 0 0>, rot=<0 0 0 1>}
 top_L_Corner_Up {type=Curve, loc=<0 0 0>, rot=<0 0 0 1>}
 top_L_Corner_Down {type=Curve, loc=<0 0 0>, rot=<0 0 0 1>}
 top_R_Corner_Down {type=Curve, loc=<0 0 0>, rot=<0 0 0 1>}
 top_R_Corner_Up {type=Curve, loc=<0 0 0>, rot=<0 0 0 1>}
 bottom_L_Corner_Down {type=Curve, loc=<0 0 0>, rot=<0 0 0 1>}
 bottomt_L_Corner_Up {type=Curve, loc=<0 0 0>, rot=<0 0 0 1>}
 TrailCurve1 {type=Curve, loc=<0 0 0>, rot=<0 0 0 1>}
 BothHandFK {type=Curve, loc=<0 0 0>, rot=<0 0 0 1>}
 bottom_R_Corner_Down1 {type=Curve, loc=<0 0 0>, rot=<0 0 0 1>}
 blendPoint1 {type=Curve, loc=<0 0 0>, rot=<0 0 0 1>}
 WeaponAimingFreeze {type=Curve, loc=<0 0 0>, rot=<0 0 0 1>}
 l_calf_top_taper {type=Curve, loc=<0 0 0>, rot=<0 0 0 1>}
 JogBlendSpaceRotation {type=Curve, loc=<0 0 0>, rot=<0 0 0 1>}
 ZeroOutPitchWeight {type=Curve, loc=<0 0 0>, rot=<0 0 0 1>}
 MeleeTwist {type=Curve, loc=<0 0 0>, rot=<0 0 0 1>}
 MeleeAnim {type=Curve, loc=<0 0 0>, rot=<0 0 0 1>}
 l_sneer_fix {type=Curve, loc=<0 0 0>, rot=<0 0 0 1>}
 jaw_fix {type=Curve, loc=<0 0 0>, rot=<0 0 0 1>}
 r_sneer_fix {type=Curve, loc=<0 0 0>, rot=<0 0 0 1>}
 TurnRotationAmount {type=Curve, loc=<0 0 0>, rot=<0 0 0 1>}
 TurnEnd {type=Curve, loc=<0 0 0>, rot=<0 0 0 1>}
 BlendIn {type=Curve, loc=<0 0 0>, rot=<0 0 0 1>}
 BlendOut {type=Curve, loc=<0 0 0>, rot=<0 0 0 1>}
 BodyRotateX {type=Curve, loc=<0 0 0>, rot=<0 0 0 1>}
 BodyTranslateX {type=Curve, loc=<0 0 0>, rot=<0 0 0 1>}
 CrescentKickScale {type=Curve, loc=<0 0 0>, rot=<0 0 0 1>}
 mask_geo {type=Curve, loc=<0 0 0>, rot=<0 0 0 1>}
 RemoveJumpAdditive {type=Curve, loc=<0 0 0>, rot=<0 0 0 1>}
 RightLeg_IK {type=Curve, loc=<0 0 0>, rot=<0 0 0 1>}
 SnapWeapon_LHand {type=Curve, loc=<0 0 0>, rot=<0 0 0 1>}
 PivotAnim {type=Curve, loc=<0 0 0>, rot=<0 0 0 1>}
 DefaultAdditiveStateWeight {type=Curve, loc=<0 0 0>, rot=<0 0 0 1>}
 StartAnim {type=Curve, loc=<0 0 0>, rot=<0 0 0 1>}
 StopAnim {type=Curve, loc=<0 0 0>, rot=<0 0 0 1>}
 DisableSpeedWarping {type=Curve, loc=<0 0 0>, rot=<0 0 0 1>}
 DisableDBNOSlopeWarp {type=Curve, loc=<0 0 0>, rot=<0 0 0 1>}
 MaskRightArm {type=Curve, loc=<0 0 0>, rot=<0 0 0 1>}
 TurnInPlaceAnim {type=Curve, loc=<0 0 0>, rot=<0 0 0 1>}
 Head_MESH {type=Curve, loc=<0 0 0>, rot=<0 0 0 1>}
 Hair_MESH {type=Curve, loc=<0 0 0>, rot=<0 0 0 1>}
 LaunchPadTransitionOff {type=Curve, loc=<0 0 0>, rot=<0 0 0 1>}
 C_glabella_up_pose {type=Curve, loc=<0 0 0>, rot=<0 0 0 1>}
 jaw_open_pose {type=Curve, loc=<0 0 0>, rot=<0 0 0 1>}
 L_blink_pose {type=Curve, loc=<0 0 0>, rot=<0 0 0 1>}
 L_brow_down_pose {type=Curve, loc=<0 0 0>, rot=<0 0 0 1>}
 L_brow_squeeze_in_pose {type=Curve, loc=<0 0 0>, rot=<0 0 0 1>}
 L_brow_up_pose {type=Curve, loc=<0 0 0>, rot=<0 0 0 1>}
 L_frown_pose {type=Curve, loc=<0 0 0>, rot=<0 0 0 1>}
 L_lip_corner_narrow_pose {type=Curve, loc=<0 0 0>, rot=<0 0 0 1>}
 L_squint_inner_pose {type=Curve, loc=<0 0 0>, rot=<0 0 0 1>}
 L_wide_pose {type=Curve, loc=<0 0 0>, rot=<0 0 0 1>}
 look_down_pose {type=Curve, loc=<0 0 0>, rot=<0 0 0 1>}
 look_left_pose {type=Curve, loc=<0 0 0>, rot=<0 0 0 1>}
 look_right_pose {type=Curve, loc=<0 0 0>, rot=<0 0 0 1>}
 look_up_pose {type=Curve, loc=<0 0 0>, rot=<0 0 0 1>}
 phoneme_fv_pose {type=Curve, loc=<0 0 0>, rot=<0 0 0 1>}
 phoneme_oo_pose {type=Curve, loc=<0 0 0>, rot=<0 0 0 1>}
 R_blink_pose {type=Curve, loc=<0 0 0>, rot=<0 0 0 1>}
 R_brow_down_pose {type=Curve, loc=<0 0 0>, rot=<0 0 0 1>}
 R_brow_squeeze_in_pose {type=Curve, loc=<0 0 0>, rot=<0 0 0 1>}
 R_brow_up_pose {type=Curve, loc=<0 0 0>, rot=<0 0 0 1>}
 R_frown_pose {type=Curve, loc=<0 0 0>, rot=<0 0 0 1>}
 R_lip_corner_narrow_pose {type=Curve, loc=<0 0 0>, rot=<0 0 0 1>}
 R_squint_inner_pose {type=Curve, loc=<0 0 0>, rot=<0 0 0 1>}
 R_wide_pose {type=Curve, loc=<0 0 0>, rot=<0 0 0 1>}
 DisableArmsHeadAdditive {type=Curve, loc=<0 0 0>, rot=<0 0 0 1>}
 FrontendRotation {type=Curve, loc=<0 0 0>, rot=<0 0 0 1>}
 BothLegFK {type=Curve, loc=<0 0 0>, rot=<0 0 0 1>}
 AllowTurnInPlace {type=Curve, loc=<0 0 0>, rot=<0 0 0 1>}
 MontageTimeStretchCurve {type=Curve, loc=<0 0 0>, rot=<0 0 0 1>}
 ParaGlideLeanAlpha {type=Curve, loc=<0 0 0>, rot=<0 0 0 1>}
 WeaponRInChestSpace {type=Curve, loc=<0 0 0>, rot=<0 0 0 1>}
 JimBeam_AO {type=Curve, loc=<0 0 0>, rot=<0 0 0 1>}
 R_squeeze_pose {type=Curve, loc=<0 0 0>, rot=<0 0 0 1>}
 L_squeeze_pose {type=Curve, loc=<0 0 0>, rot=<0 0 0 1>}
 R_upper_lip_raiser_pose {type=Curve, loc=<0 0 0>, rot=<0 0 0 1>}
 L_cheek_up_pose {type=Curve, loc=<0 0 0>, rot=<0 0 0 1>}
 R_cheek_up_pose {type=Curve, loc=<0 0 0>, rot=<0 0 0 1>}
 DisableFaceOverride {type=Curve, loc=<0 0 0>, rot=<0 0 0 1>}
 EnableFullBodyOverride {type=Curve, loc=<0 0 0>, rot=<0 0 0 1>}
 C_glabella_down_pose {type=Curve, loc=<0 0 0>, rot=<0 0 0 1>}
 C_nose_wrinkler_pose {type=Curve, loc=<0 0 0>, rot=<0 0 0 1>}
 eye_convergence_pose {type=Curve, loc=<0 0 0>, rot=<0 0 0 1>}
 jaw_back_pose {type=Curve, loc=<0 0 0>, rot=<0 0 0 1>}
 jaw_forward_pose {type=Curve, loc=<0 0 0>, rot=<0 0 0 1>}
 jaw_left_pose {type=Curve, loc=<0 0 0>, rot=<0 0 0 1>}
 jaw_right_pose {type=Curve, loc=<0 0 0>, rot=<0 0 0 1>}
 jaw_up_pose {type=Curve, loc=<0 0 0>, rot=<0 0 0 1>}
 L_lip_corner_wide_pose {type=Curve, loc=<0 0 0>, rot=<0 0 0 1>}
 L_lower_lip_down_pose {type=Curve, loc=<0 0 0>, rot=<0 0 0 1>}
 L_lower_lip_up_pose {type=Curve, loc=<0 0 0>, rot=<0 0 0 1>}
 L_smile_pose {type=Curve, loc=<0 0 0>, rot=<0 0 0 1>}
 L_upper_lip_lower_pose {type=Curve, loc=<0 0 0>, rot=<0 0 0 1>}
 L_upper_lip_raiser_pose {type=Curve, loc=<0 0 0>, rot=<0 0 0 1>}
 L_upperLidDown_pose {type=Curve, loc=<0 0 0>, rot=<0 0 0 1>}
 mouth_down_pose {type=Curve, loc=<0 0 0>, rot=<0 0 0 1>}
 mouth_up_pose {type=Curve, loc=<0 0 0>, rot=<0 0 0 1>}
 phoneme_ch_pose {type=Curve, loc=<0 0 0>, rot=<0 0 0 1>}
 phoneme_mbp_pose {type=Curve, loc=<0 0 0>, rot=<0 0 0 1>}
 R_lip_corner_wide_pose {type=Curve, loc=<0 0 0>, rot=<0 0 0 1>}
 R_lower_lip_down_pose {type=Curve, loc=<0 0 0>, rot=<0 0 0 1>}
 R_lower_lip_up_pose {type=Curve, loc=<0 0 0>, rot=<0 0 0 1>}
 R_smile_pose {type=Curve, loc=<0 0 0>, rot=<0 0 0 1>}
 R_upper_lip_lower_pose {type=Curve, loc=<0 0 0>, rot=<0 0 0 1>}
 R_upperLidDown_pose {type=Curve, loc=<0 0 0>, rot=<0 0 0 1>}
 tongue_back_pose {type=Curve, loc=<0 0 0>, rot=<0 0 0 1>}
 tongue_down_pose {type=Curve, loc=<0 0 0>, rot=<0 0 0 1>}
 tongue_forward_pose {type=Curve, loc=<0 0 0>, rot=<0 0 0 1>}
 tongue_up_pose {type=Curve, loc=<0 0 0>, rot=<0 0 0 1>}
 LiveLink {type=Curve, loc=<0 0 0>, rot=<0 0 0 1>}
 ParaGlideDeploy {type=Curve, loc=<0 0 0>, rot=<0 0 0 1>}
 HideLeftWeapon {type=Curve, loc=<0 0 0>, rot=<0 0 0 1>}
 DisableLeftHandIKGolfCart {type=Curve, loc=<0 0 0>, rot=<0 0 0 1>}
 DisableRightHandIKGolfCart {type=Curve, loc=<0 0 0>, rot=<0 0 0 1>}
 DisableHandIKRetargeting {type=Curve, loc=<0 0 0>, rot=<0 0 0 1>}
 DisableConsumableOffsetPose {type=Curve, loc=<0 0 0>, rot=<0 0 0 1>}
 UpperBodyNoAdditive {type=Curve, loc=<0 0 0>, rot=<0 0 0 1>}
 MaskLeftArm {type=Curve, loc=<0 0 0>, rot=<0 0 0 1>}
 ForceUpperBody {type=Curve, loc=<0 0 0>, rot=<0 0 0 1>}
 RightHandFK {type=Curve, loc=<0 0 0>, rot=<0 0 0 1>}
 ParaGlideLeanTransition {type=Curve, loc=<0 0 0>, rot=<0 0 0 1>}
 InMotionEmote {type=Curve, loc=<0 0 0>, rot=<0 0 0 1>}
 ForceFullBodyInFrontEnd {type=Curve, loc=<0 0 0>, rot=<0 0 0 1>}
 WeaponLinFloorSpace {type=Curve, loc=<0 0 0>, rot=<0 0 0 1>}
 FrontEnd_Idle_Bypass {type=Curve, loc=<0 0 0>, rot=<0 0 0 1>}
 DisableSlopeWarping {type=Curve, loc=<0 0 0>, rot=<0 0 0 1>}
 Disable_Custom {type=Curve, loc=<0 0 0>, rot=<0 0 0 1>}
 EmoteRequiresWeapon {type=Curve, loc=<0 0 0>, rot=<0 0 0 1>}
 DisableIKRootAdditive {type=Curve, loc=<0 0 0>, rot=<0 0 0 1>}
 DisableFullBodySprint {type=Curve, loc=<0 0 0>, rot=<0 0 0 1>}
 DetachFeet_Vehicle {type=Curve, loc=<0 0 0>, rot=<0 0 0 1>}
 IndigoBoost {type=Curve, loc=<0 0 0>, rot=<0 0 0 1>}
 DisableUpperBodyAdditiveMeshSpace {type=Curve, loc=<0 0 0>, rot=<0 0 0 1>}
 MeleeAOWeight {type=Curve, loc=<0 0 0>, rot=<0 0 0 1>}
 HandsInRootSpace {type=Curve, loc=<0 0 0>, rot=<0 0 0 1>}
 CrouchPelvisAdjustment {type=Curve, loc=<0 0 0>, rot=<0 0 0 1>}
 EnableAccelYawCorrection {type=Curve, loc=<0 0 0>, rot=<0 0 0 1>}
 ParaGlideModifyRoot {type=Curve, loc=<0 0 0>, rot=<0 0 0 1>}
 UseScalarParameterForBlink {type=Curve, loc=<0 0 0>, rot=<0 0 0 1>}
 BlinkControl {type=Curve, loc=<0 0 0>, rot=<0 0 0 1>}
 ForceTargeting {type=Curve, loc=<0 0 0>, rot=<0 0 0 1>}
 BlinkControlProp {type=Curve, loc=<0 0 0>, rot=<0 0 0 1>}
 UseScalarParameterForBlinkProp {type=Curve, loc=<0 0 0>, rot=<0 0 0 1>}
 CrackshotCozyHeadCorrection {type=Curve, loc=<0 0 0>, rot=<0 0 0 1>}
 ForwardMotion {type=Curve, loc=<0 0 0>, rot=<0 0 0 1>}
 EmoteForcedRootYaw {type=Curve, loc=<0 0 0>, rot=<0 0 0 1>}
 MeleeTrailWidth {type=Curve, loc=<0 0 0>, rot=<0 0 0 1>}
 space_world {type=Curve, loc=<0 0 0>, rot=<0 0 0 1>}
 PeelyBlenderFace {type=Curve, loc=<0 0 0>, rot=<0 0 0 1>}
 Width1 {type=Curve, loc=<0 0 0>, rot=<0 0 0 1>}
 OverrideIKHands {type=Curve, loc=<0 0 0>, rot=<0 0 0 1>}
 ParaGlideAddAlpha {type=Curve, loc=<0 0 0>, rot=<0 0 0 1>}
 ParaGlideTurn {type=Curve, loc=<0 0 0>, rot=<0 0 0 1>}
 DisablePelvisAdjustment {type=Curve, loc=<0 0 0>, rot=<0 0 0 1>}
 UncleBrollyDeploy {type=Curve, loc=<0 0 0>, rot=<0 0 0 1>}
 ForceRootSlopeOrientation {type=Curve, loc=<0 0 0>, rot=<0 0 0 1>}
 DisableAdditiveOffsetPose {type=Curve, loc=<0 0 0>, rot=<0 0 0 1>}
 DisableIdleNoiseAdditive {type=Curve, loc=<0 0 0>, rot=<0 0 0 1>}
 UsingConsumable {type=Curve, loc=<0 0 0>, rot=<0 0 0 1>}
 DisableRightArmAdditive {type=Curve, loc=<0 0 0>, rot=<0 0 0 1>}
]
variables [
 KneeRot_fr_l_114deg {type=FVector, default=(X=0,Y=0,Z=0)}
 KneeRot_fr_l_130deg {type=FVector, default=(X=0,Y=0,Z=0)}
 KneeScale_fr_l_114deg {type=FVector, default=(X=0,Y=0,Z=0)}
 KneeScale_fr_l_130deg {type=FVector, default=(X=0,Y=0,Z=0)}
 KneeTrans_fr_l_114deg {type=FVector, default=(X=0,Y=0,Z=0)}
 KneeTrans_fr_l_130deg {type=FVector, default=(X=0,Y=0,Z=0)}
 R Calf Bend {type=double, default=0}
 R Leg Back {type=double, default=0}
 R Leg Front {type=double, default=0}
 R Leg Out {type=double, default=0}
 R Leg Twist {type=double, default=0}
 baseBallSliding {type=bool, default=False}
 isCrouching {type=bool, default=False}
 isKneeSliding {type=bool, default=False}
 isRBANEnabled {type=bool, default=False}
]
