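# Physics asset for the F_MED_Cherry female character body that drives dynamic scarf bones and their collision.
# Front and back capsule chains branch from dyn_scarf_fr_01 with angle-limited joints, anchored to the pelvis through free-constrained medallion and master bones.
# Pelvis, thigh, calf and foot capsules provide body shapes with free constraints.
# An unconstrained, huge offset sphere on dyn_root_slope forms a near-flat surface, apparently standing in for the ground.
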PhysicsAsset /BRCosmetics/Characters/Player/Female/Medium/Bodies/F_MED_Cherry/Meshes/F_MED_Cherry_Dyn
bodies {
 calf_l [{shapeType=Capsule, length=33.5909, radius=6.4528, center=<-19.5058 -2.3559 0.0561>, rotation=<rot 89.1498 68.8435 68.1813>}]
 calf_r [
  {shapeType=Capsule, length=13.3392, radius=6.4528, center=<9.6179 3.2716 0.0689>, rotation=<rot 86.2764 179.9998 166.1717>}
  {shapeType=Capsule, length=28.0267, radius=5.2653, center=<32.8177 2.2017 0.9265>, rotation=<rot 88.3292 -29.375 -23.5944>}
 ]
 dyn_root_slope [{shapeType=Sphere, radius=569, shapeName=dyn_root_slope_sphere, center=<0 0 -563.8668>}]
 dyn_scarf_bk_02 [{shapeType=Capsule, length=2.937, radius=6.2188, center=<5.9443 0.0001 4>, rotation=<rot 90 0 0>}]
 dyn_scarf_bk_03 [{shapeType=Capsule, length=5.0307, radius=6.4113, center=<5.9443 0.0002 4.3379>, rotation=<rot 86.9163 85.2543 0.0009>}]
 dyn_scarf_bk_04 [{shapeType=Capsule, length=3.8045, radius=7.5075, center=<5.9443 0.0001 4.8416>, rotation=<rot 90 0 0>}]
 dyn_scarf_bk_05 [{shapeType=Capsule, length=9.3476, radius=4.3631, center=<6.7727 0.8552 2.3635>, rotation=<rot 89.8165 -89.8187 -179.8278>}]
 dyn_scarf_bk_06 [{shapeType=Capsule, length=6.645, radius=5.9817, center=<4.8448 1.4791 3.4445>, rotation=<rot 88.0469 91.1784 179.9999>}]
 dyn_scarf_bk_07 [{shapeType=Capsule, length=12.375, radius=3.125, center=<7.2945 1.6145 0.3297>, rotation=<rot 83.9023 88.8153 -0.0001>}]
 dyn_scarf_fr_01 [{shapeType=Capsule, length=15.937, radius=18, center=<5.9443 0.0004 17.5>, rotation=<rot 90 0 0>}]
 dyn_scarf_fr_01_MSTR [{shapeType=Capsule, length=0.505, radius=0.505}]
 dyn_scarf_fr_02 [{shapeType=Capsule, length=8.9995, radius=7.125, center=<5.971 0.0003 5.8>, rotation=<rot 86.6948 0 0>}]
 dyn_scarf_fr_03 [{shapeType=Capsule, length=5.0307, radius=6.3954, center=<5.9443 0.0001 4.6272>, rotation=<rot 90 90 -0.4452>}]
 dyn_scarf_fr_04 [{shapeType=Capsule, length=3.8745, radius=7.4375, center=<5.9443 0.0001 4.9222>, rotation=<rot 90 0 0>}]
 dyn_scarf_fr_05 [{shapeType=Capsule, length=6.2175, radius=4.5175, center=<7.0524 0.8191 2.5419>, rotation=<rot 87.9454 -88.3603 -179.9995>}]
 dyn_scarf_fr_06 [{shapeType=Capsule, length=4.1865, radius=7.0168, center=<6.9161 -0.1285 4.5264>, rotation=<rot 86.0429 -90.6285 -0.0008>}]
 dyn_scarf_fr_07 [{shapeType=Capsule, length=14.1238, radius=2.773, center=<6.4694 0.7811 1.9456>, rotation=<rot 86.5433 -91.6929 -0.0003>}]
 dyn_scarf_medallion [{shapeType=Capsule, length=0.505, radius=0.505}]
 dyn_scarf_medallion_MSTR [{shapeType=Capsule, length=0.505, radius=0.505}]
 foot_l [{shapeType=Capsule, length=7.8783, radius=7.737, center=<-3.3674 6.5045 0>, rotation=<rot 90 26.565 99.7674>}]
 foot_r [{shapeType=Capsule, length=7.8783, radius=7.737, center=<3.3674 -6.5045 0>, rotation=<rot 90 26.565 99.7674>}]
 pelvis [{shapeType=Capsule, length=12.3792, radius=8.9706, center=<-4.7526 -3.3094 0>}]
 thigh_l [{shapeType=Capsule, length=29.2425, radius=7.2747, center=<-22.32 0.3795 0.0271>, rotation=<rot 88.4255 -179.9998 177.165>}]
 thigh_r [
  {shapeType=Capsule, length=13.808, radius=9.4841, center=<11.2077 0.0004 -2.0654>, rotation=<rot 83.1483 -179.9998 179.9998>}
  {shapeType=Capsule, length=15.2455, radius=8.2243, center=<29.8143 0.0015 -1.97>, rotation=<rot 82.6261 0.0005 0>}
 ]
}
constraints [
 {
  bone1Name=dyn_scarf_medallion
  bone2Name=dyn_scarf_medallion_MSTR
  swing1LimitDegrees=45
  swing1Motion=Free
  swing2LimitDegrees=45
  swing2Motion=Free
  twistLimitDegrees=45
  twistMotion=Free
 }
 {
  bone1Name=dyn_scarf_fr_01_MSTR
  bone2Name=dyn_scarf_medallion
  swing1LimitDegrees=45
  swing1Motion=Free
  swing2LimitDegrees=45
  swing2Motion=Free
  twistLimitDegrees=45
  twistMotion=Free
 }
 {
  bone1Name=dyn_scarf_fr_01
  bone2Name=dyn_scarf_fr_01_MSTR
  swing1LimitDegrees=25
  swing2LimitDegrees=25
  twistLimitDegrees=5
  twistMotion=Locked
 }
 {bone1Name=dyn_scarf_fr_02, bone2Name=dyn_scarf_fr_01, swing1LimitDegrees=45, swing2LimitDegrees=45, twistLimitDegrees=5}
 {bone1Name=dyn_scarf_fr_03, bone2Name=dyn_scarf_fr_02, swing1LimitDegrees=45, swing2LimitDegrees=45, twistLimitDegrees=5}
 {bone1Name=dyn_scarf_fr_04, bone2Name=dyn_scarf_fr_03, swing1LimitDegrees=45, swing2LimitDegrees=45, twistLimitDegrees=5}
 {bone1Name=dyn_scarf_fr_05, bone2Name=dyn_scarf_fr_04, swing1LimitDegrees=45, swing2LimitDegrees=45, twistLimitDegrees=5}
 {bone1Name=dyn_scarf_fr_06, bone2Name=dyn_scarf_fr_05, swing1LimitDegrees=45, swing2LimitDegrees=45, twistLimitDegrees=5}
 {bone1Name=dyn_scarf_fr_07, bone2Name=dyn_scarf_fr_06, swing1LimitDegrees=45, swing2LimitDegrees=45, twistLimitDegrees=5}
 {bone1Name=dyn_scarf_bk_02, bone2Name=dyn_scarf_fr_01, swing1LimitDegrees=45, swing2LimitDegrees=45, twistLimitDegrees=5}
 {bone1Name=dyn_scarf_bk_03, bone2Name=dyn_scarf_bk_02, swing1LimitDegrees=45, swing2LimitDegrees=45, twistLimitDegrees=5}
 {bone1Name=dyn_scarf_bk_04, bone2Name=dyn_scarf_bk_03, swing1LimitDegrees=45, swing2LimitDegrees=45, twistLimitDegrees=5}
 {bone1Name=dyn_scarf_bk_05, bone2Name=dyn_scarf_bk_04, swing1LimitDegrees=45, swing2LimitDegrees=45, twistLimitDegrees=5}
 {bone1Name=dyn_scarf_bk_06, bone2Name=dyn_scarf_bk_05, swing1LimitDegrees=45, swing2LimitDegrees=45, twistLimitDegrees=5}
 {bone1Name=dyn_scarf_bk_07, bone2Name=dyn_scarf_bk_06, swing1LimitDegrees=45, swing2LimitDegrees=45, twistLimitDegrees=5}
 {
  bone1Name=calf_l
  bone2Name=thigh_l
  swing1LimitDegrees=45
  swing1Motion=Free
  swing2LimitDegrees=45
  swing2Motion=Free
  twistLimitDegrees=45
  twistMotion=Free
 }
 {
  bone1Name=foot_l
  bone2Name=calf_l
  swing1LimitDegrees=45
  swing1Motion=Free
  swing2LimitDegrees=45
  swing2Motion=Free
  twistLimitDegrees=45
  twistMotion=Free
 }
 {
  bone1Name=calf_r
  bone2Name=thigh_r
  swing1LimitDegrees=45
  swing1Motion=Free
  swing2LimitDegrees=45
  swing2Motion=Free
  twistLimitDegrees=45
  twistMotion=Free
 }
 {
  bone1Name=foot_r
  bone2Name=calf_r
  swing1LimitDegrees=45
  swing1Motion=Free
  swing2LimitDegrees=45
  swing2Motion=Free
  twistLimitDegrees=45
  twistMotion=Free
 }
 {
  bone1Name=thigh_l
  bone2Name=pelvis
  swing1LimitDegrees=45
  swing1Motion=Free
  swing2LimitDegrees=45
  swing2Motion=Free
  twistLimitDegrees=45
  twistMotion=Free
 }
 {
  bone1Name=thigh_r
  bone2Name=pelvis
  swing1LimitDegrees=45
  swing1Motion=Free
  swing2LimitDegrees=45
  swing2Motion=Free
  twistLimitDegrees=45
  twistMotion=Free
 }
 {
  bone1Name=dyn_scarf_medallion_MSTR
  bone2Name=pelvis
  swing1LimitDegrees=45
  swing1Motion=Free
  swing2LimitDegrees=45
  swing2Motion=Free
  twistLimitDegrees=45
  twistMotion=Free
 }
]
