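# Post-copy AnimBP for the F_MED_Cherry female body, adding deformation and secondary physics to the parent mesh's pose.
# Copied pose runs through twist and sculpter Control Rigs, scarf-medallion offsets for crouch and frontend/skydiving, slope and sim-space root bones, then RigidBody simulation.
# The event graph feeds slope, smoothed sim-space rotation and wind-like gravity, using deprecated helpers plus a parachute-and-frontend noise branch.
# The frontend/skydiving flag is only ever set true.
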AnimBlueprint /BRCosmetics/Characters/Player/Female/Medium/Bodies/F_MED_Cherry/Meshes/F_MED_Cherry_AnimBP
target_skeleton=&/F_MED_Cherry_Skeleton
graphs [
 AnimGraph {nodes [
   n1 = Control Rig(Source=Copy Pose From Mesh() {bUseAttachedParent=true}) {
    alphaBoolBlend {blendOption=Linear}
    controlRigAssetReference {blueprintRigClass=&/F_MED_Cherry_Deform_CtrlRig_twists_Dup.F_MED_Cherry_Deform_CtrlRig_twists_Dup_C}
   }
   n2 = Control Rig!(
    Source=@n1
    Alpha
    KneeRot_fr_l_114deg:Vector
    KneeTrans_fr_l_114deg:Vector="0.258,-3.591,0.009"
    KneeScale_fr_l_114deg:Vector="1,1,1"
    KneeScale_fr_l_130deg:Vector="1,1,1"
    KneeRot_fr_l_130deg:Vector="0,0,-51.2"
    KneeTrans_fr_l_130deg:Vector="0.258,-3.591,0.009"
    isRBANEnabled:bool=false
    baseBallSliding:bool=false
    isCrouching:bool=false
    isKneeSliding:bool=false
   ) -> (Pose) {
    alphaBoolBlend {blendOption=Linear}
    controlRigAssetReference {blueprintRigClass=&/F_MED_Cherry_Deform_CtrlRig_sculpters_Dup.F_MED_Cherry_Deform_CtrlRig_sculpters_Dup_C}
   }
   n3 = "Transform (Modify) Bone - Bone: dyn_scarf_medallion_MSTR"!(
    ComponentPose=Local To Component(LocalPose=@n2)
    bAlphaBoolEnabled:bool=Get frontEndorSkydiving() -> (frontEndorSkydiving:bool)
    Translation="0,0,0.5"
    Rotation="6,0,5"
    Scale="1,1,1"
   ) -> (Pose) {
    alphaInputType=Bool
    rotationMode=BMM_Additive
    rotationSpace=BCS_BoneSpace
    translationMode=BMM_Additive
    translationSpace=BCS_BoneSpace
    alphaBoolBlend {blendOption=Linear}
    boneToModify {boneName=dyn_scarf_medallion_MSTR}
    rotation {pitch=6, roll=5}
    translation {z=0.5}
   }
   n4 = "Transform (Modify) Bone - Bone: dyn_scarf_medallion_MSTR"!(
    ComponentPose=@n3
    Alpha:float=Get crouch alpha() -> (crouch alpha:double)
    Translation="0.039,4.752,-0.558"
    Rotation="9.768,4.402,24.408"
    Scale="1,1,1"
   ) -> (Pose) {
    rotationMode=BMM_Additive
    rotationSpace=BCS_BoneSpace
    translationMode=BMM_Additive
    translationSpace=BCS_BoneSpace
    alphaBoolBlend {blendOption=Linear}
    boneToModify {boneName=dyn_scarf_medallion_MSTR}
    rotation=<rot 9.768 4.402 24.4076>
    translation=<0.039 4.752 -0.558>
   }
   n5 = "Transform (Modify) Bone - Bone: dyn_root_slope"!(
    ComponentPose=@n4
    bAlphaBoolEnabled:bool=NOT Boolean(A=Get bCanPlayCustomAnimations() -> (bCanPlayCustomAnimations:bool))
    Translation=Get slope_root_trans() -> (slope_root_trans:Vector)
    Rotation=Get slope_foot_rot() -> (slope_foot_rot:Rotator)
    Scale="1,1,1"
   ) -> (Pose) {
    alphaInputType=Bool
    rotationMode=BMM_Replace
    rotationSpace=BCS_WorldSpace
    translationMode=BMM_Replace
    translationSpace=BCS_WorldSpace
    alphaBoolBlend {blendOption=Linear}
    boneToModify {boneName=dyn_root_slope}
   }
   n6 = "Transform (Modify) Bone - Bone: dyn_root_simspace"(ComponentPose=@n5, Alpha=1, Translation="0,0,0", Rotation=Get sim_space_rotation() -> (sim_space_rotation:Rotator), Scale="1,1,1") {
    rotationMode=BMM_Replace
    rotationSpace=BCS_WorldSpace
    translationSpace=BCS_BoneSpace
    alphaBoolBlend {blendOption=Linear}
    alphaScaleBiasClamp {bInterpResult=true, interpSpeedDecreasing=2, interpSpeedIncreasing=2}
    boneToModify {boneName=dyn_root_simspace}
   }
   n7 = RigidBody!(ComponentPose=@n6, Alpha, OverrideWorldGravity:Vector=Get GravityOverride(), ExternalForce) -> (Pose) {
    bClampLinearTranslationLimitToRefPose=true
    bOverrideWorldGravity=true
    overridePhysicsAsset=&/F_MED_Cherry_Dyn
    simulationSpace=BaseBoneSpace
    alphaBoolBlend {blendOption=Linear}
    baseBoneRef {boneName=dyn_root_simspace}
    componentAppliedLinearAccClamp=<3000 3000 3000>
    componentLinearAccScale {z=0.5}
    componentLinearVelScale {z=0.5}
    overrideWorldGravity {z=-980}
    simSpaceSettings {
     maxAngularAcceleration=1000
     maxAngularVelocity=3
     maxLinearAcceleration=700
     maxLinearVelocity=500
     velocityScaleZ=0
     worldAlpha=1
     externalLinearDragV=<1 1>
    }
   }
   Output Pose(Result=Component To Local(ComponentPose=@n7)) {layerGroup=None}
  ]}
 EventGraph {nodes [
   n1 = Event BlueprintUpdateAnimation()
   > n2 = Sequence!(execute) -> (then_0, then_1, then_2:Exec, then_3:Exec)
   > n3 = Branch(Condition=AND Boolean(A=Get bIsParachuteOpened(), B=Get bIsInFrontEnd()))
   n4 = GetOwnerSkeletalMeshComponent()
   n5 = GetGameTimeInSeconds()
   n6 = Self-Reference()
   > n7 = Random Vector Function!(
    execute
    CharacterPartAnimInstance:Custom Character Part Anim Instance Object Reference=@n6
    SkelMeshComponent:Mesh Component Object Reference=@n4
    JointName:Name=root
    GameTimeInSeconds:double=@n5
    TimeOffset:double=0
    PerlinScale:double=1
    Frequency:Vector="1.5,7,8"
    Amplitude:Vector="200,-500,400"
    Offset:Vector="0,-600,400"
   ) -> (then, VectorOut:Vector)
   > n8 = AlignVectorWithJoint!(
    execute
    Vector:Vector="0,-800,-980"
    MeshComp:Mesh Component Object Reference=@n4
    PartAnimInstance:Custom Character Part Anim Instance Object Reference=@n6
    JointName:Name=root
   ) -> (then, ReturnValue:Vector)
   > Set GravityOverride(GravityOverride=vector + vector(A:Vector=@n8, B:Vector=@n7) -> (ReturnValue:Vector) {is=title K2Node_CommutativeAssociativeBinaryOperator})
   n9 = Gravity Override State Function DEPRECATED(execute=@n3.else, CharacterPartAnimInstance=@n6, GameTimeInSeconds=@n5, GravityOverrideParams=Get Gravity Override Params())
   > Set GravityOverride(GravityOverride=@n9)
   n10 = Break Fort Anim Input Slope Info!(FortAnimInput_SlopeInfo:"Fort Anim Input Slope Info Structure (by ref)"=Get SlopeInfoInput() -> (SlopeInfoInput:Fort Anim Input Slope Info Structure)) -> (RootSlopeTranslation:Vector=~, LeftFootSlopeTranslation:Vector=~, RightFootSlopeTranslation:Vector=~, SlopeRotation:Rotator=~) {is=title K2Node_BreakStruct}
   Set slope_root_trans(execute=@n2.then_1, slope_root_trans:Vector=@n10) -> (Output_Get:Vector)
   > Set slope_foot_rot(slope_foot_rot:Rotator=@n10.SlopeRotation) -> (Output_Get:Rotator)
   n11 = ComplexSimSpaceFunction DEPRECATED!(
    execute=@n2.then_2
    CharacterPartAnimInstance:Custom Character Part Anim Instance Object Reference=Self-Reference()
    InterpSpeedMax:double=9
    PawnSpeedMax:double=20
    RInterpErrorTolerance:double=1
    DeltaTime:double=@n1.DeltaTimeX
    DebugDraw:bool=false
    sim_space_rotation_in:Rotator=Get sim_space_rotation()
    inComponentSpaceIn:bool=Get InComponentSpace() -> (InComponentSpace:bool)
    inComponentSpaceLastFrameIn:bool=Get InComponentSpaceLastFrame() -> (InComponentSpaceLastFrame:bool)
    lockedOnSimSpaceIn:bool=Get LockedOnSimSpace() -> (LockedOnSimSpace:bool)
    InterpSpeedIn:double=Get InterpSpeed() -> (InterpSpeed:double)
   ) -> (
    then
    sim_space_rotation:Rotator
    inComponentSpaceOut:bool
    inComponentSpaceLastFrameOut:bool
    lockedOnSimSpaceOut:bool
    InterpSpeedOut:double
   )
   > Set sim_space_rotation(sim_space_rotation:Rotator=@n11) -> (Output_Get:Rotator)
   > Set InComponentSpace(InComponentSpace:bool=@n11.inComponentSpaceOut) -> (Output_Get:bool)
   > Set InComponentSpaceLastFrame(InComponentSpaceLastFrame:bool=@n11.inComponentSpaceLastFrameOut) -> (Output_Get:bool)
   > Set LockedOnSimSpace(LockedOnSimSpace:bool=@n11.lockedOnSimSpaceOut) -> (Output_Get:bool)
   > Set InterpSpeed(InterpSpeed:double=@n11.InterpSpeedOut) -> (Output_Get:double)
   Branch(execute=@n2.then_3, Condition=OR Boolean(A=Get bIsSkyDiving(), B=Get bIsInFrontEnd()))
   > Set frontEndorSkydiving(frontEndorSkydiving:bool=true) -> (Output_Get:bool)
   Event BlueprintInitializeAnimation()
   > Set Skel_Mesh(Skel_Mesh=GetOwnerSkeletalMeshComponent())
   Event OnEndCrouching()
   > Set crouch alpha(crouch alpha:double=0) -> (Output_Get:double)
   Event OnBeginCrouching()
   > Set crouch alpha(crouch alpha=1)
   GetAlphaFromReferenceTransformDelta(BoneName=None, BaseName=None, DeltaComponentType=QuaternionTwist, InputRange=(X=0,Y=1), TwistAxis="1,0,0")
  ]}
]
variables [
 crouch alpha {type=number}
 frontEndorSkydiving {type=boolean}
 Gravity Override Params {
  type=GravityOverrideParamsStruct
  default {front_End {gravityOverride=<0 0 -500>, windAmplitude=<200 300 0>, windFrequency=<1.3 1.5 1.8>, windOffset=<0 -500 -100>}}
 }
 GravityOverride {type=Vector}
 InComponentSpace {type=boolean}
 InComponentSpaceLastFrame {type=boolean}
 InterpSpeed {type=number, default=9}
 InterpSpeedMax {type=number}
 LockedOnSimSpace {type=boolean}
 root_rotation {type=Rotator}
 sim_space_rotation {type=Rotator}
 Skel_Mesh {type=SkeletalMeshComponent}
 slope_foot_rot {type=Rotator}
 slope_root_trans {type=Vector}
]
