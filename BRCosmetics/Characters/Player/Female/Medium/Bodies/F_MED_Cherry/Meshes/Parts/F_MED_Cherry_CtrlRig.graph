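# Control Rig for the F_MED_Cherry character part that orients the root bones of four dynamic hair chains: left, back-left, back and back-right.
# The forward solve runs four chained Aim units, pointing each bone's X axis at an aim control and Z at an up control.
# Up controls hang off the head, while aim controls hang off spine_05 or the clavicles.
# Front, right and glider hair bones exist in the hierarchy but are not driven.
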ControlRigBlueprint /BRCosmetics/Characters/Player/Female/Medium/Bodies/F_MED_Cherry/Meshes/Parts/F_MED_Cherry_CtrlRig
graphs [
 RigVMFunctionLibrary {links [], nodes []}
 RigVMModel {nodes [
   RigUnit_BeginExecution = &RigUnit_BeginExecution()
   /Script/ControlRig.RigUnit_AimItem(
    DebugSettings=(Scale=10,WorldOffset=(Rotation=(X=0,Y=0,Z=0,W=1),Translation=(X=0,Y=0,Z=0),Scale3D=(X=1,Y=1,Z=1)))
    Item=(Type=Bone,Name="dyn_hair_l_01")
    Primary=(Weight=1,Axis=(X=1,Y=0,Z=0),Target=(X=1,Y=0,Z=0),Kind=Location,Space=(Type=Control,Name="dyn_hair_l_01_aim_ctrl"))
    Secondary=(Weight=1,Axis=(X=0,Y=0,Z=1),Target=(X=1,Y=0,Z=0),Kind=Location,Space=(Type=Control,Name="dyn_hair_l_01_up_ctrl"))
    Weight=1
    ExecutePin=@RigUnit_BeginExecution.ExecutePin
   ) {is=struct RigVMUnitNode}
   > &RigUnit_AimItem(
    DebugSettings=(Scale=10,WorldOffset=(Rotation=(X=0,Y=0,Z=0,W=1),Translation=(X=0,Y=0,Z=0),Scale3D=(X=1,Y=1,Z=1)))
    Item=(Type=Bone,Name="dyn_hair_bk_l_01")
    Primary=(Weight=1,Axis=(X=1,Y=0,Z=0),Target=(X=1,Y=0,Z=0),Kind=Location,Space=(Type=Control,Name="dyn_hair_bk_l_01_aim_ctrl"))
    Secondary=(Weight=1,Axis=(X=0,Y=0,Z=1),Target=(X=1,Y=0,Z=0),Kind=Location,Space=(Type=Control,Name="dyn_hair_bk_l_01_up_ctrl"))
    Weight=1
   )
   > &RigUnit_AimItem(
    DebugSettings=(Scale=10,WorldOffset=(Rotation=(X=0,Y=0,Z=0,W=1),Translation=(X=0,Y=0,Z=0),Scale3D=(X=1,Y=1,Z=1)))
    Item=(Type=Bone,Name="dyn_hair_bk_01")
    Primary=(Weight=1,Axis=(X=1,Y=0,Z=0),Target=(X=1,Y=0,Z=0),Kind=Location,Space=(Type=Control,Name="dyn_hair_bk_01_aim_ctrl"))
    Secondary=(Weight=1,Axis=(X=0,Y=0,Z=1),Target=(X=1,Y=0,Z=0),Kind=Location,Space=(Type=Control,Name="dyn_hair_bk_01_up_ctrl"))
    Weight=1
   )
   > &RigUnit_AimItem(
    DebugSettings=(Scale=10,WorldOffset=(Rotation=(X=0,Y=0,Z=0,W=1),Translation=(X=0,Y=0,Z=0),Scale3D=(X=1,Y=1,Z=1)))
    Item=(Type=Bone,Name="dyn_hair_bk_r_01")
    Primary=(Weight=1,Axis=(X=1,Y=0,Z=0),Target=(X=1,Y=0,Z=0),Kind=Location,Space=(Type=Control,Name="dyn_hair_bk_r_01_aim_ctrl"))
    Secondary=(Weight=1,Axis=(X=0,Y=0,Z=1),Target=(X=1,Y=0,Z=0),Kind=Location,Space=(Type=Control,Name="dyn_hair_bk_r_01_up_ctrl"))
    Weight=1
   )
  ]}
]
hierarchy [
 root {loc=<0 0 0>, rot=<0 0 0 1>}
 pelvis {parent=root, loc=<0 -0.5301 98.0571>, rot=<0.0127 0.707 -0.0127 -0.707>}
 spine_01 {parent=pelvis, loc=<6.1208 -0.9328 0>, rot=<0 0 0.034 -0.9994>}
 spine_02 {parent=spine_01, loc=<6.5996 -0.2095 0>, rot=<0 0 0.0238 -0.9997>}
 spine_03 {parent=spine_02, loc=<6.4914 -0.1237 0>, rot=<0 0 0.0149 -0.9999>}
 spine_04 {parent=spine_03, loc=<6.9598 0.0794 0>, rot=<0 0 -0.0132 -0.9999>}
 spine_05 {parent=spine_04, loc=<6.9987 0 0>, rot=<0 0 -0.0015 -1>}
 neck_01 {parent=spine_05, loc=<11.3138 0.1529 0>, rot=<0 0 -0.2003 -0.9797>}
 neck_02 {parent=neck_01, loc=<4.5627 0 0>, rot=<0 0 0.0504 -0.9987>}
 head {parent=neck_02, loc=<5.3914 0 0>, rot=<0 0 0.115 -0.9934>}
 faceAttach {parent=head, loc=<0.6201 0.9296 0>, rot=<0.0082 -0.7071 -0.0082 -0.7071>, scale=<0.89 0.89 0.89>}
 C_brow_mid {parent=faceAttach, loc=<-0.0005 8.1467 3.9557>, rot=<0 -0.0561 0.0111 -0.9984>}
 L_brow_mid {parent=faceAttach, loc=<3.1178 7.668 5.0784>, rot=<0 -0.0036 0.2553 -0.9669>}
 L_brow_outer {parent=faceAttach, loc=<5.2184 6.2669 5.0894>, rot=<0.2473 0.3077 0.3472 -0.8506>}
 R_brow_mid {parent=faceAttach, loc=<-3.1177 7.668 5.0786>, rot=<-0.0001 0.0042 -0.2553 -0.9669>}
 R_brow_outer {parent=faceAttach, loc=<-5.2184 6.2668 5.0894>, rot=<0.134 -0.1648 -0.4007 -0.8912>}
 dyn_hair_scaler {parent=head, loc=<6.3544 0 0>, rot=<0 0 0 -1>}
 dyn_hair_fr_l_MSTR {parent=dyn_hair_scaler, loc=<3.9329 4.6758 -5.2059>, rot=<0.0571 -0.9905 -0.0303 -0.1217>}
 dyn_hair_fr_l_01 {parent=dyn_hair_fr_l_MSTR, loc=<0.2 0 0>, rot=<0 0 0 1>}
 dyn_hair_fr_l_02 {parent=dyn_hair_fr_l_01, loc=<4.862 0.0106 0.0766>, rot=<-0.0244 -0.0215 0.0035 -0.9995>}
 dyn_hair_fr_l_03 {parent=dyn_hair_fr_l_02, loc=<6.6978 0 0>, rot=<0 0 0 -1>}
 dyn_hair_fr_l_04 {parent=dyn_hair_fr_l_03, loc=<6.6978 0 0>, rot=<0 0 0 -1>}
 dyn_hair_l_MSTR {parent=dyn_hair_scaler, loc=<4.0431 -0.323 -5.1865>, rot=<0.0576 -0.9774 0.1556 -0.1312>}
 dyn_hair_l_01 {parent=dyn_hair_l_MSTR, loc=<0.2 0 0>, rot=<0 0 0 1>}
 dyn_hair_l_02 {parent=dyn_hair_l_01, loc=<4.9312 0 0.1089>, rot=<0 -0.0307 0 -0.9995>}
 dyn_hair_l_03 {parent=dyn_hair_l_02, loc=<6.6691 0 0>, rot=<0 0 0 -1>}
 dyn_hair_l_04 {parent=dyn_hair_l_03, loc=<6.6691 0 0>, rot=<0 0 0 -1>}
 dyn_hair_bk_l_MSTR {parent=dyn_hair_scaler, loc=<3.8851 -4.0256 -2.958>, rot=<0.0477 -0.9176 0.3797 -0.1077>}
 dyn_hair_bk_l_01 {parent=dyn_hair_bk_l_MSTR, loc=<0.2 0 0>, rot=<0 0 0 -1>}
 dyn_hair_bk_l_02 {parent=dyn_hair_bk_l_01, loc=<5.0526 0.0465 -0.0411>, rot=<-0.0358 0.0121 0.0127 -0.9992>}
 dyn_hair_bk_l_03 {parent=dyn_hair_bk_l_02, loc=<6.6711 0 -0.0846>, rot=<0 0.0239 0 -0.9997>}
 dyn_hair_bk_l_04 {parent=dyn_hair_bk_l_03, loc=<6.6691 0 0>, rot=<0 0 0 -1>}
 dyn_hair_bk_MSTR {parent=dyn_hair_scaler, loc=<3.87 -5.5556 0>, rot=<0.1015 -0.6998 0.6998 -0.1015>}
 dyn_hair_bk_01 {parent=dyn_hair_bk_MSTR, loc=<0.2 0 0>, rot=<0 0 0 -1>}
 dyn_hair_bk_02 {parent=dyn_hair_bk_01, loc=<5.3512 0 0>, rot=<0 0 0 -1>}
 dyn_hair_bk_03 {parent=dyn_hair_bk_02, loc=<7.1572 0 0>, rot=<0 0 0 -1>}
 dyn_hair_bk_04 {parent=dyn_hair_bk_03, loc=<6.4891 0 0>, rot=<0 0 0 -1>}
 dyn_hair_bk_r_MSTR {parent=dyn_hair_scaler, loc=<3.885 -4.0256 2.958>, rot=<0.1077 -0.3797 0.9176 -0.0477>}
 dyn_hair_bk_r_01 {parent=dyn_hair_bk_r_MSTR, loc=<0.2003 0 -0.0001>, rot=<0 0 0 -1>}
 dyn_hair_bk_r_02 {parent=dyn_hair_bk_r_01, loc=<5.0519 -0.0465 -0.041>, rot=<0.0358 0.0121 -0.0127 -0.9992>}
 dyn_hair_bk_r_03 {parent=dyn_hair_bk_r_02, loc=<6.6719 0 -0.0848>, rot=<0 0.0239 0 -0.9997>}
 dyn_hair_bk_r_04 {parent=dyn_hair_bk_r_03, loc=<6.6686 0 0.0001>, rot=<0 0 0 1>}
 dyn_hair_r_MSTR {parent=dyn_hair_scaler, loc=<4.0434 -0.323 5.1865>, rot=<0.1312 -0.1556 0.9774 -0.0576>}
 dyn_hair_r_01 {parent=dyn_hair_r_MSTR, loc=<0.2 0 0>, rot=<0 0 0 1>}
 dyn_hair_r_02 {parent=dyn_hair_r_01, loc=<4.9319 0 0.1087>, rot=<0 -0.0307 0 -0.9995>}
 dyn_hair_r_03 {parent=dyn_hair_r_02, loc=<6.6692 0 0>, rot=<0 0 0 -1>}
 dyn_hair_r_04 {parent=dyn_hair_r_03, loc=<6.6682 0.0001 0.0002>, rot=<0 0 0 -1>}
 dyn_hair_fr_r_MSTR {parent=dyn_hair_scaler, loc=<3.9331 4.6758 5.2059>, rot=<0.1217 0.0303 0.9905 -0.0571>}
 dyn_hair_fr_r_01 {parent=dyn_hair_fr_r_MSTR, loc=<0.2001 0 0>, rot=<0 0 0 1>}
 dyn_hair_fr_r_02 {parent=dyn_hair_fr_r_01, loc=<4.8617 -0.0106 0.0766>, rot=<0.0244 -0.0215 -0.0035 -0.9995>}
 dyn_hair_fr_r_03 {parent=dyn_hair_fr_r_02, loc=<6.6981 0 -0.0001>, rot=<0 0 0 1>}
 dyn_hair_fr_r_04 {parent=dyn_hair_fr_r_03, loc=<6.6981 0 -0.0001>, rot=<0 0 0 1>}
 dyn_hair_fr_r_strand_01 {parent=dyn_hair_fr_r_02, loc=<1.4159 -2.7321 0.9221>, rot=<0 0.2047 0 -0.9788>}
 dyn_hair_fr_r_strand_02 {parent=dyn_hair_fr_r_strand_01, loc=<5.9118 0 -0.0001>, rot=<0 0 -0.3383 -0.941>}
 dyn_glider_hair_scaler {parent=head, loc=<6.3544 0 0>, rot=<0 0 0 -1>}
 dyn_hair_glider_MSTR {parent=dyn_glider_hair_scaler, loc=<5.7329 -0.6275 0>, rot=<0.5816 -0.4022 0.4022 -0.5816>}
 dyn_hair_glider_01 {parent=dyn_hair_glider_MSTR, loc=<0.2 0 0>, rot=<0 0 0 1>}
 dyn_hair_glider_02 {parent=dyn_hair_glider_01, loc=<6.2416 0 0>, rot=<0 0.1121 0 -0.9937>}
 dyn_hair_glider_03 {parent=dyn_hair_glider_02, loc=<6.4415 0 0>, rot=<0 0.1254 0 -0.9921>}
 dyn_hair_glider_04 {parent=dyn_hair_glider_03, loc=<5.7734 0 0>, rot=<0 0 0 1>}
 dyn_hair_glider_05 {parent=dyn_hair_glider_04, loc=<6.7508 0 0>, rot=<0 -0.1108 0 -0.9938>}
 dyn_hair_glider_06 {parent=dyn_hair_glider_05, loc=<7.6653 0 0>, rot=<0 -0.1314 0 -0.9913>}
 dyn_hair_glider_07 {parent=dyn_hair_glider_06, loc=<6.2538 0 0>, rot=<0 0 0 1>}
 dyn_hair_glider_end_l_01_MSTR {parent=dyn_hair_glider_04, loc=<-0.7887 0.471 -3.3424>, rot=<-0.0664 0.0478 -0.1786 -0.9805>}
 dyn_hair_glider_end_l_01 {parent=dyn_hair_glider_end_l_01_MSTR, loc=<0.3 0 0>, rot=<0 0 0 -1>}
 dyn_hair_glider_end_l_02 {parent=dyn_hair_glider_end_l_01, loc=<5.7349 0 0>, rot=<-0.0783 -0.16 0.2494 -0.9519>}
 dyn_hair_glider_end_top_01_MSTR {parent=dyn_hair_glider_04, loc=<3.0173 -0.1876 3.2136>, rot=<-0.0842 -0.0063 0.1848 -0.9791>}
 dyn_hair_glider_end_top_01 {parent=dyn_hair_glider_end_top_01_MSTR, loc=<0.3 0 0>, rot=<0 0 0 -1>}
 dyn_hair_glider_end_top_02 {parent=dyn_hair_glider_end_top_01, loc=<5.7349 0 0>, rot=<-0.1164 -0.2345 -0.1515 -0.9532>}
 dyn_hair_glider_back_l_01_MSTR {parent=dyn_hair_glider_02, loc=<-0.4191 5.8041 -2.1041>, rot=<-0.0278 -0.0282 -0.1435 -0.9889>}
 dyn_hair_glider_back_l_01 {parent=dyn_hair_glider_back_l_01_MSTR, loc=<0.3 0 0>, rot=<0 0 0 1>}
 dyn_hair_glider_back_l_02 {parent=dyn_hair_glider_back_l_01, loc=<3.3243 0.1616 0>, rot=<0.0245 0.1523 0.0614 -0.9861>}
 dyn_hair_glider_back_l_03 {parent=dyn_hair_glider_back_l_02, loc=<4.9929 0 0>, rot=<-0.032 0.2241 0.1429 -0.9635>}
 dyn_hair_glider_side_r_01_MSTR {parent=dyn_glider_hair_scaler, loc=<1.3659 1.3602 6.5993>, rot=<0.6916 -0.1472 0.5952 -0.3818>}
 dyn_hair_glider_side_r_01 {parent=dyn_hair_glider_side_r_01_MSTR, loc=<0.3 0 0>, rot=<0 0 0 1>}
 dyn_hair_glider_side_r_02 {parent=dyn_hair_glider_side_r_01, loc=<4.0401 0 0>, rot=<0 0 -0.4978 -0.8673>}
 dyn_hair_glider_side_r_03 {parent=dyn_hair_glider_side_r_02, loc=<5.8193 0 0>, rot=<-0.0026 0.1848 0.0139 -0.9827>}
 head_up_grp {parent=head, type=Null, loc=<0 0 0>, rot=<0 0 0 1>}
 dyn_hair_bk_01_up_ctrl {parent=head_up_grp, type=Control, loc=<0 0 0>, rot=<0 0 0 1>}
 dyn_hair_bk_l_01_up_ctrl {parent=head_up_grp, type=Control, loc=<0 0 0>, rot=<0 0 0 1>}
 dyn_hair_bk_r_01_up_ctrl {parent=head_up_grp, type=Control, loc=<0 0 0>, rot=<0 0 0 1>}
 dyn_hair_l_01_up_ctrl {parent=head_up_grp, type=Control, loc=<0 0 0>, rot=<0 0 0 1>}
 clavicle_l {parent=spine_05, loc=<5.4831 -0.1053 -3.2946>, rot=<0.0465 -0.7289 -0.009 -0.6829>}
 upperarm_l {parent=clavicle_l, loc=<9.8658 -1.2136 -0.4105>, rot=<0.0208 -0.486 -0.0242 -0.8734>}
 clavicle_l_aim_grp {parent=clavicle_l, type=Null, loc=<0 0 0>, rot=<0 0 0 1>}
 dyn_hair_l_01_aim_ctrl {parent=clavicle_l_aim_grp, type=Control, loc=<0 0 0>, rot=<0 0 0 1>}
 clavicle_r {parent=spine_05, loc=<5.4831 -0.1053 3.2946>, rot=<0.7289 0.0465 -0.6829 0.009>}
 upperarm_r {parent=clavicle_r, loc=<-9.8658 1.2136 0.4104>, rot=<0.0208 -0.486 -0.0242 -0.8734>}
 clavicle_r_aim_grp {parent=clavicle_r, type=Null, loc=<0 0 0>, rot=<0 0 0 1>}
 dyn_hair_bk_r_01_aim_ctrl {parent=clavicle_r_aim_grp, type=Control, loc=<0 0 0>, rot=<0 0 0 1>}
 spine_05_aim_grp {parent=spine_05, type=Null, loc=<0 0 0>, rot=<0 0 0 1>}
 dyn_hair_bk_01_aim_ctrl {parent=spine_05_aim_grp, type=Control, loc=<0 0 0>, rot=<0 0 0 1>}
 dyn_hair_bk_l_01_aim_ctrl {parent=spine_05_aim_grp, type=Control, loc=<0 0 0>, rot=<0 0 0 1>}
 dyn_root_simspace {parent=root, loc=<0 0 0>, rot=<0 0 0 1>}
]
