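# Physics asset for the dynamic hair of the F_MED_Cherry player body, with capsule bodies on hair, head, neck, upper spine, clavicles and upper arms.
# Seven four-link hair chains hang from MSTR bones on the head with limited swing and tight twist, plus a two-link strand off the front-right chain.
# Free-angle constraints join neighboring chains at their third links around the back, and also link the MSTR and body bones.
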PhysicsAsset /BRCosmetics/Characters/Player/Female/Medium/Bodies/F_MED_Cherry/Meshes/Parts/F_MED_Cherry_Dyn_HairTweak
bodies {
 clavicle_l [{shapeType=Capsule, length=5.75, radius=6.8125, center=<6.4482 -0.3686 -2.4355>, rotation=<rot 85.606 0 0>}]
 clavicle_r [{shapeType=Capsule, length=7.3062, radius=6.815, center=<-6.4482 0.3686 2.4355>, rotation=<rot 85.606 0 0>}]
 dyn_hair_bk_01 [{shapeType=Capsule, length=4.3503, radius=8, center=<2.5886 -0.0001 7>, rotation=<rot 90 0 2.5>}]
 dyn_hair_bk_02 [{shapeType=Capsule, length=4.3503, radius=8, center=<2.5887 0.0002 7>, rotation=<rot 90 0 2.5>}]
 dyn_hair_bk_03 [{shapeType=Capsule, length=3.9753, radius=8, center=<2.5887 0.0002 7>, rotation=<rot 90 0 2.5>}]
 dyn_hair_bk_04 [{shapeType=Capsule, length=3.3503, radius=8, center=<2.5886 0 7>, rotation=<rot 90 0 2.5>}]
 dyn_hair_bk_MSTR [{shapeType=Capsule, length=0.505, radius=0.505}]
 dyn_hair_bk_l_01 [{shapeType=Capsule, length=4.3503, radius=8, center=<2.5886 -0.0002 7>, rotation=<rot 90 0 2.5>}]
 dyn_hair_bk_l_02 [{shapeType=Capsule, length=4.3503, radius=8, center=<2.5887 0.0001 7>, rotation=<rot 90 0 2.5>}]
 dyn_hair_bk_l_03 [{shapeType=Capsule, length=3.9753, radius=8, center=<2.5887 0.0001 7>, rotation=<rot 90 0 2.5>}]
 dyn_hair_bk_l_04 [{shapeType=Capsule, length=3.3503, radius=8, center=<2.5887 0.0001 7>, rotation=<rot 90 0 2.5>}]
 dyn_hair_bk_l_MSTR [{shapeType=Capsule, length=0.505, radius=0.505}]
 dyn_hair_bk_r_01 [{shapeType=Capsule, length=4.4441, radius=8, center=<2.5886 -0.0002 7>, rotation=<rot 90 0 2.5>}]
 dyn_hair_bk_r_02 [{shapeType=Capsule, length=4.4441, radius=8, center=<2.5887 0.0001 7>, rotation=<rot 90 0 2.5>}]
 dyn_hair_bk_r_03 [{shapeType=Capsule, length=3.9753, radius=8, center=<2.5887 0.0001 7>, rotation=<rot 90 0 2.5>}]
 dyn_hair_bk_r_04 [{shapeType=Capsule, length=3.3503, radius=8, center=<2.5887 0.0001 7>, rotation=<rot 90 0 2.5>}]
 dyn_hair_bk_r_MSTR [{shapeType=Capsule, length=0.505, radius=0.505}]
 dyn_hair_fr_l_01 [{shapeType=Capsule, length=3.9753, radius=8, center=<2.5887 0.0002 7>, rotation=<rot 90 0 2.5>}]
 dyn_hair_fr_l_02 [{shapeType=Capsule, length=3.9753, radius=8, center=<2.5886 0 7>, rotation=<rot 90 0 2.5>}]
 dyn_hair_fr_l_03 [{shapeType=Capsule, length=3.9753, radius=8, center=<0.727 0.0812 7>, rotation=<rot 90 0 2.5>}]
 dyn_hair_fr_l_04 [{shapeType=Capsule, length=0, radius=4.7086, center=<3.1456 -1.4287 2.9997>, rotation=<rot 0 -21.5365 89.9996>}]
 dyn_hair_fr_l_MSTR [{shapeType=Capsule, length=0.505, radius=0.505}]
 dyn_hair_fr_r_01 [{shapeType=Capsule, length=3.9753, radius=8, center=<2.5887 0.0002 7>, rotation=<rot 90 0 2.5>}]
 dyn_hair_fr_r_02 [{shapeType=Capsule, length=3.9753, radius=8, center=<2.5886 0 7>, rotation=<rot 90 0 2.5>}]
 dyn_hair_fr_r_03 [{shapeType=Capsule, length=3.9753, radius=8, center=<0.5991 0.0871 7>, rotation=<rot 90 0 2.5>}]
 dyn_hair_fr_r_04 [{shapeType=Capsule, length=4.3503, radius=3.3375, center=<3.0906 0.1967 3>, rotation=<rot 0 31.946 89.9999>}]
 dyn_hair_fr_r_MSTR [{shapeType=Capsule, length=0.505, radius=0.505}]
 dyn_hair_fr_r_strand_01 [{shapeType=Capsule, length=3.9753, radius=5, center=<2.5887 0 0>, rotation=<rot 90 0 0>}]
 dyn_hair_fr_r_strand_02 [{shapeType=Capsule, length=3.9753, radius=5, center=<2.5887 0 0>, rotation=<rot 90 0 0>}]
 dyn_hair_l_01 [{shapeType=Capsule, length=3.9753, radius=8, center=<2.5887 0.0002 7>, rotation=<rot 90 0 2.5>}]
 dyn_hair_l_02 [{shapeType=Capsule, length=3.9753, radius=8, center=<2.5886 0 7>, rotation=<rot 90 0 2.5>}]
 dyn_hair_l_03 [{shapeType=Capsule, length=3.9753, radius=8, center=<0.727 0.0813 7>, rotation=<rot 90 0 2.5>}]
 dyn_hair_l_04 [{shapeType=Capsule, length=5.6941, radius=3.9859, center=<1.4159 0 3>, rotation=<rot 0 0 90>}]
 dyn_hair_l_MSTR [{shapeType=Capsule, length=0.505, radius=0.505}]
 dyn_hair_r_01 [{shapeType=Capsule, length=3.9753, radius=8, center=<2.5887 0.0002 7>, rotation=<rot 90 0 2.5>}]
 dyn_hair_r_02 [{shapeType=Capsule, length=3.9753, radius=8, center=<2.5886 0 7>, rotation=<rot 90 0 2.5>}]
 dyn_hair_r_03 [{shapeType=Capsule, length=3.9753, radius=8, center=<0.5991 0.0872 7>, rotation=<rot 90 0 2.5>}]
 dyn_hair_r_04 [{shapeType=Capsule, length=5.6941, radius=4.4483, center=<0.8529 0 3>, rotation=<rot 0 0 90>}]
 dyn_hair_r_MSTR [{shapeType=Capsule, length=0.505, radius=0.505}]
 head [
  {shapeType=Capsule, length=6.9062, radius=6, center=<1.4916 0.8211 0>, rotation=<rot 90 0 0>}
  {shapeType=Capsule, length=7.1875, radius=5.7188, center=<1.4554 6.0111 0>, rotation=<rot 90 0 0>}
 ]
 neck_01 [{shapeType=Capsule, length=4.5312, radius=6, center=<0.8723 0.821 0>, rotation=<rot 90 0 0>}]
 neck_02 [{shapeType=Capsule, length=4.5312, radius=6, center=<0.8723 0.821 0>, rotation=<rot 90 0 0>}]
 spine_05 [{shapeType=Capsule, length=3.4688, radius=8.5312, center=<8.4986 -0.8112 -0.0001>}]
 upperarm_l [{shapeType=Capsule, length=19.25, radius=4.875, center=<10.8056 -0.4936 0.3456>, rotation=<rot 89.5506 0 0>}]
 upperarm_r [{shapeType=Capsule, length=22.0842, radius=6.7201, center=<-10.8056 0.4936 -0.3456>, rotation=<rot 89.5506 0 0>}]
}
constraints [
 {bone1Name=dyn_hair_fr_l_01, bone2Name=dyn_hair_fr_l_MSTR, swing1LimitDegrees=15, swing2LimitDegrees=45, twistLimitDegrees=5}
 {bone1Name=dyn_hair_fr_l_02, bone2Name=dyn_hair_fr_l_01, swing1LimitDegrees=15, swing2LimitDegrees=45, twistLimitDegrees=5}
 {bone1Name=dyn_hair_fr_l_03, bone2Name=dyn_hair_fr_l_02, swing1LimitDegrees=15, swing2LimitDegrees=45, twistLimitDegrees=5}
 {bone1Name=dyn_hair_fr_l_04, bone2Name=dyn_hair_fr_l_03, swing1LimitDegrees=5, swing2LimitDegrees=45, twistLimitDegrees=5}
 {bone1Name=dyn_hair_l_01, bone2Name=dyn_hair_l_MSTR, swing1LimitDegrees=45, swing2LimitDegrees=45, twistLimitDegrees=5}
 {bone1Name=dyn_hair_l_02, bone2Name=dyn_hair_l_01, swing1LimitDegrees=45, swing2LimitDegrees=45, twistLimitDegrees=5}
 {bone1Name=dyn_hair_l_03, bone2Name=dyn_hair_l_02, swing1LimitDegrees=45, swing2LimitDegrees=45, twistLimitDegrees=5}
 {bone1Name=dyn_hair_l_04, bone2Name=dyn_hair_l_03, swing1LimitDegrees=45, swing2LimitDegrees=45, twistLimitDegrees=5}
 {bone1Name=dyn_hair_bk_l_01, bone2Name=dyn_hair_bk_l_MSTR, swing1LimitDegrees=45, swing2LimitDegrees=22.5, twistLimitDegrees=5}
 {bone1Name=dyn_hair_bk_l_02, bone2Name=dyn_hair_bk_l_01, swing1LimitDegrees=45, swing2LimitDegrees=22.5, twistLimitDegrees=5}
 {bone1Name=dyn_hair_bk_l_03, bone2Name=dyn_hair_bk_l_02, swing1LimitDegrees=45, swing2LimitDegrees=22.5, twistLimitDegrees=5}
 {bone1Name=dyn_hair_bk_l_04, bone2Name=dyn_hair_bk_l_03, swing1LimitDegrees=45, swing2LimitDegrees=22.5, twistLimitDegrees=5}
 {bone1Name=dyn_hair_bk_01, bone2Name=dyn_hair_bk_MSTR, swing1LimitDegrees=45, swing2LimitDegrees=22.5, twistLimitDegrees=5}
 {bone1Name=dyn_hair_bk_02, bone2Name=dyn_hair_bk_01, swing1LimitDegrees=45, swing2LimitDegrees=22.5, twistLimitDegrees=5}
 {bone1Name=dyn_hair_bk_03, bone2Name=dyn_hair_bk_02, swing1LimitDegrees=45, swing2LimitDegrees=22.5, twistLimitDegrees=5}
 {bone1Name=dyn_hair_bk_04, bone2Name=dyn_hair_bk_03, swing1LimitDegrees=45, swing2LimitDegrees=22.5, twistLimitDegrees=5}
 {bone1Name=dyn_hair_bk_r_01, bone2Name=dyn_hair_bk_r_MSTR, swing1LimitDegrees=45, swing2LimitDegrees=22.5, twistLimitDegrees=5}
 {bone1Name=dyn_hair_bk_r_02, bone2Name=dyn_hair_bk_r_01, swing1LimitDegrees=45, swing2LimitDegrees=22.5, twistLimitDegrees=5}
 {bone1Name=dyn_hair_bk_r_03, bone2Name=dyn_hair_bk_r_02, swing1LimitDegrees=45, swing2LimitDegrees=22.5, twistLimitDegrees=5}
 {bone1Name=dyn_hair_bk_r_04, bone2Name=dyn_hair_bk_r_03, swing1LimitDegrees=45, swing2LimitDegrees=22.5, twistLimitDegrees=5}
 {bone1Name=dyn_hair_r_01, bone2Name=dyn_hair_r_MSTR, swing1LimitDegrees=45, swing2LimitDegrees=45, twistLimitDegrees=5}
 {bone1Name=dyn_hair_r_02, bone2Name=dyn_hair_r_01, swing1LimitDegrees=45, swing2LimitDegrees=45, twistLimitDegrees=5}
 {bone1Name=dyn_hair_r_03, bone2Name=dyn_hair_r_02, swing1LimitDegrees=45, swing2LimitDegrees=45, twistLimitDegrees=5}
 {bone1Name=dyn_hair_r_04, bone2Name=dyn_hair_r_03, swing1LimitDegrees=45, swing2LimitDegrees=45, twistLimitDegrees=5}
 {bone1Name=dyn_hair_fr_r_01, bone2Name=dyn_hair_fr_r_MSTR, swing1LimitDegrees=15, swing2LimitDegrees=45, twistLimitDegrees=5}
 {bone1Name=dyn_hair_fr_r_02, bone2Name=dyn_hair_fr_r_01, swing1LimitDegrees=15, swing2LimitDegrees=45, twistLimitDegrees=5}
 {bone1Name=dyn_hair_fr_r_03, bone2Name=dyn_hair_fr_r_02, swing1LimitDegrees=15, swing2LimitDegrees=45, twistLimitDegrees=5}
 {bone1Name=dyn_hair_fr_r_04, bone2Name=dyn_hair_fr_r_03, swing1LimitDegrees=5, swing2LimitDegrees=45, twistLimitDegrees=5}
 {
  bone1Name=dyn_hair_fr_l_MSTR
  bone2Name=head
  swing1LimitDegrees=45
  swing1Motion=Free
  swing2LimitDegrees=45
  swing2Motion=Free
  twistLimitDegrees=45
  twistMotion=Free
 }
 {
  bone1Name=dyn_hair_l_MSTR
  bone2Name=head
  swing1LimitDegrees=45
  swing1Motion=Free
  swing2LimitDegrees=45
  swing2Motion=Free
  twistLimitDegrees=45
  twistMotion=Free
 }
 {
  bone1Name=dyn_hair_bk_l_MSTR
  bone2Name=head
  swing1LimitDegrees=45
  swing1Motion=Free
  swing2LimitDegrees=45
  swing2Motion=Free
  twistLimitDegrees=45
  twistMotion=Free
 }
 {
  bone1Name=dyn_hair_bk_MSTR
  bone2Name=head
  swing1LimitDegrees=45
  swing1Motion=Free
  swing2LimitDegrees=45
  swing2Motion=Free
  twistLimitDegrees=45
  twistMotion=Free
 }
 {
  bone1Name=dyn_hair_bk_r_MSTR
  bone2Name=head
  swing1LimitDegrees=45
  swing1Motion=Free
  swing2LimitDegrees=45
  swing2Motion=Free
  twistLimitDegrees=45
  twistMotion=Free
 }
 {
  bone1Name=dyn_hair_r_MSTR
  bone2Name=head
  swing1LimitDegrees=45
  swing1Motion=Free
  swing2LimitDegrees=45
  swing2Motion=Free
  twistLimitDegrees=45
  twistMotion=Free
 }
 {
  bone1Name=dyn_hair_fr_r_MSTR
  bone2Name=head
  swing1LimitDegrees=45
  swing1Motion=Free
  swing2LimitDegrees=45
  swing2Motion=Free
  twistLimitDegrees=45
  twistMotion=Free
 }
 {
  bone1Name=head
  bone2Name=neck_02
  swing1LimitDegrees=45
  swing1Motion=Free
  swing2LimitDegrees=45
  swing2Motion=Free
  twistLimitDegrees=45
  twistMotion=Free
 }
 {
  bone1Name=neck_02
  bone2Name=neck_01
  swing1LimitDegrees=45
  swing1Motion=Free
  swing2LimitDegrees=45
  swing2Motion=Free
  twistLimitDegrees=45
  twistMotion=Free
 }
 {
  bone1Name=neck_01
  bone2Name=spine_05
  swing1LimitDegrees=45
  swing1Motion=Free
  swing2LimitDegrees=45
  swing2Motion=Free
  twistLimitDegrees=45
  twistMotion=Free
 }
 {
  bone1Name=clavicle_r
  bone2Name=spine_05
  swing1LimitDegrees=45
  swing1Motion=Free
  swing2LimitDegrees=45
  swing2Motion=Free
  twistLimitDegrees=45
  twistMotion=Free
 }
 {
  bone1Name=clavicle_l
  bone2Name=spine_05
  swing1LimitDegrees=45
  swing1Motion=Free
  swing2LimitDegrees=45
  swing2Motion=Free
  twistLimitDegrees=45
  twistMotion=Free
 }
 {
  bone1Name=dyn_hair_bk_r_03
  bone2Name=dyn_hair_bk_03
  swing1LimitDegrees=45
  swing1Motion=Free
  swing2LimitDegrees=45
  swing2Motion=Free
  twistLimitDegrees=45
  twistMotion=Free
 }
 {
  bone1Name=dyn_hair_bk_l_03
  bone2Name=dyn_hair_bk_03
  swing1LimitDegrees=45
  swing1Motion=Free
  swing2LimitDegrees=45
  swing2Motion=Free
  twistLimitDegrees=45
  twistMotion=Free
 }
 {
  bone1Name=dyn_hair_r_03
  bone2Name=dyn_hair_bk_r_03
  swing1LimitDegrees=45
  swing1Motion=Free
  swing2LimitDegrees=45
  swing2Motion=Free
  twistLimitDegrees=45
  twistMotion=Free
 }
 {
  bone1Name=dyn_hair_fr_r_03
  bone2Name=dyn_hair_r_03
  swing1LimitDegrees=45
  swing1Motion=Free
  swing2LimitDegrees=45
  swing2Motion=Free
  twistLimitDegrees=45
  twistMotion=Free
 }
 {
  bone1Name=dyn_hair_l_03
  bone2Name=dyn_hair_bk_l_03
  swing1LimitDegrees=45
  swing1Motion=Free
  swing2LimitDegrees=45
  swing2Motion=Free
  twistLimitDegrees=45
  twistMotion=Free
 }
 {
  bone1Name=dyn_hair_fr_l_03
  bone2Name=dyn_hair_l_03
  swing1LimitDegrees=45
  swing1Motion=Free
  swing2LimitDegrees=45
  swing2Motion=Free
  twistLimitDegrees=45
  twistMotion=Free
 }
 {bone1Name=dyn_hair_fr_r_strand_01, bone2Name=dyn_hair_fr_r_02, swing1LimitDegrees=15, swing2LimitDegrees=15, twistLimitDegrees=5}
 {bone1Name=dyn_hair_fr_r_strand_02, bone2Name=dyn_hair_fr_r_strand_01, swing1LimitDegrees=15, swing2LimitDegrees=15, twistLimitDegrees=5}
 {
  bone1Name=upperarm_l
  bone2Name=clavicle_l
  swing1LimitDegrees=45
  swing1Motion=Free
  swing2LimitDegrees=45
  swing2Motion=Free
  twistLimitDegrees=45
  twistMotion=Free
 }
 {
  bone1Name=upperarm_r
  bone2Name=clavicle_r
  swing1LimitDegrees=45
  swing1Motion=Free
  swing2LimitDegrees=45
  swing2Motion=Free
  twistLimitDegrees=45
  twistMotion=Free
 }
]
