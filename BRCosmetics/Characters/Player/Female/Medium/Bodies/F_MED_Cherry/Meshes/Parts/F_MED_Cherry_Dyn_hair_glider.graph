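# Physics asset for the Cherry female character's glider hair, with capsule bodies on dynamic hair bones.
# A seven-segment main chain has swing limits widening toward the tip, its first segment locked to a master bone that swings and twists freely on the scaler root.
# Short, tightly limited branches attach through their own master bones: back-left at segment two, end-left and end-top at segment four, side-right at the scaler.
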PhysicsAsset /BRCosmetics/Characters/Player/Female/Medium/Bodies/F_MED_Cherry/Meshes/Parts/F_MED_Cherry_Dyn_hair_glider
bodies {
 dyn_glider_hair_scaler [{shapeType=Capsule, length=1, radius=1}]
 dyn_hair_glider_01 [{shapeType=Capsule, length=4.3503, radius=5, center=<2.5886 -0.0001 0>, rotation=<rot 86.3099 -88.6656 -96.4255>}]
 dyn_hair_glider_02 [{shapeType=Capsule, length=4.3503, radius=5, center=<2.5886 -0.0001 0>, rotation=<rot 86.3096 -88.6646 -96.4244>}]
 dyn_hair_glider_03 [{shapeType=Capsule, length=4.3503, radius=5, center=<2.5886 -0.0001 0>, rotation=<rot 86.3099 -88.6655 -96.4252>}]
 dyn_hair_glider_04 [{shapeType=Capsule, length=4.3503, radius=5, center=<2.5886 -0.0001 0>, rotation=<rot 86.3099 -88.6655 -96.4252>}]
 dyn_hair_glider_05 [{shapeType=Capsule, length=4.3503, radius=5, center=<2.5886 -0.0001 0>, rotation=<rot 86.3099 -88.6656 -96.4253>}]
 dyn_hair_glider_06 [{shapeType=Capsule, length=4.3503, radius=5, center=<2.5886 -0.0001 0>, rotation=<rot 86.3094 -88.6651 -96.4248>}]
 dyn_hair_glider_07 [{shapeType=Capsule, length=4.3503, radius=5, center=<2.5886 -0.0001 0>, rotation=<rot 86.3094 -88.6651 -96.4248>}]
 dyn_hair_glider_MSTR [{shapeType=Capsule, length=0.505, radius=0.505}]
 dyn_hair_glider_back_l_01 [{shapeType=Capsule, length=3.2253, radius=3.7812, center=<2.5886 0 0>, rotation=<rot 90 0 0>}]
 dyn_hair_glider_back_l_01_MSTR [{shapeType=Capsule, length=0.505, radius=0.505}]
 dyn_hair_glider_back_l_02 [{shapeType=Capsule, length=3.2253, radius=3.7812, center=<2.5886 0 0>, rotation=<rot 90 0 0>}]
 dyn_hair_glider_back_l_03 [{shapeType=Capsule, length=3.2253, radius=3.7812, center=<2.5886 0 0>, rotation=<rot 90 0 0>}]
 dyn_hair_glider_end_l_01 [{shapeType=Capsule, length=3.2253, radius=3.7812, center=<2.5886 0 0>, rotation=<rot 90 0 0>}]
 dyn_hair_glider_end_l_01_MSTR [{shapeType=Capsule, length=0.505, radius=0.505}]
 dyn_hair_glider_end_l_02 [{shapeType=Capsule, length=3.2253, radius=3.7812, center=<2.5886 0 0>, rotation=<rot 90 0 0>}]
 dyn_hair_glider_end_top_01 [{shapeType=Capsule, length=3.2253, radius=3.7812, center=<2.5886 0 0>, rotation=<rot 90 0 0>}]
 dyn_hair_glider_end_top_01_MSTR [{shapeType=Capsule, length=0.505, radius=0.505}]
 dyn_hair_glider_end_top_02 [{shapeType=Capsule, length=3.2253, radius=3.7812, center=<2.5886 0 0>, rotation=<rot 90 0 0>}]
 dyn_hair_glider_side_r_01 [{shapeType=Capsule, length=3.2253, radius=3.7812, center=<2.5886 0 0>, rotation=<rot 90 0 0>}]
 dyn_hair_glider_side_r_01_MSTR [{shapeType=Capsule, length=0.505, radius=0.505}]
 dyn_hair_glider_side_r_02 [{shapeType=Capsule, length=3.2253, radius=3.7812, center=<2.5886 0 0>, rotation=<rot 90 0 0>}]
 dyn_hair_glider_side_r_03 [{shapeType=Capsule, length=4.6003, radius=2.4062, center=<2.5886 0 0>, rotation=<rot 90 0 0>}]
}
constraints [
 {
  bone1Name=dyn_hair_glider_01
  bone2Name=dyn_hair_glider_MSTR
  swing1LimitDegrees=5
  swing1Motion=Locked
  swing2LimitDegrees=5
  swing2Motion=Locked
  twistLimitDegrees=5
  twistMotion=Locked
 }
 {bone1Name=dyn_hair_glider_02, bone2Name=dyn_hair_glider_01, swing1LimitDegrees=5, swing2LimitDegrees=5, twistLimitDegrees=5}
 {bone1Name=dyn_hair_glider_03, bone2Name=dyn_hair_glider_02, swing1LimitDegrees=10, swing2LimitDegrees=10, twistLimitDegrees=5}
 {bone1Name=dyn_hair_glider_04, bone2Name=dyn_hair_glider_03, swing1LimitDegrees=20, swing2LimitDegrees=20, twistLimitDegrees=5}
 {bone1Name=dyn_hair_glider_05, bone2Name=dyn_hair_glider_04, swing1LimitDegrees=45, swing2LimitDegrees=45, twistLimitDegrees=5}
 {bone1Name=dyn_hair_glider_06, bone2Name=dyn_hair_glider_05, swing1LimitDegrees=45, swing2LimitDegrees=45, twistLimitDegrees=5}
 {bone1Name=dyn_hair_glider_07, bone2Name=dyn_hair_glider_06, swing1LimitDegrees=45, swing2LimitDegrees=45, twistLimitDegrees=5}
 {bone1Name=dyn_hair_glider_end_l_01_MSTR, bone2Name=dyn_hair_glider_04, swing1LimitDegrees=2, swing2LimitDegrees=2, twistLimitDegrees=2}
 {bone1Name=dyn_hair_glider_end_l_01, bone2Name=dyn_hair_glider_end_l_01_MSTR, swing1LimitDegrees=5, swing2LimitDegrees=5, twistLimitDegrees=5}
 {bone1Name=dyn_hair_glider_end_l_02, bone2Name=dyn_hair_glider_end_l_01, swing1LimitDegrees=5, swing2LimitDegrees=5, twistLimitDegrees=5}
 {bone1Name=dyn_hair_glider_end_top_01_MSTR, bone2Name=dyn_hair_glider_04, swing1LimitDegrees=2, swing2LimitDegrees=2, twistLimitDegrees=2}
 {
  bone1Name=dyn_hair_glider_end_top_01
  bone2Name=dyn_hair_glider_end_top_01_MSTR
  swing1LimitDegrees=5
  swing2LimitDegrees=5
  twistLimitDegrees=5
 }
 {bone1Name=dyn_hair_glider_end_top_02, bone2Name=dyn_hair_glider_end_top_01, swing1LimitDegrees=5, swing2LimitDegrees=5, twistLimitDegrees=5}
 {bone1Name=dyn_hair_glider_back_l_01_MSTR, bone2Name=dyn_hair_glider_02, swing1LimitDegrees=2, swing2LimitDegrees=2, twistLimitDegrees=2}
 {
  bone1Name=dyn_hair_glider_back_l_01
  bone2Name=dyn_hair_glider_back_l_01_MSTR
  swing1LimitDegrees=5
  swing2LimitDegrees=5
  twistLimitDegrees=5
 }
 {bone1Name=dyn_hair_glider_back_l_02, bone2Name=dyn_hair_glider_back_l_01, swing1LimitDegrees=5, swing2LimitDegrees=5, twistLimitDegrees=5}
 {bone1Name=dyn_hair_glider_back_l_03, bone2Name=dyn_hair_glider_back_l_02, swing1LimitDegrees=5, swing2LimitDegrees=5, twistLimitDegrees=5}
 {
  bone1Name=dyn_hair_glider_side_r_01
  bone2Name=dyn_hair_glider_side_r_01_MSTR
  swing1LimitDegrees=5
  swing2LimitDegrees=5
  twistLimitDegrees=5
 }
 {bone1Name=dyn_hair_glider_side_r_02, bone2Name=dyn_hair_glider_side_r_01, swing1LimitDegrees=5, swing2LimitDegrees=5, twistLimitDegrees=5}
 {bone1Name=dyn_hair_glider_side_r_03, bone2Name=dyn_hair_glider_side_r_02, swing1LimitDegrees=5, swing2LimitDegrees=5, twistLimitDegrees=5}
 {
  bone1Name=dyn_hair_glider_MSTR
  bone2Name=dyn_glider_hair_scaler
  swing1LimitDegrees=45
  swing1Motion=Free
  swing2LimitDegrees=45
  swing2Motion=Free
  twistLimitDegrees=45
  twistMotion=Free
 }
 {bone1Name=dyn_hair_glider_side_r_01_MSTR, bone2Name=dyn_glider_hair_scaler, swing1LimitDegrees=2, swing2LimitDegrees=2, twistLimitDegrees=2}
]
