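# Anim Blueprint for F_MED_Cherry's body part, driving dynamic hair with a special Cherry glider variant.
# Copied pose runs through Control Rig, head-relative AnimDynamics, then two RigidBody sims gated by inverse glider flags.
# The tagged glider, when deployed or in front end, swaps sims, shrinks the unused hair via scaler bones, toggles a material parameter and changes gravity/wind settings.
# Crouched assault-rifle aiming blends in additive hair bone corrections.
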AnimBlueprint /BRCosmetics/Characters/Player/Female/Medium/Bodies/F_MED_Cherry/Meshes/Parts/F_MED_Cherry_AnimBP
target_skeleton=&/F_MED_Cherry_Skeleton
graphs [
 AnimGraph {nodes [
   n1 = Control Rig(Source=Copy Pose From Mesh() {bUseAttachedParent=true}) {
    alphaBoolBlend {blendOption=Linear}
    controlRigAssetReference {blueprintRigClass=&/F_MED_Cherry_CtrlRig.F_MED_Cherry_CtrlRig_C}
   }
   n2 = Anim Dynamics - Bone: dyn_hair_l_02(ComponentPose=Local To Component(LocalPose=@n1), Alpha=1, ExternalForce="0,0,0") {
    angularDampingOverride=0.98
    bGravityOverrideInSimSpace=true
    bLinearSpring=true
    bOverrideAngularDamping=true
    bUseGravityOverride=true
    linearSpringConstant=500
    alphaBoolBlend {blendOption=Linear}
    boundBone {boneName=dyn_hair_l_02}
    componentLinearAccScale=<0.1 0.1 0.1>
    componentLinearVelScale=<0.5 0.5 0.5>
    gravityOverride {z=-100}
    physicsBodyDefinitions [{
      sphereCollisionRadius=0
      boundBone {boneName=dyn_hair_l_02}
      boxExtents=<30 30 30>
      constraintSetup {angularLimitsMax=<0 10 0>, angularLimitsMin=<0 0 0>, angularTarget=<1 0 0>}
      localJointOffset=<0 0 -4>
     }]
    relativeSpaceBone {boneName=head}
   }
   n3 = Anim Dynamics - Bone: dyn_hair_l_03(ComponentPose=@n2, Alpha=1, ExternalForce="0,0,0") {
    angularDampingOverride=0.98
    bGravityOverrideInSimSpace=true
    bLinearSpring=true
    bOverrideAngularDamping=true
    bUseGravityOverride=true
    linearSpringConstant=500
    alphaBoolBlend {blendOption=Linear}
    boundBone {boneName=dyn_hair_l_03}
    componentLinearAccScale=<0.1 0.1 0.1>
    componentLinearVelScale=<0.5 0.5 0.5>
    gravityOverride {z=-100}
    physicsBodyDefinitions [{
      sphereCollisionRadius=0
      boundBone {boneName=dyn_hair_l_03}
      boxExtents=<30 30 30>
      constraintSetup {angularLimitsMax=<0 10 0>, angularLimitsMin=<0 0 0>, angularTarget=<1 0 0>}
      localJointOffset=<0 0 -4>
     }]
    relativeSpaceBone {boneName=head}
   }
   n4 = Anim Dynamics - Bone: dyn_hair_bk_l_02(ComponentPose=@n3, Alpha=1, ExternalForce="0,0,0") {
    angularDampingOverride=0.98
    bGravityOverrideInSimSpace=true
    bLinearSpring=true
    bOverrideAngularDamping=true
    bUseGravityOverride=true
    linearSpringConstant=500
    alphaBoolBlend {blendOption=Linear}
    boundBone {boneName=dyn_hair_bk_l_02}
    componentLinearAccScale=<0.1 0.1 0.1>
    componentLinearVelScale=<0.5 0.5 0.5>
    gravityOverride {z=-100}
    physicsBodyDefinitions [{
      sphereCollisionRadius=0
      boundBone {boneName=dyn_hair_bk_l_02}
      boxExtents=<30 30 30>
      constraintSetup {angularLimitsMax=<0 10 0>, angularLimitsMin=<0 0 0>, angularTarget=<1 0 0>}
      localJointOffset=<0 0 -4>
     }]
    relativeSpaceBone {boneName=head}
   }
   n5 = Anim Dynamics - Bone: dyn_hair_l_03(ComponentPose=@n4, Alpha=1, ExternalForce="0,0,0") {
    angularDampingOverride=0.98
    bGravityOverrideInSimSpace=true
    bLinearSpring=true
    bOverrideAngularDamping=true
    bUseGravityOverride=true
    linearSpringConstant=500
    alphaBoolBlend {blendOption=Linear}
    boundBone {boneName=dyn_hair_l_03}
    componentLinearAccScale=<0.1 0.1 0.1>
    componentLinearVelScale=<0.5 0.5 0.5>
    gravityOverride {z=-100}
    physicsBodyDefinitions [{
      sphereCollisionRadius=0
      boundBone {boneName=dyn_hair_l_03}
      boxExtents=<30 30 30>
      constraintSetup {angularLimitsMax=<0 10 0>, angularLimitsMin=<0 0 0>, angularTarget=<1 0 0>}
      localJointOffset=<0 0 -4>
     }]
    relativeSpaceBone {boneName=head}
   }
   n6 = Anim Dynamics - Bone: dyn_hair_bk_02(ComponentPose=@n5, Alpha=1, ExternalForce="0,0,0") {
    angularDampingOverride=0.98
    bGravityOverrideInSimSpace=true
    bLinearSpring=true
    bOverrideAngularDamping=true
    bUseGravityOverride=true
    linearSpringConstant=500
    alphaBoolBlend {blendOption=Linear}
    boundBone {boneName=dyn_hair_bk_02}
    componentLinearAccScale=<0.1 0.1 0.1>
    componentLinearVelScale=<0.5 0.5 0.5>
    gravityOverride {z=-100}
    physicsBodyDefinitions [{
      sphereCollisionRadius=0
      boundBone {boneName=dyn_hair_bk_02}
      boxExtents=<30 30 30>
      constraintSetup {angularLimitsMax=<0 10 0>, angularLimitsMin=<0 0 0>, angularTarget=<1 0 0>}
      localJointOffset=<0 0 -4>
     }]
    relativeSpaceBone {boneName=head}
   }
   n7 = Anim Dynamics - Bone: dyn_hair_bk_03(ComponentPose=@n6, Alpha=1, ExternalForce="0,0,0") {
    angularDampingOverride=0.98
    bGravityOverrideInSimSpace=true
    bLinearSpring=true
    bOverrideAngularDamping=true
    bUseGravityOverride=true
    linearSpringConstant=500
    alphaBoolBlend {blendOption=Linear}
    boundBone {boneName=dyn_hair_bk_03}
    componentLinearAccScale=<0.1 0.1 0.1>
    componentLinearVelScale=<0.5 0.5 0.5>
    gravityOverride {z=-100}
    physicsBodyDefinitions [{
      sphereCollisionRadius=0
      boundBone {boneName=dyn_hair_bk_03}
      boxExtents=<30 30 30>
      constraintSetup {angularLimitsMax=<0 10 0>, angularLimitsMin=<0 0 0>, angularTarget=<1 0 0>}
      localJointOffset=<0 0 -4>
     }]
    relativeSpaceBone {boneName=head}
   }
   n8 = Get GravityOverride()
   n9 = RigidBody!(ComponentPose=@n7, Alpha=Get isCherryGliderInv() -> (isCherryGliderInv:double), OverrideWorldGravity:Vector=@n8, ExternalForce) -> (Pose) {
    bClampLinearTranslationLimitToRefPose=true
    bOverrideWorldGravity=true
    overridePhysicsAsset=&/F_MED_Cherry_Dyn_HairTweak
    simulationSpace=BaseBoneSpace
    alphaBoolBlend {blendOption=Linear}
    baseBoneRef {boneName=neck_02}
    componentAppliedLinearAccClamp=<3000 3000 3000>
    componentLinearAccScale=<0.4 0.3>
    componentLinearVelScale=<0.4 0.3>
    overrideWorldGravity {z=-980}
    simSpaceSettings {
     maxAngularAcceleration=1000
     maxAngularVelocity=0.3
     maxLinearAcceleration=700
     maxLinearVelocity=500
     velocityScaleZ=0
     worldAlpha=1
     externalLinearDragV=<0.5 0.5>
    }
   }
   n10 = RigidBody(ComponentPose=@n9, Alpha=Get isCherryGlider() -> (isCherryGlider:double), OverrideWorldGravity=@n8) {
    bClampLinearTranslationLimitToRefPose=true
    bOverrideWorldGravity=true
    overridePhysicsAsset=&/F_MED_Cherry_Dyn_hair_glider
    simulationSpace=BaseBoneSpace
    alphaBoolBlend {blendOption=Linear}
    baseBoneRef {boneName=neck_01}
    componentAppliedLinearAccClamp=<3000 3000 3000>
    overrideWorldGravity {z=-980}
    simSpaceSettings {
     maxAngularAcceleration=1000
     maxAngularVelocity=3
     maxLinearAcceleration=700
     maxLinearVelocity=500
     velocityScaleZ=0
     worldAlpha=1
     externalLinearDragV=<0.5 0.5>
    }
   }
   n11 = "Transform (Modify) Bone - Bone: dyn_glider_hair_scaler"(ComponentPose=@n10, Alpha=Get isCherryGliderInv(), Translation="0,0,0", Rotation="0,0,0", Scale="0.01,0.01,0.01") {
    rotationSpace=BCS_WorldSpace
    scaleMode=BMM_Replace
    translationSpace=BCS_BoneSpace
    alphaBoolBlend {blendOption=Linear}
    alphaScaleBiasClamp {bInterpResult=true, interpSpeedIncreasing=4}
    boneToModify {boneName=dyn_glider_hair_scaler}
    scale=<0.01 0.01 0.01>
   }
   n12 = "Transform (Modify) Bone - Bone: dyn_hair_scaler"(ComponentPose=@n11, Alpha=Get isCherryGlider(), Translation="0,0,0", Rotation="0,0,0", Scale="0.01,0.01,0.01") {
    rotationSpace=BCS_WorldSpace
    scaleMode=BMM_Replace
    translationSpace=BCS_BoneSpace
    alphaBoolBlend {blendOption=Linear}
    alphaScaleBiasClamp {bInterpResult=true, interpSpeedIncreasing=4}
    boneToModify {boneName=dyn_hair_scaler}
    scale=<0.01 0.01 0.01>
   }
   n13 = Get Hair_correction() -> (Hair_correction:double)
   n14 = "Transform (Modify) Bone - Bone: dyn_hair_bk_r_04"(ComponentPose=@n12, Alpha=@n13, Translation="-2,3,1", Rotation="-0.000,0.000,30", Scale="0.5,0.5,0.5") {
    rotationMode=BMM_Additive
    scaleMode=BMM_Additive
    translationMode=BMM_Additive
    alphaBoolBlend {blendOption=Linear}
    alphaScaleBiasClamp {bInterpResult=true, interpSpeedIncreasing=5}
    boneToModify {boneName=dyn_hair_bk_r_04}
    rotation {roll=30}
    scale=<0.5 0.5 0.5>
    translation=<-2 3 1>
   }
   n15 = "Transform (Modify) Bone - Bone: dyn_hair_bk_03"(ComponentPose=@n14, Alpha=@n13, Translation="-0.000,-0.000,1.203", Rotation="-30.381,8.640,27.702", Scale="0.8,0.8,0.8") {
    rotationMode=BMM_Additive
    scaleMode=BMM_Additive
    translationMode=BMM_Additive
    alphaBoolBlend {blendOption=Linear}
    alphaScaleBiasClamp {bInterpResult=true, interpSpeedIncreasing=5}
    boneToModify {boneName=dyn_hair_bk_03}
    rotation=<rot -30.381 8.64 27.702>
    scale=<0.8 0.8 0.8>
    translation {z=1.203}
   }
   n16 = "Transform (Modify) Bone - Bone: dyn_hair_bk_04"(ComponentPose=@n15, Alpha=@n13, Translation="0,0,0", Rotation="-23.825,-8.784,20.934", Scale="1,1,1") {
    rotationMode=BMM_Additive
    translationMode=BMM_Additive
    alphaBoolBlend {blendOption=Linear}
    alphaScaleBiasClamp {bInterpResult=true, interpSpeedIncreasing=5}
    boneToModify {boneName=dyn_hair_bk_04}
    rotation=<rot -23.825 -8.784 20.934>
   }
   Output Pose(Result=Component To Local(ComponentPose=@n16)) {layerGroup=None}
  ]}
 EventGraph {nodes [
   Event BlueprintUpdateAnimation()
   > n1 = Sequence!(execute) -> (then_0, then_1, then_2:Exec)
   > n2 = Cast To FortPlayerPawn(Object=GetOwningActor())
   n3 = Get ParachuteAttachment!(self:Fort Player Pawn Object Reference=@n2) -> (ParachuteAttachment:Fort Player Parachute Object Reference)
   > Is Valid(InputObject=@n3)
   n4 = Get ParachuteItemDef!(self:Fort Player Parachute Object Reference=@n3) -> (ParachuteItemDef:Athena Glider Item Definition Object Reference)
   > Is Valid(InputObject=@n4)
   n5 = HasAnyMatchingGameplayTags!(
    self:Gameplay Tag Asset Interface Interface=@n4
    TagContainer:"Gameplay Tag Container Structure (by ref)"=Get CherryGliderTag() -> (CherryGliderTag:Gameplay Tag Container Structure)
   ) -> (ReturnValue:bool)
   n6 = Get bIsInFrontEnd()
   n7 = Get bIsParachuteOpened()
   > Set UseDifferentHairGravityState(UseDifferentHairGravityState:bool=AND Boolean(A=AND Boolean(A=@n7, B=@n6), B=@n5)) -> (Output_Get:bool)
   n8 = Get bIsInFrontEnd()
   n9 = Get bIsParachuteOpened()
   > n10 = Branch(Condition=AND Boolean(A=@n5, B=OR Boolean(A=@n9, B=@n8)))
   n11 = GetPartSkeletalMeshComponent()
   > n12 = Get Dynamic Material Instance(SkeletalMeshComponent=@n11, ElementIndex=1)
   > SetScalarParameterValue(self=@n12, ParameterName=Hide Element 01, Value=0)
   > Set isCherryGlider(isCherryGlider:double=1) -> (Output_Get:double)
   > Set isCherryGliderInv(isCherryGliderInv:double=0) -> (Output_Get:double)
   n13 = Make Gravity Override Setting Struct!(
    WindFrequency_28_#1:Vector="1.5,7,8"
    WindAmplitude_31_#2:Vector="200,500,400"
    WindOffset_32_#3:Vector="0,-600,400"
    GravityOverride_15_#4:Vector="0,0,0"
    JointName_27_#5:Name=root
   ) -> (GravityOverrideSettingStruct:Gravity Override Setting Struct Structure) {is=title K2Node_MakeStruct}
   > Set Gravity Override Params(Gravity Override Params:Gravity Override Params Struct Structure=Make Gravity Override Params Struct!(
     Skydive_GoingUp_28_#6:Gravity Override Setting Struct Structure=(WindFrequency_28_#1=(X=1.5,Y=7,Z=8),WindAmplitude_31_#2=(X=200,Y=-500,Z=400),WindOffset_32_#3=(X=0,Y=0,Z=0),GravityOverride_15_#4=(X=0,Y=-300,Z=1500),JointName_27_#5="root")
     Skydive_FloatingDown_29_#7:Gravity Override Setting Struct Structure=(WindFrequency_28_#1=(X=1.5,Y=7,Z=8),WindAmplitude_31_#2=(X=200,Y=-500,Z=400),WindOffset_32_#3=(X=0,Y=-600,Z=400),GravityOverride_15_#4=(X=0,Y=-700,Z=600),JointName_27_#5="root")
     Skydive_DivingDown_30_#8:Gravity Override Setting Struct Structure=(WindFrequency_28_#1=(X=1.5,Y=12,Z=10),WindAmplitude_31_#2=(X=50,Y=300,Z=100),WindOffset_32_#3=(X=0,Y=0,Z=0),GravityOverride_15_#4=(X=-1500,Y=0,Z=0),JointName_27_#5="pelvis")
     Skydive_Parachuting_31_#9:Gravity Override Setting Struct Structure=@n13
     OnGround_Standing_32_#10:Gravity Override Setting Struct Structure=(WindFrequency_28_#1=(X=0,Y=0,Z=0),WindAmplitude_31_#2=(X=0,Y=0,Z=0),WindOffset_32_#3=(X=0,Y=0,Z=0),GravityOverride_15_#4=(X=0,Y=0,Z=-980),JointName_27_#5="root")
     OnGround_Moving_33_#11:Gravity Override Setting Struct Structure=(WindFrequency_28_#1=(X=0,Y=0,Z=0),WindAmplitude_31_#2=(X=0,Y=0,Z=0),WindOffset_32_#3=(X=0,Y=0,Z=0),GravityOverride_15_#4=(X=0,Y=0,Z=-980),JointName_27_#5="root")
     EmoteOrMelee_34_#12:Gravity Override Setting Struct Structure=(WindFrequency_28_#1=(X=0,Y=0,Z=0),WindAmplitude_31_#2=(X=0,Y=0,Z=0),WindOffset_32_#3=(X=0,Y=0,Z=0),GravityOverride_15_#4=(X=0,Y=0,Z=-980),JointName_27_#5="root")
     Swimming_35_#13:Gravity Override Setting Struct Structure=(WindFrequency_28_#1=(X=0,Y=0,Z=0),WindAmplitude_31_#2=(X=0,Y=0,Z=0),WindOffset_32_#3=(X=0,Y=0,Z=0),GravityOverride_15_#4=(X=-800,Y=0,Z=0),JointName_27_#5="pelvis")
     Front_End_36_#14:Gravity Override Setting Struct Structure=@n13
     DBNO_40_#15:Gravity Override Setting Struct Structure=(WindFrequency_28_#1=(X=0,Y=0,Z=0),WindAmplitude_31_#2=(X=0,Y=0,Z=0),WindOffset_32_#3=(X=0,Y=0,Z=0),GravityOverride_15_#4=(X=-980,Y=400,Z=0),JointName_27_#5="pelvis")
     DBNO_Carried_41_#16:Gravity Override Setting Struct Structure=(WindFrequency_28_#1=(X=0,Y=0,Z=0),WindAmplitude_31_#2=(X=0,Y=0,Z=0),WindOffset_32_#3=(X=0,Y=0,Z=0),GravityOverride_15_#4=(X=-980,Y=500,Z=-500),JointName_27_#5="pelvis")
     OnGround_Falling_45_#17:Gravity Override Setting Struct Structure=(WindFrequency_28_#1=(X=0,Y=0,Z=0),WindAmplitude_31_#2=(X=0,Y=0,Z=0),WindOffset_32_#3=(X=0,Y=0,Z=0),GravityOverride_15_#4=(X=0,Y=0,Z=-980),JointName_27_#5="root")
     Ziplining_46_#18:Gravity Override Setting Struct Structure=(WindFrequency_28_#1=(X=0,Y=0,Z=0),WindAmplitude_31_#2=(X=0,Y=0,Z=0),WindOffset_32_#3=(X=0,Y=0,Z=0),GravityOverride_15_#4=(X=0,Y=0,Z=-980),JointName_27_#5="root")
     Skydive_Surfing_49_#19:Gravity Override Setting Struct Structure=(WindFrequency_28_#1=(X=1.5,Y=7,Z=8),WindAmplitude_31_#2=(X=200,Y=-500,Z=400),WindOffset_32_#3=(X=0,Y=0,Z=0),GravityOverride_15_#4=(X=500,Y=-300,Z=-980),JointName_27_#5="root")
     Riding_Standing_54_#20:Gravity Override Setting Struct Structure=(WindFrequency_28_#1=(X=0,Y=0,Z=0),WindAmplitude_31_#2=(X=0,Y=0,Z=0),WindOffset_32_#3=(X=0,Y=0,Z=0),GravityOverride_15_#4=(X=0,Y=-350,Z=-980),JointName_27_#5="root")
     Riding_Moving_55_#21:Gravity Override Setting Struct Structure=(WindFrequency_28_#1=(X=0,Y=0,Z=0),WindAmplitude_31_#2=(X=0,Y=0,Z=0),WindOffset_32_#3=(X=0,Y=0,Z=0),GravityOverride_15_#4=(X=0,Y=-450,Z=-450),JointName_27_#5="root")
     Riding_Moving_Targeting_58_#22:Gravity Override Setting Struct Structure=(WindFrequency_28_#1=(X=0,Y=0,Z=0),WindAmplitude_31_#2=(X=0,Y=0,Z=0),WindOffset_32_#3=(X=0,Y=0,Z=0),GravityOverride_15_#4=(X=0,Y=-200,Z=-980),JointName_27_#5="root")
     Tactical_Sprint_57_#23:Gravity Override Setting Struct Structure=(WindFrequency_28_#1=(X=0,Y=0,Z=0),WindAmplitude_31_#2=(X=0,Y=0,Z=0),WindOffset_32_#3=(X=0,Y=0,Z=0),GravityOverride_15_#4=(X=0,Y=-450,Z=-500),JointName_27_#5="root")
     Rail_Grinding_61_#24:Gravity Override Setting Struct Structure=(WindFrequency_28_#1=(X=0,Y=0,Z=0),WindAmplitude_31_#2=(X=0,Y=0,Z=0),WindOffset_32_#3=(X=0,Y=0,Z=0),GravityOverride_15_#4=(X=0,Y=0,Z=-980),JointName_27_#5="root")
     DBNO_Tumbling_67_#25:Gravity Override Setting Struct Structure=(WindFrequency_28_#1=(X=0,Y=0,Z=0),WindAmplitude_31_#2=(X=0,Y=0,Z=0),WindOffset_32_#3=(X=0,Y=0,Z=0),GravityOverride_15_#4=(X=-2000,Y=0,Z=0),JointName_27_#5="pelvis")
     DBNO_RollFwd_68_#26:Gravity Override Setting Struct Structure=(WindFrequency_28_#1=(X=0,Y=0,Z=0),WindAmplitude_31_#2=(X=0,Y=0,Z=0),WindOffset_32_#3=(X=0,Y=0,Z=0),GravityOverride_15_#4=(X=-2000,Y=0,Z=0),JointName_27_#5="pelvis")
     DBNO_RollBwd_69_#27:Gravity Override Setting Struct Structure=(WindFrequency_28_#1=(X=0,Y=0,Z=0),WindAmplitude_31_#2=(X=0,Y=0,Z=0),WindOffset_32_#3=(X=0,Y=0,Z=0),GravityOverride_15_#4=(X=-4000,Y=0,Z=0),JointName_27_#5="pelvis")
     DBNO_Impulse_70_#28:Gravity Override Setting Struct Structure=(WindFrequency_28_#1=(X=0,Y=0,Z=0),WindAmplitude_31_#2=(X=0,Y=0,Z=0),WindOffset_32_#3=(X=0,Y=0,Z=0),GravityOverride_15_#4=(X=-2000,Y=0,Z=0),JointName_27_#5="pelvis")
    ) -> (GravityOverrideParamsStruct:Gravity Override Params Struct Structure) {is=title K2Node_MakeStruct}) -> (Output_Get:Gravity Override Params Struct Structure)
   n14 = Get Dynamic Material Instance(execute=@n10.else, SkeletalMeshComponent=@n11, ElementIndex=1)
   > SetScalarParameterValue(self=@n14, ParameterName=Hide Element 01, Value=1)
   > Set isCherryGlider(isCherryGlider=0)
   > Set isCherryGliderInv(isCherryGliderInv=1)
   > Set Gravity Override Params(Gravity Override Params=Make Gravity Override Params Struct(
     Skydive_GoingUp_28_#6=(WindFrequency_28_#1=(X=1.5,Y=7,Z=8),WindAmplitude_31_#2=(X=200,Y=-500,Z=400),WindOffset_32_#3=(X=0,Y=0,Z=0),GravityOverride_15_#4=(X=0,Y=-300,Z=1500),JointName_27_#5="root")
     Skydive_FloatingDown_29_#7=(WindFrequency_28_#1=(X=1.5,Y=7,Z=8),WindAmplitude_31_#2=(X=200,Y=-500,Z=400),WindOffset_32_#3=(X=0,Y=-600,Z=400),GravityOverride_15_#4=(X=0,Y=-700,Z=600),JointName_27_#5="root")
     Skydive_DivingDown_30_#8=(WindFrequency_28_#1=(X=1.5,Y=12,Z=10),WindAmplitude_31_#2=(X=50,Y=300,Z=100),WindOffset_32_#3=(X=0,Y=0,Z=0),GravityOverride_15_#4=(X=-1500,Y=0,Z=0),JointName_27_#5="pelvis")
     Skydive_Parachuting_31_#9=Make Gravity Override Setting Struct(
      WindFrequency_28_#1="1.5,7,8"
      WindAmplitude_31_#2="100,300,200"
      WindOffset_32_#3="0,-600,400"
      GravityOverride_15_#4="0,0,-980"
      JointName_27_#5=root
     )
     OnGround_Standing_32_#10=(WindFrequency_28_#1=(X=0,Y=0,Z=0),WindAmplitude_31_#2=(X=0,Y=0,Z=0),WindOffset_32_#3=(X=0,Y=0,Z=0),GravityOverride_15_#4=(X=0,Y=0,Z=-980),JointName_27_#5="root")
     OnGround_Moving_33_#11=(WindFrequency_28_#1=(X=0,Y=0,Z=0),WindAmplitude_31_#2=(X=0,Y=0,Z=0),WindOffset_32_#3=(X=0,Y=0,Z=0),GravityOverride_15_#4=(X=0,Y=0,Z=-980),JointName_27_#5="root")
     EmoteOrMelee_34_#12=(WindFrequency_28_#1=(X=0,Y=0,Z=0),WindAmplitude_31_#2=(X=0,Y=0,Z=0),WindOffset_32_#3=(X=0,Y=0,Z=0),GravityOverride_15_#4=(X=0,Y=0,Z=-980),JointName_27_#5="root")
     Swimming_35_#13=(WindFrequency_28_#1=(X=0,Y=0,Z=0),WindAmplitude_31_#2=(X=0,Y=0,Z=0),WindOffset_32_#3=(X=0,Y=0,Z=0),GravityOverride_15_#4=(X=-800,Y=0,Z=0),JointName_27_#5="pelvis")
     Front_End_36_#14=(WindFrequency_28_#1=(X=0,Y=0,Z=0),WindAmplitude_31_#2=(X=0,Y=0,Z=0),WindOffset_32_#3=(X=0,Y=0,Z=0),GravityOverride_15_#4=(X=0,Y=0,Z=-980),JointName_27_#5="root")
     DBNO_40_#15=(WindFrequency_28_#1=(X=0,Y=0,Z=0),WindAmplitude_31_#2=(X=0,Y=0,Z=0),WindOffset_32_#3=(X=0,Y=0,Z=0),GravityOverride_15_#4=(X=-980,Y=400,Z=0),JointName_27_#5="pelvis")
     DBNO_Carried_41_#16=(WindFrequency_28_#1=(X=0,Y=0,Z=0),WindAmplitude_31_#2=(X=0,Y=0,Z=0),WindOffset_32_#3=(X=0,Y=0,Z=0),GravityOverride_15_#4=(X=-980,Y=500,Z=-500),JointName_27_#5="pelvis")
     OnGround_Falling_45_#17=(WindFrequency_28_#1=(X=0,Y=0,Z=0),WindAmplitude_31_#2=(X=0,Y=0,Z=0),WindOffset_32_#3=(X=0,Y=0,Z=0),GravityOverride_15_#4=(X=0,Y=0,Z=-980),JointName_27_#5="root")
     Ziplining_46_#18=(WindFrequency_28_#1=(X=0,Y=0,Z=0),WindAmplitude_31_#2=(X=0,Y=0,Z=0),WindOffset_32_#3=(X=0,Y=0,Z=0),GravityOverride_15_#4=(X=0,Y=0,Z=-980),JointName_27_#5="root")
     Skydive_Surfing_49_#19=(WindFrequency_28_#1=(X=1.5,Y=7,Z=8),WindAmplitude_31_#2=(X=200,Y=-500,Z=400),WindOffset_32_#3=(X=0,Y=0,Z=0),GravityOverride_15_#4=(X=500,Y=-300,Z=-980),JointName_27_#5="root")
     Riding_Standing_54_#20=(WindFrequency_28_#1=(X=0,Y=0,Z=0),WindAmplitude_31_#2=(X=0,Y=0,Z=0),WindOffset_32_#3=(X=0,Y=0,Z=0),GravityOverride_15_#4=(X=0,Y=-350,Z=-980),JointName_27_#5="root")
     Riding_Moving_55_#21=(WindFrequency_28_#1=(X=0,Y=0,Z=0),WindAmplitude_31_#2=(X=0,Y=0,Z=0),WindOffset_32_#3=(X=0,Y=0,Z=0),GravityOverride_15_#4=(X=0,Y=-450,Z=-450),JointName_27_#5="root")
     Riding_Moving_Targeting_58_#22=(WindFrequency_28_#1=(X=0,Y=0,Z=0),WindAmplitude_31_#2=(X=0,Y=0,Z=0),WindOffset_32_#3=(X=0,Y=0,Z=0),GravityOverride_15_#4=(X=0,Y=-200,Z=-980),JointName_27_#5="root")
     Tactical_Sprint_57_#23=(WindFrequency_28_#1=(X=0,Y=0,Z=0),WindAmplitude_31_#2=(X=0,Y=0,Z=0),WindOffset_32_#3=(X=0,Y=0,Z=0),GravityOverride_15_#4=(X=0,Y=-450,Z=-500),JointName_27_#5="root")
     Rail_Grinding_61_#24=(WindFrequency_28_#1=(X=0,Y=0,Z=0),WindAmplitude_31_#2=(X=0,Y=0,Z=0),WindOffset_32_#3=(X=0,Y=0,Z=0),GravityOverride_15_#4=(X=0,Y=0,Z=-980),JointName_27_#5="root")
     DBNO_Tumbling_67_#25=(WindFrequency_28_#1=(X=0,Y=0,Z=0),WindAmplitude_31_#2=(X=0,Y=0,Z=0),WindOffset_32_#3=(X=0,Y=0,Z=0),GravityOverride_15_#4=(X=-2000,Y=0,Z=0),JointName_27_#5="pelvis")
     DBNO_RollFwd_68_#26=(WindFrequency_28_#1=(X=0,Y=0,Z=0),WindAmplitude_31_#2=(X=0,Y=0,Z=0),WindOffset_32_#3=(X=0,Y=0,Z=0),GravityOverride_15_#4=(X=-2000,Y=0,Z=0),JointName_27_#5="pelvis")
     DBNO_RollBwd_69_#27=(WindFrequency_28_#1=(X=0,Y=0,Z=0),WindAmplitude_31_#2=(X=0,Y=0,Z=0),WindOffset_32_#3=(X=0,Y=0,Z=0),GravityOverride_15_#4=(X=-4000,Y=0,Z=0),JointName_27_#5="pelvis")
     DBNO_Impulse_70_#28=(WindFrequency_28_#1=(X=0,Y=0,Z=0),WindAmplitude_31_#2=(X=0,Y=0,Z=0),WindOffset_32_#3=(X=0,Y=0,Z=0),GravityOverride_15_#4=(X=-2000,Y=0,Z=0),JointName_27_#5="pelvis")
    ))
   n15 = Branch(execute=@n1.then_1, Condition=Get UseDifferentHairGravityState() -> (UseDifferentHairGravityState:bool))
   n16 = GetGameTimeInSeconds()
   n17 = GetOwnerSkeletalMeshComponent()
   n18 = Self-Reference()
   > n19 = Random Vector Function!(
    execute
    CharacterPartAnimInstance:Custom Character Part Anim Instance Object Reference=@n18
    SkelMeshComponent:Mesh Component Object Reference=@n17
    JointName:Name=root
    GameTimeInSeconds:double=@n16
    TimeOffset:double=0
    PerlinScale:double=1
    Frequency:Vector="1.5,7,8"
    Amplitude:Vector="200,500,400"
    Offset:Vector="0,-600,400"
   ) -> (then, VectorOut:Vector)
   > n20 = AlignVectorWithJoint!(
    execute
    Vector:Vector="0, 0, 0"
    MeshComp:Mesh Component Object Reference=@n17
    PartAnimInstance:Custom Character Part Anim Instance Object Reference=@n18
    JointName:Name=root
   ) -> (then, ReturnValue:Vector)
   > Set GravityOverride(GravityOverride=vector + vector(A:Vector=@n19, B:Vector=@n20) -> (ReturnValue:Vector) {is=title K2Node_CommutativeAssociativeBinaryOperator})
   n21 = Gravity Override State Function DEPRECATED(execute=@n15.else, CharacterPartAnimInstance=@n18, GameTimeInSeconds=@n16, GravityOverrideParams=Get Gravity Override Params())
   > Set GravityOverride(GravityOverride=@n21)
   n22 = Get bIsTargeting()
   n23 = Get Rifle_ADS() -> (Rifle_ADS:bool)
   n24 = Get bIsCrouching()
   Set Hair_correction(execute=@n1.then_2, Hair_correction:double=SelectFloat(A=1, B=0, bPickA=AND Boolean!(A=@n23, B=@n22, C:bool=@n24) -> (ReturnValue))) -> (Output_Get:double)
   Event OnWeaponChanged()
   > Set Rifle_ADS(Rifle_ADS:bool="Equal (Enum)"(A:EFortWeaponCoreAnimation Enum=Get WeaponCoreAnim!() -> (WeaponCoreAnim), B:EFortWeaponCoreAnimation Enum=AssaultRifle)) -> (Output_Get:bool)
   Event OnCharacterCustomizationLoaded()
   > n25 = Cast To PlayerPawn_Athena(Object=TryGetPawnOwner())
   > Set Body_Mesh(Body_Mesh=GetSkeletalMeshForPartType(self=@n25))
   > Set Skel_Mesh(Skel_Mesh=GetPartSkeletalMeshComponent())
  ]}
]
variables [
 Body_Mesh {type=SkeletalMeshComponent}
 CherryGlider {type=SkeletalMesh, default=null}
 CherryGliderParachuteGravityOverride {type=Vector, default=<0 0 -980>}
 CherryGliderTag {type=GameplayTagContainer, default {gameplayTags [{tagName=Cosmetics.Glider.Cherry}]}}
 CherryGliderWindAmp {type=Vector}
 CherryGliderWindFreq {type=Vector}
 CherryGliderWindOffset {type=Vector}
 Current State Out {type=string, default=""}
 Gravity Override Params {
  type=GravityOverrideParamsStruct
  default {front_End {gravityOverride=<0 0 -600>, windAmplitude=<192 268 0>, windFrequency=<1.6 1.33 1.68>, windOffset=<0 -456 -118>}}
 }
 GravityOverride {type=Vector}
 GravityOverride_GliderHair {type=Vector}
 Hair_correction {type=number}
 InterpSpeedMax {type=number}
 isCherryGlider {type=number}
 isCherryGliderInv {type=number, default=1}
 NewVar {type=Transform, default {location=<0 0 0>, rotation=<rot 0 0 0>, scale=<1 1 1>}}
 NewVar_0 {type=number}
 Rifle_ADS {type=boolean}
 root_space_rotation {type=Rotator}
 Skel_Mesh {type=SkeletalMeshComponent}
 slope_foot_rot {type=Rotator}
 slope_root_trans {type=Vector}
 UseDifferentHairGravityState {type=boolean}
]
